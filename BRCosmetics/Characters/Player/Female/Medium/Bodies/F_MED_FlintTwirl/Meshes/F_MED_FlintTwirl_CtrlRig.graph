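# Control Rig for the F_MED_FlintTwirl female body that procedurally drives skirt bones and hanging chain bones from body motion.
# A DeltaRotation helper uses swing-twist to turn each thigh's rotation into a remapped, clamped 0–1 weight.
# Those weights, multiplied for combined poses, blend corrective transforms over parent-constrained skirt bones.
# Chain bones and nulls follow weighted constraints plus similar correctives, including a smoothed, eased crouching adjustment.
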ControlRigBlueprint /BRCosmetics/Characters/Player/Female/Medium/Bodies/F_MED_FlintTwirl/Meshes/F_MED_FlintTwirl_CtrlRig
graphs [
 Multiply_3_ContainedGraph {nodes [
   n1 = RigVMFunctionEntryNode()
   n2 = &RigVMFunction_MathDoubleMul(A=@n1.A, B=@n1.B)
   RigVMFunctionReturnNode(Result=&RigVMFunction_MathDoubleMul(B=@n1.D, A=&RigVMFunction_MathDoubleMul(B=@n1.C, A=@n2.Result).Result).Result)
  ]}
 New Function_ContainedGraph {nodes [
   n1 = RigVMFunctionEntryNode()
   RigVMFunction_MathBoolToInteger = &RigVMFunction_MathBoolToInteger(Value=@n1.onlyTwist)
   RigVMFunction_MathQuaternionSwingTwist = &RigVMFunction_MathQuaternionSwingTwist(
    twistAxis=@n1.twistAxis
    Input=&RigUnit_GetRelativeTransformForItem(bChildInitial=False, bParentInitial=False, child=@n1.child, parent=@n1.parent).RelativeTransform.Rotation
   )
   n2 = RigVMDispatchNode(
    Values=((),())
    Index=@RigVMFunction_MathBoolToInteger.Result
    Values.0=@RigVMFunction_MathQuaternionSwingTwist.Swing
    Values.1=@RigVMFunction_MathQuaternionSwingTwist.Twist
   ) {template="DISPATCH_RigVMDispatch_SelectInt32(in Index,in Values,out Result)"}
   RigVMFunction_MathQuaternionToEuler = &RigVMFunction_MathQuaternionToEuler(Value=@n2.Result, RotationOrder=@n1.Rotation Order)
   > DISPATCH_RigVMDispatch_Print = RigVMDispatchNode(ScreenColor=(R=1,G=1,B=1,A=1), ScreenDuration=1.5, Enabled=@n1.debugRotation, Value=@RigVMFunction_MathQuaternionToEuler.Result) {template="DISPATCH_RigVMDispatch_Print(in Prefix,in Value,in Enabled,in ScreenDuration,in ScreenColor)"}
   Cast = RigVMDispatchNode(Value=@n1.triggerAxis) {template="DISPATCH_RigVMDispatch_CastEnumToInt(in Value,out Result)"}
   DISPATCH_RigVMDispatch_SelectInt32 = RigVMDispatchNode(
    Values=(0,0,0)
    Index=@Cast.Result
    Values.0=@RigVMFunction_MathQuaternionToEuler.Result.X
    Values.1=@RigVMFunction_MathQuaternionToEuler.Result.Y
    Values.2=@RigVMFunction_MathQuaternionToEuler.Result.Z
   ) {template="DISPATCH_RigVMDispatch_SelectInt32(in Index,in Values,out Result)"}
   Remap = &RigVMFunction_MathDoubleRemap(
    bClamp=False
    Value=@DISPATCH_RigVMDispatch_SelectInt32.Result
    TargetMinimum=@n1.outputMinMax.X
    TargetMaximum=@n1.outputMinMax.Y
    SourceMinimum=@n1.rotationMinMax.X
    SourceMaximum=@n1.rotationMinMax.Y
   )
   Absolute = &RigVMFunction_MathDoubleAbs(Value=@Remap.Result)
   Multiply = &RigVMFunction_MathDoubleMul(
    A=&RigVMFunction_MathDoubleClamp(Value=RigVMDispatchNode(True=@Absolute.Result, Condition=@n1.absoluteValue, False=@Remap.Result) {template="DISPATCH_RigVMDispatch_If(in Condition,in True,in False,out Result)"}.Result).Result
    B=@n1.premult
   )
   > RigVMFunctionReturnNode(outputQuat=@n2.Result, outputWeight=@Multiply.Result)
  ]}
 New Function_ContainedGraph {nodes [
   n1 = RigVMFunctionEntryNode()
   RigVMFunction_Sequence = &RigVMFunction_Sequence(ExecuteContext=@n1.ExecuteContext)
   &RigUnit_ParentConstraint(
    AdvancedSettings=(InterpolationType=Average,RotationOrderForFilter=XZY)
    Child=(Type=Bone,Name="MSTR_frontSkirt_l")
    Filter=(TranslationFilter=(bX=false,bY=false,bZ=false),RotationFilter=(bX=True,bY=True,bZ=True),ScaleFilter=(bX=false,bY=false,bZ=false))
    Parents=((Item=(Type=Bone,Name="thigh_l"),Weight=1),(Item=(Type=Bone,Name="pelvis"),Weight=0))
    Weight=1
    ExecutePin=@RigVMFunction_Sequence.A
   )
   > &RigUnit_ParentConstraint(
    AdvancedSettings=(InterpolationType=Average,RotationOrderForFilter=XZY)
    Child=(Type=Bone,Name="MSTR_sideSkirt_l")
    Filter=(TranslationFilter=(bX=false,bY=false,bZ=false),RotationFilter=(bX=True,bY=True,bZ=True),ScaleFilter=(bX=false,bY=false,bZ=false))
    Parents=((Item=(Type=Bone,Name="thigh_l"),Weight=0.75),(Item=(Type=Bone,Name="pelvis"),Weight=0.25))
    Weight=1
   )
   > &RigUnit_ParentConstraint(
    AdvancedSettings=(InterpolationType=Average,RotationOrderForFilter=XZY)
    Child=(Type=Bone,Name="MSTR_backSkirt_l")
    Filter=(TranslationFilter=(bX=false,bY=false,bZ=false),RotationFilter=(bX=True,bY=True,bZ=True),ScaleFilter=(bX=false,bY=false,bZ=false))
    Parents=((Item=(Type=Bone,Name="thigh_l"),Weight=0.25),(Item=(Type=Bone,Name="pelvis"),Weight=0.75))
    Weight=1
   )
   > &RigUnit_ParentConstraint(
    AdvancedSettings=(InterpolationType=Average,RotationOrderForFilter=XZY)
    Child=(Type=Bone,Name="MSTR_frontSkirt_r")
    Filter=(TranslationFilter=(bX=false,bY=false,bZ=false),RotationFilter=(bX=True,bY=True,bZ=True),ScaleFilter=(bX=false,bY=false,bZ=false))
    Parents=((Item=(Type=Bone,Name="thigh_r"),Weight=1),(Item=(Type=Bone,Name="pelvis"),Weight=0))
    Weight=1
   )
   > &RigUnit_ParentConstraint(
    AdvancedSettings=(InterpolationType=Average,RotationOrderForFilter=XZY)
    Child=(Type=Bone,Name="MSTR_sideSkirt_r")
    Filter=(TranslationFilter=(bX=false,bY=false,bZ=false),RotationFilter=(bX=True,bY=True,bZ=True),ScaleFilter=(bX=false,bY=false,bZ=false))
    Parents=((Item=(Type=Bone,Name="thigh_r"),Weight=0.75),(Item=(Type=Bone,Name="pelvis"),Weight=0.25))
    Weight=1
   )
   > &RigUnit_ParentConstraint(
    AdvancedSettings=(InterpolationType=Average,RotationOrderForFilter=XZY)
    Child=(Type=Bone,Name="MSTR_backSkirt_r")
    Filter=(TranslationFilter=(bX=false,bY=false,bZ=false),RotationFilter=(bX=True,bY=True,bZ=True),ScaleFilter=(bX=false,bY=false,bZ=false))
    Parents=((Item=(Type=Bone,Name="thigh_r"),Weight=0.25),(Item=(Type=Bone,Name="pelvis"),Weight=0.75))
    Weight=1
   )
   > &RigUnit_ParentConstraint(
    AdvancedSettings=(InterpolationType=Shortest,RotationOrderForFilter=ZYX)
    Child=(Type=Bone,Name="MSTR_frontSkirt")
    Filter=(TranslationFilter=(bX=True,bY=True,bZ=True),RotationFilter=(bX=True,bY=True,bZ=True),ScaleFilter=(bX=false,bY=false,bZ=false))
    Parents=((Item=(Type=Bone,Name="MSTR_frontSkirt_l"),Weight=0.5),(Item=(Type=Bone,Name="MSTR_frontSkirt_r"),Weight=0.5))
    Weight=1
   )
   > &RigUnit_ParentConstraint(
    AdvancedSettings=(InterpolationType=Shortest,RotationOrderForFilter=ZYX)
    Child=(Type=Bone,Name="MSTR_backSkirt")
    Parents=((Item=(Type=Bone,Name="MSTR_backSkirt_l"),Weight=0.5),(Item=(Type=Bone,Name="MSTR_backSkirt_r"),Weight=0.5))
    Weight=1
   )
   > RigVMFunction_Sequence_1 = RigVMAggregateNode()
   DeltaRotation v4 = RigVMFunctionReferenceNode(
    Rotation Order=ZYX
    child=(Type=Bone,Name="thigh_l")
    debugRotation=false
    outputMinMax=(X=0,Y=1)
    parent=(Type=Null,Name="thigh_l_null")
    premult=1
    rotationMinMax=(X=-20,Y=-120)
    triggerAxis=Z
    twistAxis=(X=1,Y=0,Z=0)
    ExecuteContext=@RigVMFunction_Sequence_1.A
   )
   > &RigUnit_ModifyTransforms(
    ItemToModify="((Item=(Type=Bone,Name=\"dyn_frontSkirt_01_l\"),Transform=(Rotation=(X=0.0895,Y=0.0311,Z=-0.1146,W=-0.9889),Translation=(X=1.2548,Y=4.8674,Z=8.4062),Scale3D=(X=1,Y=1,Z=1))),(Item=(Type=Bone,Name=\"dyn_sideSkirt_01_l\"),Transform=(Rotation=(X=0,Y=0,Z=0.061,W=0. ...(truncated)"
    WeightMinimum=0
    Weight=@"DeltaRotation v4".outputWeight
   )
   > DeltaRotation v4_8 = RigVMFunctionReferenceNode(
    Rotation Order=ZYX
    child=(Type=Bone,Name="thigh_r")
    outputMinMax=(X=0,Y=1)
    parent=(Type=Null,Name="thigh_r_null")
    premult=1
    rotationMinMax=(X=-20,Y=-120)
    triggerAxis=Z
    twistAxis=(X=1,Y=0,Z=0)
   )
   > &RigUnit_ModifyTransforms(
    ItemToModify="((Item=(Type=Bone,Name=\"dyn_frontSkirt_01_r\"),Transform=(Rotation=(X=-0.1688,Y=-0.1448,Z=-0.0824,W=0.9715),Translation=(X=2.0007,Y=-5.0008,Z=5.4996),Scale3D=(X=1,Y=1,Z=1))),(Item=(Type=Bone,Name=\"dyn_sideSkirt_01_r\"),Transform=(Rotation=(X=0,Y=0,Z=-0.061,W ...(truncated)"
    WeightMinimum=0
    Weight=@"DeltaRotation v4_8".outputWeight
   )
   Multiply = &RigVMFunction_MathDoubleMul(A=@"DeltaRotation v4_8".outputWeight, B=@"DeltaRotation v4".outputWeight)
   > &RigUnit_ModifyTransforms(
    ItemToModify="((Item=(Type=Bone,Name=\"dyn_frontSkirt_01_r\"),Transform=(Rotation=(X=0,Y=0,Z=0.0872,W=0.9962),Translation=(X=0,Y=0,Z=0),Scale3D=(X=1,Y=1,Z=1))),(Item=(Type=Bone,Name=\"dyn_frontSkirt_01_l\"),Transform=(Rotation=(X=0.0829,Y=0.0388,Z=-0.0832,W=0.9923),Translat ...(truncated)"
    WeightMinimum=0
    Weight=@Multiply.Result
   )
   > DeltaRotation v4_7 = RigVMFunctionReferenceNode(
    Rotation Order=ZYX
    child=(Type=Bone,Name="thigh_l")
    debugRotation=false
    outputMinMax=(X=0,Y=1)
    parent=(Type=Null,Name="thigh_l_null")
    premult=1
    rotationMinMax=(X=0,Y=-20)
    triggerAxis=Z
    twistAxis=(X=1,Y=0,Z=0)
   )
   > &RigUnit_ModifyTransforms(
    ItemToModify="((Item=(Type=Bone,Name=\"dyn_frontSkirt_01_l\"),Transform=(Rotation=(X=0,Y=0,Z=0,W=1),Translation=(X=0,Y=0,Z=0),Scale3D=(X=1,Y=1,Z=1))),(Item=(Type=Bone,Name=\"dyn_sideSkirt_01_l\"),Transform=(Rotation=(X=0,Y=0,Z=0,W=1),Translation=(X=0,Y=0,Z=0),Scale3D=(X=1,Y ...(truncated)"
    WeightMinimum=0
    Weight=@"DeltaRotation v4_7".outputWeight
   )
   > DeltaRotation v4_5 = RigVMFunctionReferenceNode(
    Rotation Order=ZYX
    child=(Type=Bone,Name="thigh_r")
    outputMinMax=(X=0,Y=1)
    parent=(Type=Null,Name="thigh_r_null")
    premult=1
    rotationMinMax=(X=0,Y=-20)
    triggerAxis=Z
    twistAxis=(X=1,Y=0,Z=0)
   )
   > &RigUnit_ModifyTransforms(
    ItemToModify="((Item=(Type=Bone,Name=\"dyn_frontSkirt_01_r\"),Transform=(Rotation=(X=0,Y=0,Z=0,W=1),Translation=(X=0,Y=0,Z=0),Scale3D=(X=1,Y=1,Z=1))),(Item=(Type=Bone,Name=\"dyn_sideSkirt_01_r\"),Transform=(Rotation=(X=0,Y=0,Z=0,W=1),Translation=(X=0,Y=0,Z=0),Scale3D=(X=1,Y ...(truncated)"
    WeightMinimum=0
    Weight=@"DeltaRotation v4_5".outputWeight
   )
   > &RigUnit_ModifyTransforms(
    ItemToModify=((Item=(Type=Bone,Name="dyn_frontSkirt_01"),Transform=(Rotation=(X=0,Y=0,Z=0,W=1),Translation=(X=0,Y=0,Z=0),Scale3D=(X=1,Y=1,Z=1))))
    WeightMinimum=0
    Weight=&RigVMFunction_MathDoubleMul(B=@"DeltaRotation v4_5".outputWeight, A=@"DeltaRotation v4_7".outputWeight).Result
   )
   DeltaRotation v4_9 = RigVMFunctionReferenceNode(
    Rotation Order=ZYX
    child=(Type=Bone,Name="thigh_l")
    outputMinMax=(X=0,Y=1)
    parent=(Type=Null,Name="thigh_l_null")
    premult=1
    rotationMinMax=(X=0,Y=75)
    triggerAxis=Z
    twistAxis=(X=1,Y=0,Z=0)
    ExecuteContext=@RigVMFunction_Sequence_1.B
   )
   > &RigUnit_ModifyTransforms(
    ItemToModify="((Item=(Type=Bone,Name=\"dyn_backSkirt_02\"),Transform=(Rotation=(X=0,Y=-0.3461,Z=0,W=0.9382),Translation=(X=0,Y=0,Z=0),Scale3D=(X=1,Y=1,Z=1))),(Item=(Type=Bone,Name=\"dyn_backSkirt_01_l\"),Transform=(Rotation=(X=-0.208,Y=0.3976,Z=0.2165,W=-0.867),Translation= ...(truncated)"
    WeightMinimum=0
    Weight=@"DeltaRotation v4_9".outputWeight
   )
   > DeltaRotation v4_1 = RigVMFunctionReferenceNode(
    Rotation Order=ZYX
    child=(Type=Bone,Name="thigh_r")
    debugRotation=false
    outputMinMax=(X=0,Y=1)
    parent=(Type=Null,Name="thigh_r_null")
    premult=1
    rotationMinMax=(X=0,Y=75)
    triggerAxis=Z
    twistAxis=(X=1,Y=0,Z=0)
   )
   > &RigUnit_ModifyTransforms(
    ItemToModify="((Item=(Type=Bone,Name=\"dyn_backSkirt_02\"),Transform=(Rotation=(X=0,Y=-0.3461,Z=0,W=0.9382),Translation=(X=0,Y=0,Z=0),Scale3D=(X=1,Y=1,Z=1))),(Item=(Type=Bone,Name=\"dyn_backSkirt_01_r\"),Transform=(Rotation=(X=0.2055,Y=0.4076,Z=-0.2189,W=-0.8624),Translatio ...(truncated)"
    WeightMinimum=0
    Weight=@"DeltaRotation v4_1".outputWeight
   )
   > &RigUnit_ModifyTransforms(
    ItemToModify="((Item=(Type=Bone,Name=\"dyn_backSkirt_02\"),Transform=(Rotation=(X=-0.059,Y=-0.7046,Z=-0.0249,W=-0.7067),Translation=(X=1.3598,Y=-0.0658,Z=0),Scale3D=(X=1,Y=1,Z=1))),(Item=(Type=Bone,Name=\"dyn_frontSkirt_01\"),Transform=(Rotation=(X=0,Y=0,Z=0,W=1),Translatio ...(truncated)"
    WeightMinimum=0
    Weight=&RigVMFunction_MathDoubleMul(A=@"DeltaRotation v4_1".outputWeight, B=@"DeltaRotation v4_9".outputWeight).Result
   )
   > &RigUnit_ModifyTransforms(
    ItemToModify="((Item=(Type=Bone,Name=\"dyn_backSkirt_01_r\"),Transform=(Rotation=(X=0.0046,Y=-0.0434,Z=0.1044,W=0.9936),Translation=(X=0,Y=0,Z=0),Scale3D=(X=1,Y=1,Z=1))),(Item=(Type=Bone,Name=\"dyn_frontSkirt_01\"),Transform=(Rotation=(X=-0.3353,Y=-0.8688,Z=-0.3072,W=-0.195 ...(truncated)"
    WeightMinimum=0
    Weight=&RigVMFunction_MathDoubleMul(A=@"DeltaRotation v4".outputWeight, B=@"DeltaRotation v4_1".outputWeight).Result
   )
   > &RigUnit_ModifyTransforms(
    ItemToModify="((Item=(Type=Bone,Name=\"dyn_backSkirt_01_l\"),Transform=(Rotation=(X=0.0946,Y=0.0106,Z=-0.0414,W=-0.9946),Translation=(X=-0.0152,Y=0.0984,Z=0.5157),Scale3D=(X=1,Y=1,Z=1))),(Item=(Type=Bone,Name=\"dyn_frontSkirt_01\"),Transform=(Rotation=(X=0.4362,Y=-0.8261,Z= ...(truncated)"
    WeightMinimum=0
    Weight=&RigVMFunction_MathDoubleMul(A=@"DeltaRotation v4_8".outputWeight, B=@"DeltaRotation v4_9".outputWeight).Result
   )
   > DeltaRotation v4_6 = RigVMFunctionReferenceNode(
    Rotation Order=ZYX
    child=(Type=Bone,Name="thigh_l")
    outputMinMax=(X=0,Y=1)
    parent=(Type=Null,Name="thigh_l_null")
    premult=1
    rotationMinMax=(X=0,Y=12)
    triggerAxis=Z
    twistAxis=(X=1,Y=0,Z=0)
   )
   > &RigUnit_ModifyTransforms(
    ItemToModify="((Item=(Type=Bone,Name=\"dyn_backSkirt_01_l\"),Transform=(Rotation=(X=0,Y=0,Z=0.0523,W=0.9986),Translation=(X=0,Y=0,Z=0),Scale3D=(X=1,Y=1,Z=1))),(Item=(Type=Bone,Name=\"dyn_sideSkirt_01_l\"),Transform=(Rotation=(X=0,Y=0,Z=0,W=1),Translation=(X=0,Y=0,Z=0),Scale ...(truncated)"
    WeightMinimum=0
    Weight=@"DeltaRotation v4_6".outputWeight
   )
   > DeltaRotation v4_11 = RigVMFunctionReferenceNode(
    Rotation Order=ZYX
    child=(Type=Bone,Name="thigh_r")
    outputMinMax=(X=0,Y=1)
    parent=(Type=Null,Name="thigh_r_null")
    premult=1
    rotationMinMax=(X=0,Y=12)
    triggerAxis=Z
    twistAxis=(X=1,Y=0,Z=0)
   )
   > &RigUnit_ModifyTransforms(
    ItemToModify="((Item=(Type=Bone,Name=\"dyn_backSkirt_01_r\"),Transform=(Rotation=(X=0,Y=0,Z=0,W=1),Translation=(X=0,Y=0,Z=0),Scale3D=(X=1,Y=1,Z=1))),(Item=(Type=Bone,Name=\"dyn_sideSkirt_01_r\"),Transform=(Rotation=(X=0,Y=0,Z=0,W=1),Translation=(X=0,Y=0,Z=0),Scale3D=(X=1,Y= ...(truncated)"
    WeightMinimum=0
    Weight=@"DeltaRotation v4_11".outputWeight
   )
   > &RigUnit_ParentConstraint(
    AdvancedSettings=(InterpolationType=Average,RotationOrderForFilter=XZY)
    Child=(Type=Bone,Name="dyn_backSkirt_01")
    Parents=((Item=(Type=Bone,Name="dyn_backSkirt_01_l"),Weight=0.5),(Item=(Type=Bone,Name="dyn_backSkirt_01_r"),Weight=0.5))
    Weight=1
   )
   DeltaRotation v4_10 = RigVMFunctionReferenceNode(
    Rotation Order=ZYX
    child=(Type=Bone,Name="thigh_l")
    outputMinMax=(X=0,Y=1)
    parent=(Type=Null,Name="thigh_l_null")
    premult=1
    rotationMinMax=(X=0,Y=-90)
    triggerAxis=Y
    twistAxis=(X=1,Y=0,Z=0)
    ExecuteContext=@RigVMFunction_Sequence_1.C
   )
   > &RigUnit_ModifyTransforms(
    ItemToModify="((Item=(Type=Bone,Name=\"dyn_frontSkirt_01_l\"),Transform=(Rotation=(X=0.0208,Y=0.0421,Z=-0.3079,W=-0.9503),Translation=(X=2,Y=0,Z=0),Scale3D=(X=1,Y=1,Z=1))),(Item=(Type=Bone,Name=\"dyn_sideSkirt_01_l\"),Transform=(Rotation=(X=0.123,Y=0.1544,Z=-0.1085,W=-0.974 ...(truncated)"
    WeightMinimum=0
    Weight=@"DeltaRotation v4_10".outputWeight
   )
   > DeltaRotation v4_2 = RigVMFunctionReferenceNode(
    Rotation Order=ZYX
    child=(Type=Bone,Name="thigh_r")
    outputMinMax=(X=0,Y=1)
    parent=(Type=Null,Name="thigh_r_null")
    premult=1
    rotationMinMax=(X=0,Y=-90)
    triggerAxis=Y
    twistAxis=(X=1,Y=0,Z=0)
   )
   > &RigUnit_ModifyTransforms(
    ItemToModify="((Item=(Type=Bone,Name=\"dyn_frontSkirt_01_r\"),Transform=(Rotation=(X=0.0163,Y=-0.044,Z=-0.1995,W=0.9788),Translation=(X=1.9999,Y=0,Z=0),Scale3D=(X=1,Y=1,Z=1))),(Item=(Type=Bone,Name=\"dyn_sideSkirt_01_r\"),Transform=(Rotation=(X=0.0347,Y=0.2536,Z=0.0577,W=-0 ...(truncated)"
    WeightMinimum=0
    Weight=@"DeltaRotation v4_2".outputWeight
   )
   > &RigUnit_ModifyTransforms(
    ItemToModify="((Item=(Type=Bone,Name=\"dyn_backSkirt_01_l\"),Transform=(Rotation=(X=0,Y=0,Z=0,W=1),Translation=(X=0,Y=0,Z=0),Scale3D=(X=1,Y=1,Z=1))),(Item=(Type=Bone,Name=\"dyn_backSkirt_01_r\"),Transform=(Rotation=(X=0,Y=0,Z=0,W=1),Translation=(X=0,Y=0,Z=0),Scale3D=(X=1,Y= ...(truncated)"
    WeightMinimum=0
    Weight=&RigVMFunction_MathDoubleMul(A=@"DeltaRotation v4_2".outputWeight, B=@"DeltaRotation v4_10".outputWeight).Result
   )
   > DeltaRotation v4_3 = RigVMFunctionReferenceNode(
    Rotation Order=ZYX
    child=(Type=Bone,Name="thigh_l")
    outputMinMax=(X=0,Y=1)
    parent=(Type=Null,Name="thigh_l_null")
    premult=1
    rotationMinMax=(X=0,Y=30)
    triggerAxis=Y
    twistAxis=(X=1,Y=0,Z=0)
   )
   > &RigUnit_ModifyTransforms(
    ItemToModify="((Item=(Type=Bone,Name=\"dyn_frontSkirt_01_l\"),Transform=(Rotation=(X=-0.0114,Y=-0.0864,Z=-0.13,W=0.9877),Translation=(X=0,Y=0,Z=0),Scale3D=(X=1,Y=1,Z=1))),(Item=(Type=Bone,Name=\"dyn_sideSkirt_01_l\"),Transform=(Rotation=(X=0,Y=-0.1305,Z=0,W=0.9914),Translat ...(truncated)"
    WeightMinimum=0
    Weight=@"DeltaRotation v4_3".outputWeight
   )
   > DeltaRotation v4_4 = RigVMFunctionReferenceNode(
    Rotation Order=ZYX
    child=(Type=Bone,Name="thigh_r")
    outputMinMax=(X=0,Y=1)
    parent=(Type=Null,Name="thigh_r_null")
    premult=1
    rotationMinMax=(X=0,Y=30)
    triggerAxis=Y
    twistAxis=(X=1,Y=0,Z=0)
   )
   > &RigUnit_ModifyTransforms(
    ItemToModify="((Item=(Type=Bone,Name=\"dyn_frontSkirt_01_r\"),Transform=(Rotation=(X=0.0114,Y=-0.0864,Z=0.13,W=0.9877),Translation=(X=0,Y=0,Z=0),Scale3D=(X=1,Y=1,Z=1))),(Item=(Type=Bone,Name=\"dyn_sideSkirt_01_r\"),Transform=(Rotation=(X=0,Y=-0.1305,Z=0,W=0.9914),Translatio ...(truncated)"
    WeightMinimum=0
    Weight=@"DeltaRotation v4_4".outputWeight
   )
   DeltaRotation v4_12 = RigVMFunctionReferenceNode(
    Rotation Order=ZYX
    child=(Type=Bone,Name="thigh_l")
    onlyTwist=true
    outputMinMax=(X=0,Y=1)
    parent=(Type=Null,Name="thigh_l_null")
    premult=1
    rotationMinMax=(X=0,Y=-100)
    triggerAxis=X
    twistAxis=(X=1,Y=0,Z=0)
    ExecuteContext=@RigVMFunction_Sequence_1.D
   )
   > &RigUnit_ModifyTransforms(
    ItemToModify="((Item=(Type=Bone,Name=\"dyn_backSkirt_01\"),Transform=(Rotation=(X=0,Y=-0.0915,Z=0,W=0.9958),Translation=(X=0,Y=0,Z=0),Scale3D=(X=1,Y=1,Z=1))),(Item=(Type=Bone,Name=\"dyn_frontSkirt_01_l\"),Transform=(Rotation=(X=0.7564,Y=0.1016,Z=-0.1211,W=-0.6347),Translati ...(truncated)"
    WeightMinimum=0
    Weight=@"DeltaRotation v4_12".outputWeight
   )
   > DeltaRotation v4_13 = RigVMFunctionReferenceNode(
    Rotation Order=ZYX
    child=(Type=Bone,Name="thigh_r")
    onlyTwist=true
    outputMinMax=(X=0,Y=1)
    parent=(Type=Null,Name="thigh_r_null")
    premult=1
    rotationMinMax=(X=0,Y=-100)
    triggerAxis=X
    twistAxis=(X=1,Y=0,Z=0)
   )
   > &RigUnit_ModifyTransforms(
    ItemToModify="((Item=(Type=Bone,Name=\"dyn_backSkirt_01\"),Transform=(Rotation=(X=0,Y=-0.0915,Z=0,W=0.9958),Translation=(X=0,Y=0,Z=0),Scale3D=(X=1,Y=1,Z=1))),(Item=(Type=Bone,Name=\"dyn_frontSkirt_01_r\"),Transform=(Rotation=(X=0.73,Y=-0.1344,Z=-0.2503,W=0.6216),Translation ...(truncated)"
    WeightMinimum=0
    Weight=@"DeltaRotation v4_13".outputWeight
   )
   > DeltaRotation v4_14 = RigVMFunctionReferenceNode(
    Rotation Order=ZYX
    child=(Type=Bone,Name="thigh_l")
    onlyTwist=true
    outputMinMax=(X=0,Y=1)
    parent=(Type=Null,Name="thigh_l_null")
    premult=1
    rotationMinMax=(X=0,Y=45)
    triggerAxis=X
    twistAxis=(X=1,Y=0,Z=0)
   )
   > &RigUnit_ModifyTransforms(
    ItemToModify="((Item=(Type=Bone,Name=\"dyn_backSkirt_01\"),Transform=(Rotation=(X=0,Y=0,Z=0,W=1),Translation=(X=0,Y=0,Z=0),Scale3D=(X=1,Y=1,Z=1))),(Item=(Type=Bone,Name=\"dyn_frontSkirt_01_l\"),Transform=(Rotation=(X=-0.2601,Y=0.0817,Z=-0.0247,W=-0.9618),Translation=(X=0,Y= ...(truncated)"
    WeightMinimum=0
    Weight=@"DeltaRotation v4_14".outputWeight
   )
   > DeltaRotation v4_15 = RigVMFunctionReferenceNode(
    Rotation Order=ZYX
    child=(Type=Bone,Name="thigh_r")
    onlyTwist=true
    outputMinMax=(X=0,Y=1)
    parent=(Type=Null,Name="thigh_r_null")
    premult=1
    rotationMinMax=(X=0,Y=45)
    triggerAxis=X
    twistAxis=(X=1,Y=0,Z=0)
   )
   > &RigUnit_ModifyTransforms(
    ItemToModify="((Item=(Type=Bone,Name=\"dyn_frontSkirt_01_r\"),Transform=(Rotation=(X=-0.1066,Y=-0.0892,Z=-0.0096,W=0.9902),Translation=(X=0,Y=0,Z=0),Scale3D=(X=1,Y=1,Z=1))),(Item=(Type=Bone,Name=\"dyn_sideSkirt_01_r\"),Transform=(Rotation=(X=0,Y=0,Z=0,W=1),Translation=(X=0, ...(truncated)"
    WeightMinimum=0
    Weight=@"DeltaRotation v4_15".outputWeight
   )
   > &RigUnit_ModifyTransforms(
    ItemToModify=((Item=(Type=Bone,Name="dyn_frontSkirt_01_l"),Transform=(Rotation=(X=-0.0691,Y=-0.2504,Z=-0.1712,W=0.9504),Translation=(X=0,Y=1,Z=1),Scale3D=(X=1,Y=1,Z=1))))
    WeightMinimum=0
    Weight=&RigVMFunction_MathDoubleMul(B=1, A=@"DeltaRotation v4_14".outputWeight).Result
   )
   > &RigUnit_ModifyTransforms(
    ItemToModify=((Item=(Type=Bone,Name="dyn_frontSkirt_01_r"),Transform=(Rotation=(X=0.0172,Y=-0.1346,Z=0.0844,W=0.9871),Translation=(X=0,Y=0,Z=0),Scale3D=(X=1,Y=1,Z=1))))
    WeightMinimum=0
    Weight=&RigVMFunction_MathDoubleMul(B=@"DeltaRotation v4_15".outputWeight, A=@"DeltaRotation v4_8".outputWeight).Result
   )
   > DeltaRotation v4_16 = RigVMFunctionReferenceNode(
    Rotation Order=ZYX
    child=(Type=Bone,Name="thigh_l")
    onlyTwist=true
    outputMinMax=(X=0,Y=1)
    parent=(Type=Null,Name="thigh_l_null")
    premult=1
    rotationMinMax=(X=0,Y=-60)
    triggerAxis=X
    twistAxis=(X=1,Y=0,Z=0)
   )
   n2 = &RigVMFunction_MathDoubleMul(B=@"DeltaRotation v4".outputWeight, A=@"DeltaRotation v4_16".outputWeight)
   > &RigUnit_ModifyTransforms(
    ItemToModify="((Item=(Type=Bone,Name=\"dyn_frontSkirt_01_l\"),Transform=(Rotation=(X=-0.2638,Y=-0.166,Z=-0.2437,W=-0.9184),Translation=(X=0,Y=0.9999,Z=0.9999),Scale3D=(X=1,Y=1,Z=1))),(Item=(Type=Bone,Name=\"dyn_frontSkirt_01\"),Transform=(Rotation=(X=0.0053,Y=-0.133,Z=0.179 ...(truncated)"
    WeightMinimum=0
    Weight=@n2.Result
   )
   > DeltaRotation v4_17 = RigVMFunctionReferenceNode(
    Rotation Order=ZYX
    child=(Type=Bone,Name="thigh_r")
    onlyTwist=true
    outputMinMax=(X=0,Y=1)
    parent=(Type=Null,Name="thigh_r_null")
    premult=1
    rotationMinMax=(X=0,Y=-60)
    triggerAxis=X
    twistAxis=(X=1,Y=0,Z=0)
   )
   n3 = &RigVMFunction_MathDoubleMul(A=@"DeltaRotation v4_17".outputWeight, B=@"DeltaRotation v4_8".outputWeight)
   > &RigUnit_ModifyTransforms(
    ItemToModify="((Item=(Type=Bone,Name=\"dyn_frontSkirt_01_r\"),Transform=(Rotation=(X=0.0206,Y=0.2415,Z=-0.2058,W=0.9481),Translation=(X=-1.9611,Y=-2.6073,Z=-5.7824),Scale3D=(X=1,Y=1,Z=1))),(Item=(Type=Bone,Name=\"dyn_frontSkirt_01\"),Transform=(Rotation=(X=-0.2006,Y=-0.2199 ...(truncated)"
    WeightMinimum=0
    Weight=@n3.Result
   )
   > &RigUnit_ModifyTransforms(
    ItemToModify="((Item=(Type=Bone,Name=\"dyn_frontSkirt_01\"),Transform=(Rotation=(X=-0.0343,Y=-0.0026,Z=-0.0747,W=-0.9966),Translation=(X=-0.0103,Y=0.0683,Z=-1.0038),Scale3D=(X=1,Y=1,Z=1))),(Item=(Type=Bone,Name=\"dyn_frontSkirt_01_l\"),Transform=(Rotation=(X=-0.0752,Y=-0.03 ...(truncated)"
    WeightMinimum=0
    Weight=&RigVMFunction_MathDoubleMul(A=@"DeltaRotation v4_16".outputWeight, B=@"DeltaRotation v4_17".outputWeight).Result
   )
   > &RigUnit_ModifyTransforms(
    ItemToModify="((Item=(Type=Bone,Name=\"dyn_frontSkirt_01\"),Transform=(Rotation=(X=0.4007,Y=0.0276,Z=-0.2341,W=0.8854),Translation=(X=3.7054,Y=-1.8622,Z=-0.8005),Scale3D=(X=1,Y=1,Z=1))),(Item=(Type=Bone,Name=\"dyn_frontSkirt_01_l\"),Transform=(Rotation=(X=0.2721,Y=0.0491,Z= ...(truncated)"
    WeightMinimum=0
    Weight=&RigVMFunction_MathDoubleMul(A=@n2.Result, B=@n3.Result).Result
   )
   > &RigUnit_ModifyTransforms(
    ItemToModify="((Item=(Type=Bone,Name=\"dyn_sideSkirt_01_l\"),Transform=(Rotation=(X=0,Y=0,Z=0,W=1),Translation=(X=0,Y=0,Z=0),Scale3D=(X=1,Y=1,Z=1))),(Item=(Type=Bone,Name=\"dyn_backSkirt_01_l\"),Transform=(Rotation=(X=0,Y=0,Z=0,W=1),Translation=(X=0,Y=0,Z=0),Scale3D=(X=1,Y= ...(truncated)"
    WeightMinimum=0
    Weight=&RigVMFunction_MathDoubleMul(A=@"DeltaRotation v4_12".outputWeight, B=@"DeltaRotation v4_10".outputWeight).Result
   )
   > &RigUnit_ModifyTransforms(
    ItemToModify="((Item=(Type=Bone,Name=\"dyn_sideSkirt_01_r\"),Transform=(Rotation=(X=0.1254,Y=-0.1771,Z=-0.1832,W=0.9588),Translation=(X=0,Y=0,Z=0),Scale3D=(X=1,Y=1,Z=1))),(Item=(Type=Bone,Name=\"dyn_backSkirt_01_r\"),Transform=(Rotation=(X=0,Y=0,Z=0,W=1),Translation=(X=0,Y= ...(truncated)"
    WeightMinimum=0
    Weight=&RigVMFunction_MathDoubleMul(A=@"DeltaRotation v4_13".outputWeight, B=@"DeltaRotation v4_2".outputWeight).Result
   )
   > &RigUnit_ModifyTransforms(
    ItemToModify="((Item=(Type=Bone,Name=\"dyn_sideSkirt_01_l\"),Transform=(Rotation=(X=-0.0228,Y=0.0984,Z=0.0244,W=-0.9946),Translation=(X=0,Y=0,Z=0),Scale3D=(X=1,Y=1,Z=1))),(Item=(Type=Bone,Name=\"dyn_backSkirt_01_l\"),Transform=(Rotation=(X=0,Y=0.1077,Z=0,W=-0.9942),Translat ...(truncated)"
    WeightMinimum=0
    Weight=&RigVMFunction_MathDoubleMul(A=@"DeltaRotation v4_10".outputWeight, B=@"DeltaRotation v4_14".outputWeight).Result
   )
   > &RigUnit_ModifyTransforms(
    ItemToModify="((Item=(Type=Bone,Name=\"dyn_sideSkirt_01_r\"),Transform=(Rotation=(X=0.0223,Y=-0.0415,Z=-0.0733,W=-0.9962),Translation=(X=0,Y=0,Z=0),Scale3D=(X=1,Y=1,Z=1))),(Item=(Type=Bone,Name=\"dyn_backSkirt_01_r\"),Transform=(Rotation=(X=0,Y=0.1183,Z=0,W=-0.993),Translat ...(truncated)"
    WeightMinimum=0
    Weight=&RigVMFunction_MathDoubleMul(B=@"DeltaRotation v4_15".outputWeight, A=@"DeltaRotation v4_2".outputWeight).Result
   )
   > &RigUnit_ModifyTransforms(
    ItemToModify="((Item=(Type=Bone,Name=\"dyn_frontSkirt_01_l\"),Transform=(Rotation=(X=0,Y=0,Z=0,W=1),Translation=(X=0,Y=0,Z=0),Scale3D=(X=1,Y=1,Z=1))),(Item=(Type=Bone,Name=\"dyn_frontSkirt_01\"),Transform=(Rotation=(X=0,Y=0,Z=0,W=1),Translation=(X=0,Y=0,Z=0),Scale3D=(X=1,Y= ...(truncated)"
    WeightMaximum=0.5
    WeightMinimum=0
    Weight=&RigVMFunction_MathDoubleMul(A=@"DeltaRotation v4".outputWeight, B=@"DeltaRotation v4_10".outputWeight).Result
   )
   > &RigUnit_ModifyTransforms(
    ItemToModify="((Item=(Type=Bone,Name=\"dyn_frontSkirt_01_r\"),Transform=(Rotation=(X=0,Y=0,Z=0,W=1),Translation=(X=0,Y=1,Z=1),Scale3D=(X=1,Y=1,Z=1))),(Item=(Type=Bone,Name=\"dyn_frontSkirt_01\"),Transform=(Rotation=(X=0,Y=0,Z=0,W=1),Translation=(X=0,Y=0,Z=0),Scale3D=(X=1,Y= ...(truncated)"
    WeightMaximum=0.5
    WeightMinimum=0
    Weight=&RigVMFunction_MathDoubleMul(B=@"DeltaRotation v4_2".outputWeight, A=@"DeltaRotation v4_8".outputWeight).Result
   )
   RigVMFunction_MathBoolToFloat = &RigVMFunction_MathBoolToFloat(Value=RigVMVariableNode() {variable=driving}.Value)
   &RigUnit_ModifyTransforms(
    ItemToModify="((Item=(Type=Bone,Name=\"dyn_frontSkirt_01\"),Transform=(Rotation=(X=0.2166,Y=0.5022,Z=0.047,W=0.8359),Translation=(X=0,Y=0,Z=0),Scale3D=(X=1,Y=1,Z=1))),(Item=(Type=Bone,Name=\"dyn_frontSkirt_01_r\"),Transform=(Rotation=(X=0,Y=-0.0872,Z=0,W=0.9962),Translation ...(truncated)"
    WeightMinimum=0
    Weight=@RigVMFunction_MathBoolToFloat.Result
    ExecutePin=@RigVMFunction_Sequence_1.E
   )
   RigVMFunctionReturnNode(ExecuteContext=@RigVMFunction_Sequence.B)
  ]}
 New Function_ContainedGraph {nodes [
   RigVMFunctionEntryNode()
   > RigVMFunction_Sequence = RigVMAggregateNode()
   &RigUnit_ParentConstraint(
    AdvancedSettings=(RotationOrderForFilter=XZY)
    Child=(Type=Bone,Name="MSTR_chain_A_l")
    Parents="((Item=(Type=Bone,Name=\"spine_05\"),Weight=0.0024),(Item=(Type=Bone,Name=\"clavicle_r\"),Weight=0.0039),(Item=(Type=Bone,Name=\"clavicle_l\"),Weight=0.0041),(Item=(Type=Bone,Name=\"spine_01\"),Weight=0.0586),(Item=(Type=Bone,Name=\"pelvis\"),Weight=0.1204),(Item=(T ...(truncated)"
    Weight=1
    ExecutePin=@RigVMFunction_Sequence.A
   )
   > &RigUnit_ParentConstraint(
    AdvancedSettings=(RotationOrderForFilter=XZY)
    Child=(Type=Null,Name="null_chain_A_01_l")
    Parents="((Item=(Type=Bone,Name=\"spine_05\"),Weight=0.0024),(Item=(Type=Bone,Name=\"clavicle_r\"),Weight=0.0039),(Item=(Type=Bone,Name=\"clavicle_l\"),Weight=0.0041),(Item=(Type=Bone,Name=\"spine_01\"),Weight=0.0586),(Item=(Type=Bone,Name=\"pelvis\"),Weight=0.1204),(Item=(T ...(truncated)"
    Weight=1
   )
   > &RigUnit_ParentConstraint(
    AdvancedSettings=(RotationOrderForFilter=XZY)
    Child=(Type=Null,Name="null_chain_A_02_l")
    Parents="((Item=(Type=Bone,Name=\"MSTR_frontSkirt_l\"),Weight=0.01),(Item=(Type=Bone,Name=\"MSTR_frontSkirt\"),Weight=0.0222),(Item=(Type=Bone,Name=\"spine_04\"),Weight=0.0223),(Item=(Type=Bone,Name=\"spine_01\"),Weight=0.0252),(Item=(Type=Bone,Name=\"MSTR_hipRibbon\"),Weigh ...(truncated)"
    Weight=1
   )
   > &RigUnit_ParentConstraint(
    AdvancedSettings=(RotationOrderForFilter=XZY)
    Child=(Type=Null,Name="null_chain_A_03_l")
    Parents="((Item=(Type=Bone,Name=\"MSTR_frontSkirt_l\"),Weight=0.0138),(Item=(Type=Bone,Name=\"spine_01\"),Weight=0.0169),(Item=(Type=Bone,Name=\"MSTR_frontSkirt_r\"),Weight=0.0372),(Item=(Type=Bone,Name=\"MSTR_hipRibbon\"),Weight=0.0571),(Item=(Type=Bone,Name=\"spine_03\"),W ...(truncated)"
    Weight=1
   )
   > &RigUnit_ParentConstraint(
    AdvancedSettings=(RotationOrderForFilter=XZY)
    Child=(Type=Bone,Name="MSTR_chain_A_r")
    Parents="((Item=(Type=Bone,Name=\"MSTR_hipRibbon\"),Weight=0.0179),(Item=(Type=Bone,Name=\"spine_01\"),Weight=0.0493),(Item=(Type=Bone,Name=\"spine_03\"),Weight=0.1263),(Item=(Type=Bone,Name=\"spine_04\"),Weight=0.1705),(Item=(Type=Bone,Name=\"pelvis\"),Weight=0.1864),(Item= ...(truncated)"
    Weight=1
   )
   > &RigUnit_ParentConstraint(
    AdvancedSettings=(RotationOrderForFilter=XZY)
    Child=(Type=Null,Name="null_chain_A_01_r")
    Parents="((Item=(Type=Bone,Name=\"MSTR_hipRibbon\"),Weight=0.0179),(Item=(Type=Bone,Name=\"spine_01\"),Weight=0.0493),(Item=(Type=Bone,Name=\"spine_03\"),Weight=0.1263),(Item=(Type=Bone,Name=\"spine_04\"),Weight=0.1705),(Item=(Type=Bone,Name=\"pelvis\"),Weight=0.1864),(Item= ...(truncated)"
    Weight=1
   )
   > &RigUnit_ParentConstraint(
    AdvancedSettings=(RotationOrderForFilter=XZY)
    Child=(Type=Null,Name="null_chain_A_02_r")
    Parents=((Item=(Type=Bone,Name="dyn_hipRibbon_01_B_l"),Weight=1))
    Weight=1
    bMaintainOffset=true
   )
   > &RigUnit_ParentConstraint(
    AdvancedSettings=(RotationOrderForFilter=XZY)
    Child=(Type=Null,Name="null_chain_A_03_r")
    Parents="((Item=(Type=Bone,Name=\"MSTR_frontSkirt_l\"),Weight=0.0138),(Item=(Type=Bone,Name=\"spine_01\"),Weight=0.0169),(Item=(Type=Bone,Name=\"MSTR_frontSkirt_r\"),Weight=0.0372),(Item=(Type=Bone,Name=\"MSTR_hipRibbon\"),Weight=0.0571),(Item=(Type=Bone,Name=\"spine_03\"),W ...(truncated)"
    Weight=1
   )
   > &RigUnit_ParentConstraint(
    AdvancedSettings=(RotationOrderForFilter=XZY)
    Child=(Type=Bone,Name="MSTR_chain_B_l")
    Parents="((Item=(Type=Bone,Name=\"spine_05\"),Weight=0.0026),(Item=(Type=Bone,Name=\"clavicle_r\"),Weight=0.0036),(Item=(Type=Bone,Name=\"clavicle_l\"),Weight=0.0052),(Item=(Type=Bone,Name=\"spine_01\"),Weight=0.0619),(Item=(Type=Bone,Name=\"pelvis\"),Weight=0.1358),(Item=(T ...(truncated)"
    Weight=1
   )
   > &RigUnit_ParentConstraint(
    AdvancedSettings=(RotationOrderForFilter=XZY)
    Child=(Type=Null,Name="null_chain_B_01_l")
    Parents="((Item=(Type=Bone,Name=\"spine_05\"),Weight=0.0026),(Item=(Type=Bone,Name=\"clavicle_r\"),Weight=0.0036),(Item=(Type=Bone,Name=\"clavicle_l\"),Weight=0.0052),(Item=(Type=Bone,Name=\"spine_01\"),Weight=0.0619),(Item=(Type=Bone,Name=\"pelvis\"),Weight=0.1358),(Item=(T ...(truncated)"
    Weight=1
   )
   > &RigUnit_ParentConstraint(
    AdvancedSettings=(RotationOrderForFilter=XZY)
    Child=(Type=Null,Name="null_chain_B_02_l")
    Parents="((Item=(Type=Bone,Name=\"MSTR_frontSkirt\"),Weight=0.0142),(Item=(Type=Bone,Name=\"MSTR_frontSkirt_l\"),Weight=0.0152),(Item=(Type=Bone,Name=\"spine_04\"),Weight=0.0163),(Item=(Type=Bone,Name=\"spine_01\"),Weight=0.0273),(Item=(Type=Bone,Name=\"MSTR_hipRibbon\"),Wei ...(truncated)"
    Weight=1
   )
   > &RigUnit_ParentConstraint(
    AdvancedSettings=(RotationOrderForFilter=XZY)
    Child=(Type=Null,Name="null_chain_B_03_l")
    Parents="((Item=(Type=Bone,Name=\"dyn_frontSkirt_01\"),Weight=0.0229),(Item=(Type=Bone,Name=\"MSTR_frontSkirt_r\"),Weight=0.0279),(Item=(Type=Bone,Name=\"MSTR_hipRibbon\"),Weight=0.0348),(Item=(Type=Bone,Name=\"spine_03\"),Weight=0.0356),(Item=(Type=Bone,Name=\"spine_02\"),W ...(truncated)"
    Weight=1
   )
   > &RigUnit_ParentConstraint(
    AdvancedSettings=(RotationOrderForFilter=XZY)
    Child=(Type=Null,Name="null_chain_B_04_l")
    Parents="((Item=(Type=Bone,Name=\"dyn_sideSkirt_01_r\"),Weight=0.0222),(Item=(Type=Bone,Name=\"spine_02\"),Weight=0.0299),(Item=(Type=Bone,Name=\"MSTR_sideSkirt_r\"),Weight=0.0364),(Item=(Type=Bone,Name=\"dyn_frontSkirt_01\"),Weight=0.0737),(Item=(Type=Bone,Name=\"dyn_front ...(truncated)"
    Weight=1
   )
   > &RigUnit_ParentConstraint(
    AdvancedSettings=(RotationOrderForFilter=XZY)
    Child=(Type=Bone,Name="MSTR_chain_B_r")
    Parents="((Item=(Type=Bone,Name=\"MSTR_hipRibbon\"),Weight=0.0179),(Item=(Type=Bone,Name=\"spine_01\"),Weight=0.0493),(Item=(Type=Bone,Name=\"spine_03\"),Weight=0.1263),(Item=(Type=Bone,Name=\"spine_04\"),Weight=0.1705),(Item=(Type=Bone,Name=\"pelvis\"),Weight=0.1864),(Item= ...(truncated)"
    Weight=1
   )
   > &RigUnit_ParentConstraint(
    AdvancedSettings=(RotationOrderForFilter=XZY)
    Child=(Type=Null,Name="null_chain_B_01_r")
    Parents="((Item=(Type=Bone,Name=\"MSTR_hipRibbon\"),Weight=0.0179),(Item=(Type=Bone,Name=\"spine_01\"),Weight=0.0493),(Item=(Type=Bone,Name=\"spine_03\"),Weight=0.1263),(Item=(Type=Bone,Name=\"spine_04\"),Weight=0.1705),(Item=(Type=Bone,Name=\"pelvis\"),Weight=0.1864),(Item= ...(truncated)"
    Weight=1
   )
   > &RigUnit_ParentConstraint(
    AdvancedSettings=(RotationOrderForFilter=XZY)
    Child=(Type=Null,Name="null_chain_B_02_r")
    Parents="((Item=(Type=Bone,Name=\"spine_01\"),Weight=0.0289),(Item=(Type=Bone,Name=\"thigh_twist_01_r\"),Weight=0.0323),(Item=(Type=Bone,Name=\"MSTR_hipRibbon\"),Weight=0.0399),(Item=(Type=Bone,Name=\"MSTR_frontSkirt_r\"),Weight=0.0473),(Item=(Type=Bone,Name=\"MSTR_sideSkir ...(truncated)"
    Weight=1
   )
   > &RigUnit_ParentConstraint(
    AdvancedSettings=(RotationOrderForFilter=XZY)
    Child=(Type=Null,Name="null_chain_B_03_r")
    Parents="((Item=(Type=Bone,Name=\"dyn_frontSkirt_01_r\"),Weight=0.027),(Item=(Type=Bone,Name=\"MSTR_hipRibbon\"),Weight=0.0275),(Item=(Type=Bone,Name=\"spine_03\"),Weight=0.0366),(Item=(Type=Bone,Name=\"MSTR_frontSkirt\"),Weight=0.0613),(Item=(Type=Bone,Name=\"spine_02\"),We ...(truncated)"
    Weight=1
   )
   > &RigUnit_ParentConstraint(
    AdvancedSettings=(RotationOrderForFilter=XZY)
    Child=(Type=Null,Name="null_chain_B_04_r")
    Parents="((Item=(Type=Bone,Name=\"dyn_sideSkirt_01_r\"),Weight=0.0222),(Item=(Type=Bone,Name=\"spine_02\"),Weight=0.0299),(Item=(Type=Bone,Name=\"MSTR_sideSkirt_r\"),Weight=0.0364),(Item=(Type=Bone,Name=\"dyn_frontSkirt_01\"),Weight=0.0737),(Item=(Type=Bone,Name=\"dyn_front ...(truncated)"
    Weight=1
   )
   DeltaRotation v4_5 = RigVMFunctionReferenceNode(
    Rotation Order=ZYX
    child=(Type=Bone,Name="thigh_l")
    debugRotation=false
    outputMinMax=(X=0,Y=1)
    parent=(Type=Null,Name="thigh_l_null")
    premult=1
    rotationMinMax=(X=-50,Y=-100)
    triggerAxis=Z
    twistAxis=(X=1,Y=0,Z=0)
    ExecuteContext=@RigVMFunction_Sequence.B
   )
   > &RigUnit_ModifyTransforms(
    ItemToModify="((Item=(Type=Null,Name=\"null_chain_B_03_l\"),Transform=(Rotation=(X=0,Y=0,Z=0,W=1),Translation=(X=1.0688,Y=1.0585,Z=-1.3959),Scale3D=(X=1,Y=1,Z=1))),(Item=(Type=Null,Name=\"null_chain_B_04_l\"),Transform=(Rotation=(X=0,Y=0,Z=0,W=-1),Translation=(X=0,Y=-0.5176 ...(truncated)"
    WeightMinimum=0
    Weight=@"DeltaRotation v4_5".outputWeight
   )
   > DeltaRotation v4_6 = RigVMFunctionReferenceNode(
    Rotation Order=ZYX
    child=(Type=Bone,Name="thigh_r")
    outputMinMax=(X=0,Y=1)
    parent=(Type=Null,Name="thigh_r_null")
    premult=1
    rotationMinMax=(X=-50,Y=-100)
    triggerAxis=Z
    twistAxis=(X=1,Y=0,Z=0)
   )
   > &RigUnit_ModifyTransforms(
    ItemToModify="((Item=(Type=Null,Name=\"null_chain_B_03_r\"),Transform=(Rotation=(X=0,Y=0,Z=0,W=1),Translation=(X=0,Y=0,Z=1.4291),Scale3D=(X=1,Y=1,Z=1))),(Item=(Type=Null,Name=\"null_chain_B_04_l\"),Transform=(Rotation=(X=0,Y=0,Z=0,W=1),Translation=(X=-0.9053,Y=1.462,Z=0.001 ...(truncated)"
    WeightMinimum=0
    Weight=@"DeltaRotation v4_6".outputWeight
   )
   > &RigUnit_ModifyTransforms(
    ItemToModify="((Item=(Type=Null,Name=\"null_chain_A_02_r\"),Transform=(Rotation=(X=0,Y=0,Z=0,W=1),Translation=(X=3.0359,Y=-3.9288,Z=-6.7126),Scale3D=(X=1,Y=1,Z=1))),(Item=(Type=Null,Name=\"null_chain_A_03_l\"),Transform=(Rotation=(X=0,Y=0,Z=0,W=1),Translation=(X=2.0122,Y=0. ...(truncated)"
    WeightMinimum=0
    Weight=&RigVMFunction_MathDoubleMul(A=@"DeltaRotation v4_5".outputWeight, B=@"DeltaRotation v4_6".outputWeight).Result
   )
   > DeltaRotation v4_10 = RigVMFunctionReferenceNode(
    Rotation Order=ZYX
    child=(Type=Bone,Name="thigh_l")
    outputMinMax=(X=0,Y=1)
    parent=(Type=Null,Name="thigh_l_null")
    premult=1
    rotationMinMax=(X=0,Y=-90)
    triggerAxis=Y
    twistAxis=(X=1,Y=0,Z=0)
   )
   > &RigUnit_ModifyTransforms(
    ItemToModify=((Item=(Type=Null,Name="null_chain_B_04_l"),Transform=(Rotation=(X=0,Y=0,Z=0,W=1),Translation=(X=-1.3401,Y=-0.2938,Z=-2.5875),Scale3D=(X=1,Y=1,Z=1))))
    WeightMinimum=0
    Weight=@"DeltaRotation v4_10".outputWeight
   )
   > DeltaRotation v4_2 = RigVMFunctionReferenceNode(
    Rotation Order=ZYX
    child=(Type=Bone,Name="thigh_r")
    outputMinMax=(X=0,Y=1)
    parent=(Type=Null,Name="thigh_r_null")
    premult=1
    rotationMinMax=(X=0,Y=-90)
    triggerAxis=Y
    twistAxis=(X=1,Y=0,Z=0)
   )
   > &RigUnit_ModifyTransforms(
    ItemToModify=((Item=(Type=Null,Name="null_chain_B_04_l"),Transform=(Rotation=(X=0,Y=0,Z=0,W=1),Translation=(X=-1.3401,Y=-0.2938,Z=-2.5875),Scale3D=(X=1,Y=1,Z=1))))
    WeightMinimum=0
    Weight=@"DeltaRotation v4_2".outputWeight
   )
   RigVMFunction_MathBoolNot = &RigVMFunction_MathBoolNot(Value=RigVMVariableNode() {variable=crouching}.Value)
   Multiply_1 = RigVMAggregateNode(
    B=@"DeltaRotation v4_2".outputWeight
    C=@"DeltaRotation v4_5".outputWeight
    A=@"DeltaRotation v4_6".outputWeight
    D=&RigVMFunction_MathBoolToFloat(Value=@RigVMFunction_MathBoolNot.Result).Result
   )
   > &RigUnit_ModifyTransforms(
    ItemToModify="((Item=(Type=Null,Name=\"null_chain_B_04_l\"),Transform=(Rotation=(X=0,Y=0,Z=0,W=1),Translation=(X=0,Y=4,Z=-3),Scale3D=(X=1,Y=1,Z=1))),(Item=(Type=Null,Name=\"null_chain_B_03_r\"),Transform=(Rotation=(X=0,Y=0,Z=0,W=1),Translation=(X=0,Y=4,Z=1),Scale3D=(X=1,Y=1 ...(truncated)"
    WeightMaximum=0.4
    WeightMinimum=0
    Weight=@Multiply_1.Result
   )
   RigVMFunction_MathBoolToFloat = &RigVMFunction_MathBoolToFloat(Value=RigVMVariableNode() {variable=crouching}.Value)
   > &RigUnit_ModifyTransforms(
    ItemToModify="((Item=(Type=Null,Name=\"null_chain_B_04_l\"),Transform=(Rotation=(X=0,Y=0,Z=0,W=-1),Translation=(X=3.2313,Y=3.1675,Z=-0.4807),Scale3D=(X=1,Y=1,Z=1))),(Item=(Type=Null,Name=\"null_chain_B_03_r\"),Transform=(Rotation=(X=0,Y=0,Z=0,W=1),Translation=(X=0,Y=3,Z=1), ...(truncated)"
    WeightMinimum=0
    Weight=/Script/RigVM.RigVMFunction_AnimEasing(
     SourceMaximum=1
     SourceMinimum=0
     TargetMaximum=1
     TargetMinimum=0
     Type=CubicEaseInOut
     Value=&RigVMFunction_KalmanFloat(BufferSize=13, Value=@RigVMFunction_MathBoolToFloat.Result) {template="KalmanFilter::Execute(in Value,in BufferSize,out Result)"}.Result
    ) {is=struct RigVMUnitNode}.Result
   )
   > &RigUnit_ModifyTransforms(
    ItemToModify="((Item=(Type=Null,Name=\"null_chain_B_04_l\"),Transform=(Rotation=(X=0,Y=0,Z=0,W=1),Translation=(X=0,Y=4.5,Z=0.5),Scale3D=(X=1,Y=1,Z=1))),(Item=(Type=Null,Name=\"null_chain_B_03_r\"),Transform=(Rotation=(X=0,Y=0,Z=0,W=1),Translation=(X=-1,Y=5,Z=3),Scale3D=(X=1 ...(truncated)"
    WeightMinimum=0
    Weight=&RigVMFunction_MathBoolToFloat(Value=RigVMVariableNode() {variable=driving}.Value).Result
   )
   SplineFromItems_3 = RigVMFunctionReferenceNode(
    Items=((Type=Null,Name="null_chain_A_01_l"),(Type=Null,Name="null_chain_A_02_l"),(Type=Null,Name="null_chain_A_03_l"),(Type=Null,Name="null_chain_A_02_r"),(Type=Null,Name="null_chain_A_01_r"))
    Samples Per Segment=25
    Spline Mode=Hermite
    ExecuteContext=@RigVMFunction_Sequence.C
   )
   ItemArray_1 = RigVMDispatchNode(Value=((Type=Bone,Name="dyn_chain_A_01_l"),(Type=Bone,Name="dyn_chain_A_02_l"),(Type=Bone,Name="dyn_chain_A_03_l"))) {template="DISPATCH_RigVMDispatch_Constant(io Value)"}
   > /Script/ControlRigSpline.RigUnit_FitChainToSplineCurveItemArray(
    Alignment=Stretched
    DebugSettings=(bEnabled=False,Scale=1,CurveColor=(R=1,G=1,B=0,A=1),SegmentsColor=(R=1,G=0,B=0,A=1),WorldOffset=(Rotation=(X=0,Y=0,Z=0,W=1),Translation=(X=0,Y=0,Z=0),Scale3D=(X=1,Y=1,Z=1)))
    Maximum=0.49
    Minimum=0
    PoleVectorPosition=(X=0,Y=0,Z=0)
    PrimaryAxis=(X=1,Y=0,Z=0)
    RotationEaseType=Linear
    Rotations=()
    SamplingPrecision=25
    SecondaryAxis=(X=0,Y=0,Z=0)
    Weight=1
    bPropagateToChildren=true
    Items=@ItemArray_1.Value
    Spline=@SplineFromItems_3.Spline
   ) {is=struct RigVMUnitNode}
   ItemArray_4 = RigVMDispatchNode(Value=((Type=Bone,Name="dyn_chain_A_01_r"),(Type=Bone,Name="dyn_chain_A_02_r"),(Type=Bone,Name="dyn_chain_A_03_r"))) {template="DISPATCH_RigVMDispatch_Constant(io Value)"}
   > &RigUnit_FitChainToSplineCurveItemArray(
    Alignment=Stretched
    DebugSettings=(bEnabled=False,Scale=1,CurveColor=(R=1,G=1,B=0,A=1),SegmentsColor=(R=1,G=0,B=0,A=1),WorldOffset=(Rotation=(X=0,Y=0,Z=0,W=1),Translation=(X=0,Y=0,Z=0),Scale3D=(X=1,Y=1,Z=1)))
    Maximum=0.51
    Minimum=1
    PoleVectorPosition=(X=0,Y=0,Z=0)
    PrimaryAxis=(X=1,Y=0,Z=0)
    RotationEaseType=Linear
    Rotations=()
    SamplingPrecision=25
    SecondaryAxis=(X=0,Y=0,Z=0)
    Weight=1
    bPropagateToChildren=true
    Items=@ItemArray_4.Value
    Spline=@SplineFromItems_3.Spline
   )
   > SplineFromItems_4 = RigVMFunctionReferenceNode(
    Items="((Type=Null,Name=\"null_chain_B_01_l\"),(Type=Null,Name=\"null_chain_B_02_l\"),(Type=Null,Name=\"null_chain_B_03_l\"),(Type=Null,Name=\"null_chain_B_04_l\"),(Type=Null,Name=\"null_chain_B_03_r\"),(Type=Null,Name=\"null_chain_B_02_r\"),(Type=Null,Name=\"null_chain_B_01_ ...(truncated)"
    Samples Per Segment=25
    Spline Mode=Hermite
   )
   ItemArray_3 = RigVMDispatchNode(Value=((Type=Bone,Name="dyn_chain_B_01_l"),(Type=Bone,Name="dyn_chain_B_02_l"),(Type=Bone,Name="dyn_chain_B_03_l"),(Type=Bone,Name="dyn_chain_B_04_l"))) {template="DISPATCH_RigVMDispatch_Constant(io Value)"}
   > &RigUnit_FitChainToSplineCurveItemArray(
    Alignment=Stretched
    DebugSettings=(bEnabled=False,Scale=1,CurveColor=(R=1,G=1,B=0,A=1),SegmentsColor=(R=1,G=0,B=0,A=1),WorldOffset=(Rotation=(X=0,Y=0,Z=0,W=1),Translation=(X=0,Y=0,Z=0),Scale3D=(X=1,Y=1,Z=1)))
    Maximum=0.49
    Minimum=0
    PoleVectorPosition=(X=0,Y=0,Z=0)
    PrimaryAxis=(X=1,Y=0,Z=0)
    RotationEaseType=Linear
    Rotations=()
    SamplingPrecision=25
    SecondaryAxis=(X=0,Y=0,Z=0)
    Weight=1
    bPropagateToChildren=true
    Items=@ItemArray_3.Value
    Spline=@SplineFromItems_4.Spline
   )
   ItemArray_5 = RigVMDispatchNode(Value=((Type=Bone,Name="dyn_chain_B_01_r"),(Type=Bone,Name="dyn_chain_B_02_r"),(Type=Bone,Name="dyn_chain_B_03_r"),(Type=Bone,Name="dyn_chain_B_04_r"))) {template="DISPATCH_RigVMDispatch_Constant(io Value)"}
   > &RigUnit_FitChainToSplineCurveItemArray(
    Alignment=Stretched
    DebugSettings=(bEnabled=False,Scale=1,CurveColor=(R=1,G=1,B=0,A=1),SegmentsColor=(R=1,G=0,B=0,A=1),WorldOffset=(Rotation=(X=0,Y=0,Z=0,W=1),Translation=(X=0,Y=0,Z=0),Scale3D=(X=1,Y=1,Z=1)))
    Maximum=0.5
    Minimum=1
    PoleVectorPosition=(X=0,Y=0,Z=0)
    PrimaryAxis=(X=1,Y=0,Z=0)
    RotationEaseType=Linear
    Rotations=()
    SamplingPrecision=25
    SecondaryAxis=(X=0,Y=0,Z=0)
    Weight=1
    bPropagateToChildren=true
    Items=@ItemArray_5.Value
    Spline=@SplineFromItems_4.Spline
   )
   > &RigUnit_ParentConstraint(
    AdvancedSettings=(InterpolationType=Average,RotationOrderForFilter=XZY)
    Child=(Type=Bone,Name="dyn_chain_B_04_r")
    Parents=((Item=(Type=Bone,Name="dyn_chain_B_04_l"),Weight=1))
    Weight=1
   )
   &RigUnit_ParentConstraint(
    AdvancedSettings=(RotationOrderForFilter=XZY)
    Child=(Type=Bone,Name="MSTR_star_A_01")
    Parents="((Item=(Type=Bone,Name=\"dyn_chain_A_01_l\"),Weight=0.0013),(Item=(Type=Bone,Name=\"dyn_chain_A_02_l\"),Weight=0.0409),(Item=(Type=Bone,Name=\"dyn_chain_A_03_l\"),Weight=0.0525),(Item=(Type=Bone,Name=\"dyn_chain_A_01_r\"),Weight=0.1441),(Item=(Type=Bone,Name=\"MSTR ...(truncated)"
    Weight=1
    ExecutePin=@RigVMFunction_Sequence.D
   )
   > &RigUnit_ParentConstraint(
    AdvancedSettings=(RotationOrderForFilter=XZY)
    Child=(Type=Bone,Name="MSTR_star_A_02")
    Parents="((Item=(Type=Bone,Name=\"MSTR_chain_A_l\"),Weight=0.0134),(Item=(Type=Bone,Name=\"MSTR_chain_A_r\"),Weight=0.0226),(Item=(Type=Bone,Name=\"dyn_chain_A_01_r\"),Weight=0.0371),(Item=(Type=Bone,Name=\"dyn_chain_A_01_l\"),Weight=0.0616),(Item=(Type=Bone,Name=\"dyn_chai ...(truncated)"
    Weight=1
   )
   > &RigUnit_ParentConstraint(
    AdvancedSettings=(RotationOrderForFilter=XZY)
    Child=(Type=Bone,Name="MSTR_star_A_03")
    Parents="((Item=(Type=Bone,Name=\"dyn_chain_A_01_r\"),Weight=0.0014),(Item=(Type=Bone,Name=\"dyn_chain_A_02_r\"),Weight=0.0468),(Item=(Type=Bone,Name=\"dyn_chain_A_03_l\"),Weight=0.0671),(Item=(Type=Bone,Name=\"dyn_chain_A_02_l\"),Weight=0.1313),(Item=(Type=Bone,Name=\"dyn_ ...(truncated)"
    Weight=1
   )
   > &RigUnit_ParentConstraint(
    AdvancedSettings=(RotationOrderForFilter=XZY)
    Child=(Type=Bone,Name="MSTR_star_B_01")
    Parents="((Item=(Type=Bone,Name=\"dyn_chain_B_03_l\"),Weight=0.0016),(Item=(Type=Bone,Name=\"dyn_chain_B_04_l\"),Weight=0.0025),(Item=(Type=Bone,Name=\"dyn_chain_B_04_r\"),Weight=0.0079),(Item=(Type=Bone,Name=\"MSTR_chain_B_r\"),Weight=0.0802),(Item=(Type=Bone,Name=\"dyn_ch ...(truncated)"
    Weight=1
   )
   > &RigUnit_ParentConstraint(
    AdvancedSettings=(RotationOrderForFilter=XZY)
    Child=(Type=Bone,Name="MSTR_star_B_02")
    Filter=(TranslationFilter=(bX=True,bY=True,bZ=True),RotationFilter=(bX=false,bY=false,bZ=false),ScaleFilter=(bX=True,bY=True,bZ=True))
    Parents="((Item=(Type=Bone,Name=\"MSTR_chain_B_r\"),Weight=0.0028),(Item=(Type=Bone,Name=\"dyn_chain_B_02_l\"),Weight=0.004),(Item=(Type=Bone,Name=\"dyn_chain_B_04_l\"),Weight=0.046),(Item=(Type=Bone,Name=\"dyn_chain_B_03_l\"),Weight=0.0505),(Item=(Type=Bone,Name=\"dyn_chai ...(truncated)"
    Weight=1
   )
   > &RigUnit_ParentConstraint(
    AdvancedSettings=(RotationOrderForFilter=XZY)
    Child=(Type=Bone,Name="MSTR_star_B_03")
    Filter=(TranslationFilter=(bX=True,bY=True,bZ=True),RotationFilter=(bX=false,bY=false,bZ=false),ScaleFilter=(bX=True,bY=True,bZ=True))
    Parents="((Item=(Type=Bone,Name=\"dyn_chain_B_01_r\"),Weight=0.0054),(Item=(Type=Bone,Name=\"dyn_chain_B_02_l\"),Weight=0.0568),(Item=(Type=Bone,Name=\"dyn_chain_B_02_r\"),Weight=0.0658),(Item=(Type=Bone,Name=\"dyn_chain_B_04_r\"),Weight=0.1439),(Item=(Type=Bone,Name=\"dyn_ ...(truncated)"
    Weight=1
   )
   > &RigUnit_ParentConstraint(
    AdvancedSettings=(RotationOrderForFilter=XZY)
    Child=(Type=Bone,Name="MSTR_star_B_04")
    Filter=(TranslationFilter=(bX=True,bY=True,bZ=True),RotationFilter=(bX=false,bY=false,bZ=false),ScaleFilter=(bX=True,bY=True,bZ=True))
    Parents="((Item=(Type=Bone,Name=\"dyn_chain_B_01_r\"),Weight=0.001),(Item=(Type=Bone,Name=\"dyn_chain_B_02_r\"),Weight=0.0142),(Item=(Type=Bone,Name=\"dyn_chain_B_01_l\"),Weight=0.0175),(Item=(Type=Bone,Name=\"dyn_chain_B_04_r\"),Weight=0.0571),(Item=(Type=Bone,Name=\"dyn_c ...(truncated)"
    Weight=1
   )
   > &RigUnit_ParentConstraint(
    AdvancedSettings=(RotationOrderForFilter=XZY)
    Child=(Type=Bone,Name="MSTR_star_B_05")
    Parents="((Item=(Type=Bone,Name=\"dyn_chain_B_03_r\"),Weight=0.003),(Item=(Type=Bone,Name=\"dyn_chain_B_04_r\"),Weight=0.0035),(Item=(Type=Bone,Name=\"dyn_chain_B_04_l\"),Weight=0.007),(Item=(Type=Bone,Name=\"MSTR_chain_B_l\"),Weight=0.0132),(Item=(Type=Bone,Name=\"dyn_chai ...(truncated)"
    Weight=1
   )
   > &RigUnit_ParentConstraint(
    AdvancedSettings=(RotationOrderForFilter=XZY)
    Child=(Type=Bone,Name="MSTR_star_B_06")
    Parents=((Item=(Type=Bone,Name="dyn_chain_B_03_l"),Weight=0.0193),(Item=(Type=Bone,Name="MSTR_chain_B_l"),Weight=0.118),(Item=(Type=Bone,Name="dyn_chain_B_02_l"),Weight=0.3135),(Item=(Type=Bone,Name="dyn_chain_B_01_l"),Weight=0.5492))
    Weight=1
   )
   > &RigUnit_RotationConstraintLocalSpaceOffset(
    AdvancedSettings=(InterpolationType=Average,RotationOrderForFilter=XZY)
    Child=(Type=Bone,Name="MSTR_star_B_03")
    Filter=(bX=false,bY=false,bZ=true)
    Parents=((Item=(Type=Bone,Name="dyn_frontSkirt_01"),Weight=1))
    Weight=1
   )
   > &RigUnit_RotationConstraintLocalSpaceOffset(
    AdvancedSettings=(InterpolationType=Average,RotationOrderForFilter=XZY)
    Child=(Type=Bone,Name="MSTR_star_B_04")
    Filter=(bX=false,bY=false,bZ=true)
    Parents=((Item=(Type=Bone,Name="dyn_frontSkirt_01"),Weight=0.5),(Item=(Type=Bone,Name="dyn_frontSkirt_01_l"),Weight=0.5))
    Weight=1
   )
   > &RigUnit_RotationConstraintLocalSpaceOffset(
    AdvancedSettings=(InterpolationType=Average,RotationOrderForFilter=XZY)
    Child=(Type=Bone,Name="MSTR_star_B_02")
    Filter=(bX=false,bY=false,bZ=true)
    Parents=((Item=(Type=Bone,Name="dyn_frontSkirt_01"),Weight=0.5),(Item=(Type=Bone,Name="dyn_frontSkirt_01_r"),Weight=0.5))
    Weight=1
   )
   > &RigUnit_RotationConstraintLocalSpaceOffset(
    AdvancedSettings=(InterpolationType=Average,RotationOrderForFilter=XZY)
    Child=(Type=Bone,Name="MSTR_star_B_05")
    Filter=(bX=false,bY=false,bZ=true)
    Parents=((Item=(Type=Bone,Name="dyn_frontSkirt_01"),Weight=0.25),(Item=(Type=Bone,Name="dyn_frontSkirt_01_l"),Weight=0.75))
    Weight=1
   )
   > &RigUnit_RotationConstraintLocalSpaceOffset(
    AdvancedSettings=(InterpolationType=Average,RotationOrderForFilter=XZY)
    Child=(Type=Bone,Name="MSTR_star_A_02")
    Filter=(bX=false,bY=false,bZ=true)
    Parents=((Item=(Type=Bone,Name="dyn_frontSkirt_01"),Weight=1))
    Weight=0.5
   )
   RigVMFunctionReturnNode(ExecuteContext=@RigVMFunction_Sequence.E)
  ]}
 RigVMFunctionLibrary {nodes [
   Chains = RigVMCollapseNode(ExecuteContext=())
   DeltaRotation v4 = RigVMCollapseNode(
    ExecuteContext=()
    Rotation Order=ZYX
    child=(Type=Bone,Name="None")
    outputMinMax=(X=0,Y=1)
    parent=(Type=Bone,Name="None")
    premult=1
    rotationMinMax=(X=0,Y=0)
    twistAxis=(X=1,Y=0,Z=0)
   )
   SkirtSetup = RigVMCollapseNode(ExecuteContext=())
  ]}
 RigVMFunction_Sequence_2_ContainedGraph {nodes [
   n1 = RigVMFunctionEntryNode()
   n2 = &RigVMFunction_Sequence(ExecuteContext=@n1.ExecuteContext)
   n3 = &RigVMFunction_Sequence(ExecuteContext=@n2.B)
   n4 = &RigVMFunction_Sequence(ExecuteContext=@n3.B)
   n5 = &RigVMFunction_Sequence(ExecuteContext=@n4.B)
   RigVMFunctionReturnNode(A=@n2.A, B=@n3.A, C=@n4.A, D=@n5.A, E=@n5.B)
  ]}
 RigVMFunction_Sequence_2_ContainedGraph {nodes [
   n1 = RigVMFunctionEntryNode()
   RigVMFunction_Sequence = &RigVMFunction_Sequence(ExecuteContext=@n1.ExecuteContext)
   n2 = &RigVMFunction_Sequence(ExecuteContext=@RigVMFunction_Sequence.B)
   n3 = &RigVMFunction_Sequence(ExecuteContext=@n2.B)
   n4 = &RigVMFunction_Sequence(ExecuteContext=@n3.B)
   RigVMFunctionReturnNode(A=@RigVMFunction_Sequence.A, B=@n2.A, C=@n3.A, D=@n4.A, E=@n4.B)
  ]}
 RigVMModel {nodes [
   n1 = &RigUnit_PrepareForExecution()
   &RigUnit_HierarchyAddNull(
    Name=thigh_l_null
    Parent=(Type=Bone,Name="pelvis")
    Space=GlobalSpace
    Transform=&RigUnit_GetTransform(Item=(Type=Bone,Name="thigh_l"), bInitial=false).Transform
    ExecutePin=@n1.ExecutePin
   )
   > &RigUnit_HierarchyAddNull(
    Name=thigh_r_null
    Parent=(Type=Bone,Name="pelvis")
    Space=GlobalSpace
    Transform=&RigUnit_GetTransform(Item=(Type=Bone,Name="thigh_r"), bInitial=false).Transform
   )
   ItemArray = RigVMDispatchNode(Value="((Type=Bone,Name=\"dyn_chain_A_01_l\"),(Type=Bone,Name=\"dyn_chain_A_02_l\"),(Type=Bone,Name=\"dyn_chain_A_02_r\"),(Type=Bone,Name=\"dyn_chain_A_01_r\"),(Type=Bone,Name=\"dyn_chain_B_01_l\"),(Type=Bone,Name=\"dyn_chain_B_02_l\"),(Type=Bone,Name=\"dyn_chain_B_03_l\"),(Ty ...(truncated)") {template="DISPATCH_RigVMDispatch_Constant(io Value)"}
   > For_Each_1 = RigVMDispatchNode(Array=/Script/ControlRig.RigUnit_GetTransformItemArray(Space=GlobalSpace, bInitial=False, Items=@ItemArray.Value) {is=struct RigVMUnitNode}.Transforms) {template="DISPATCH_RigVMDispatch_ArrayIterator(in Array,out Element,out Index,out Count,out Ratio)"}
   > &RigUnit_HierarchyAddNull(
    Parent=(Type=Bone,Name="spine_02")
    Space=GlobalSpace
    Transform=@For_Each_1.Element
    Name=/Script/RigVM.RigVMFunction_NameReplace(
     New=null_
     Old=dyn_
     Name=RigVMDispatchNode(Index=@For_Each_1.Index, Array=@ItemArray.Value) {template="DISPATCH_RigVMDispatch_ArrayGetAtIndex(in Array,in Index,out Element)"}.Element.Name
    ) {is=struct RigVMUnitNode, template="Replace::Execute(in Name,in Old,in New,out Result)"}.Result
   )
   RigUnit_BeginExecution = &RigUnit_BeginExecution()
   FN_DeformRig_FNC = RigVMFunctionReferenceNode(
    DeformRig_ Curve="(deformRig_l_upperarm_up_alpha_72_#1=False,deformRig_r_upperarm_up_alpha_73_#2=False,deformRig_l_upperarm_fr_alpha_74_#3=False,deformRig_r_upperarm_fr_alpha_75_CB51B8 ...(truncated)"
    DeformRig_Bool="(upperarm_1_#4=True,elbow_3_#5=True,wrist_7_#6=True,groin_8_#7=True,glute_17_#8=True,knee_9_40C408324CA261 ...(truncated)"
    DeformRig_Debug=(upperarm_debug_18_#4=False,wrist_debug_20_#6=False,lat_debug_26_#9=False,groinGlute_debug_27_#7=False)
    DeformRig_Value="(upperarm_up_18_#10=(Rotation=(X=0,Y=0,Z=0,W=1),Translation=(X=0,Y=0,Z=5),Scale3D=(X=0.7,Y=1,Z=1)),upperarm_fwd_fr_19_#11=(Rotation=(X=0,Y=0,Z=0,W=1),Translation=(X=0,Y=10,Z=0),Scale3D=(X=0.5,Y=1,Z= ...(truncated)"
    ExecuteContext=@RigUnit_BeginExecution.ExecutePin
   )
   > &RigUnit_RotationConstraintLocalSpaceOffset(
    AdvancedSettings=(InterpolationType=Shortest,RotationOrderForFilter=XZY)
    Child=(Type=Bone,Name="foot_newpivot_l")
    Filter=(bX=True,bY=True,bZ=True)
    Parents=((Item=(Type=Bone,Name="foot_l"),Weight=0.5),(Item=(Type=Bone,Name="calf_twist_02_l"),Weight=0.5))
    Weight=1
   )
   > &RigUnit_RotationConstraintLocalSpaceOffset(
    AdvancedSettings=(InterpolationType=Shortest,RotationOrderForFilter=XZY)
    Child=(Type=Bone,Name="foot_newpivot_r")
    Filter=(bX=True,bY=True,bZ=True)
    Parents=((Item=(Type=Bone,Name="foot_r"),Weight=0.5),(Item=(Type=Bone,Name="calf_twist_02_r"),Weight=0.5))
    Weight=1
   )
   > SkirtSetup = RigVMFunctionReferenceNode()
   > &RigUnit_ParentConstraint(
    AdvancedSettings=(RotationOrderForFilter=XZY)
    Child=(Type=Bone,Name="MSTR_hipRibbon")
    Parents=((Item=(Type=Bone,Name="spine_01"),Weight=0.0521),(Item=(Type=Bone,Name="spine_03"),Weight=0.1309),(Item=(Type=Bone,Name="spine_04"),Weight=0.1765),(Item=(Type=Bone,Name="pelvis"),Weight=0.18),(Item=(Type=Bone,Name="spine_02"),Weight=0.4605))
    Weight=1
   )
   > &RigUnit_ParentConstraint(
    AdvancedSettings=(RotationOrderForFilter=XZY)
    Child=(Type=Bone,Name="MSTR_backRibbon")
    Parents="((Item=(Type=Bone,Name=\"spine_05\"),Weight=0.006),(Item=(Type=Bone,Name=\"clavicle_r\"),Weight=0.0124),(Item=(Type=Bone,Name=\"clavicle_l\"),Weight=0.0125),(Item=(Type=Bone,Name=\"spine_01\"),Weight=0.0939),(Item=(Type=Bone,Name=\"pelvis\"),Weight=0.1156),(Item=(Ty ...(truncated)"
    Weight=1
   )
   > Chains = RigVMFunctionReferenceNode()
   > &RigUnit_ParentConstraint(
    AdvancedSettings=(RotationOrderForFilter=XZY)
    Child=(Type=Bone,Name="MSTR_pendant")
    Parents=((Item=(Type=Bone,Name="spine_05"),Weight=0.416),(Item=(Type=Bone,Name="neck_01"),Weight=0.308),(Item=(Type=Bone,Name="neck_02"),Weight=0.275))
    Weight=1
   )
   RigVMFunction_MathBoolToFloat = &RigVMFunction_MathBoolToFloat(Value=RigVMVariableNode() {variable=crouching}.Value)
   > n2 = &RigUnit_ModifyTransforms(
    ItemToModify="((Item=(Type=Bone,Name=\"MSTR_backRibbon\"),Transform=(Rotation=(X=0,Y=0,Z=-0.1736,W=0.9848),Translation=(X=0,Y=0,Z=0),Scale3D=(X=1,Y=1,Z=1))),(Item=(Type=Bone,Name=\"dyn_frontSkirt_01\"),Transform=(Rotation=(X=0.1074,Y=0.221,Z=-0.0245,W=0.969),Translation=(X= ...(truncated)"
    WeightMinimum=0
    Weight=&RigVMFunction_AnimEasing(
     SourceMaximum=1
     SourceMinimum=0
     TargetMaximum=1
     TargetMinimum=0
     Type=CubicEaseInOut
     Value=&RigVMFunction_KalmanFloat(BufferSize=13, Value=@RigVMFunction_MathBoolToFloat.Result) {template="KalmanFilter::Execute(in Value,in BufferSize,out Result)"}.Result
    ).Result
   )
   Less = /Script/RigVM.RigVMFunction_MathDoubleLess(B=1, A=RigVMVariableNode() {variable=rban alpha}.Value) {is=struct RigVMUnitNode, template="Less::Execute(in A,in B,out Result)"}
   RigVMFunction_MathBoolNot = &RigVMFunction_MathBoolNot(Value=RigVMVariableNode() {variable=rban enabled}.Value)
   RigVMFunction_MathBoolOr = &RigVMFunction_MathBoolOr(B=@Less.Result, A=@RigVMFunction_MathBoolNot.Result)
   RigVMFunction_ControlFlowBranch = &RigVMFunction_ControlFlowBranch(ExecuteContext=@n2.ExecutePin, Condition=@RigVMFunction_MathBoolOr.Result)
   n3 = &RigVMFunction_MathBoolToFloat(Value=RigVMVariableNode() {variable=sliding}.Value)
   &RigUnit_ModifyTransforms(
    ItemToModify="((Item=(Type=Bone,Name=\"dyn_backSash_03_l\"),Transform=(Rotation=(X=0,Y=-0.1736,Z=0,W=0.9848),Translation=(X=0,Y=0,Z=0),Scale3D=(X=1,Y=1,Z=1))),(Item=(Type=Bone,Name=\"dyn_backSash_03_r\"),Transform=(Rotation=(X=0,Y=-0.1736,Z=0,W=0.9848),Translation=(X=0,Y=0, ...(truncated)"
    WeightMinimum=0
    Weight=&RigVMFunction_AnimEasing(
     SourceMaximum=1
     SourceMinimum=0
     TargetMaximum=1
     TargetMinimum=0
     Type=CubicEaseInOut
     Value=&RigVMFunction_KalmanFloat(BufferSize=13, Value=@n3.Result) {template="KalmanFilter::Execute(in Value,in BufferSize,out Result)"}.Result
    ).Result
    ExecutePin=@RigVMFunction_ControlFlowBranch.True
   )
   n4 = &RigVMFunction_MathBoolToFloat(Value=RigVMVariableNode() {variable=crouching}.Value)
   > &RigUnit_ModifyTransforms(
    ItemToModify="((Item=(Type=Bone,Name=\"dyn_backSash_04_l\"),Transform=(Rotation=(X=0,Y=-0.5,Z=0,W=0.866),Translation=(X=0,Y=0,Z=0),Scale3D=(X=1,Y=1,Z=1))),(Item=(Type=Bone,Name=\"dyn_backSash_04_r\"),Transform=(Rotation=(X=0,Y=-0.5,Z=0,W=0.866),Translation=(X=0,Y=0,Z=0),Sca ...(truncated)"
    WeightMinimum=0
    Weight=&RigVMFunction_AnimEasing(
     SourceMaximum=1
     SourceMinimum=0
     TargetMaximum=1
     TargetMinimum=0
     Type=CubicEaseInOut
     Value=&RigVMFunction_KalmanFloat(BufferSize=13, Value=@n4.Result) {template="KalmanFilter::Execute(in Value,in BufferSize,out Result)"}.Result
    ).Result
   )
  ]}
]
hierarchy [
 root {loc=<0 0 0>, rot=<0 0 0 1>}
 pelvis {parent=root, loc=<0 -0.5301 98.0571>, rot=<0.0127 0.707 -0.0127 -0.707>}
 spine_01 {parent=pelvis, loc=<6.1208 -0.9328 0>, rot=<0 0 0.034 -0.9994>}
 spine_02 {parent=spine_01, loc=<6.5997 -0.2095 0>, rot=<0 0 0.0238 -0.9997>}
 spine_03 {parent=spine_02, loc=<6.4914 -0.1237 0>, rot=<0 0 0.0149 -0.9999>}
 spine_04 {parent=spine_03, loc=<6.9598 0.0794 0>, rot=<0 0 -0.0132 -0.9999>}
 spine_05 {parent=spine_04, loc=<6.9987 0 0>, rot=<0 0 -0.0015 -1>}
 neck_01 {parent=spine_05, loc=<11.3138 0.1529 0>, rot=<0 0 -0.2003 -0.9797>}
 neck_02 {parent=neck_01, loc=<4.5627 0 0>, rot=<0 0 0.0504 -0.9987>}
 head {parent=neck_02, loc=<5.3914 0 0>, rot=<0 0 0.115 -0.9934>}
 MSTR_pendant {parent=neck_01, loc=<0.5469 5.2283 -0.0011>, rot=<0 0 0.1142 -0.9935>}
 dyn_pendant {parent=MSTR_pendant, loc=<0 0 0>, rot=<0 0 0 1>}
 clavicle_l {parent=spine_05, loc=<5.4831 -0.1053 -3.2946>, rot=<0.0465 -0.7289 -0.009 -0.6829>}
 upperarm_l {parent=clavicle_l, loc=<9.8658 -1.2136 -0.4105>, rot=<0.0164 -0.3245 -0.0269 -0.9454>}
 lowerarm_l {parent=upperarm_l, loc=<26.465 -0.8345 0.259>, rot=<0.0036 0.0161 -0.2298 -0.9731>}
 hand_l {parent=lowerarm_l, loc=<22.7355 -0.3203 -0.6105>, rot=<-0.6784 -0.0223 0.0735 -0.7307>}
 thumb_01_l {parent=hand_l, loc=<1.3118 -0.3191 -2.459>, rot=<0.6546 -0.4042 0.0664 -0.6354>}
 thumb_02_l {parent=thumb_01_l, loc=<5.9336 -0.0448 0.0401>, rot=<-0.0037 0.0213 0.298 -0.9543>}
 thumb_03_l {parent=thumb_02_l, loc=<2.2223 0.3682 -0.0907>, rot=<0.0118 -0.013 0.0066 -0.9998>}
 index_metacarpal_l {parent=hand_l, loc=<1.6535 0.0183 -2.96>, rot=<0.0158 -0.0124 -0.0173 -0.9996>}
 index_01_l {parent=index_metacarpal_l, loc=<8.3008 0.0357 -0.0524>, rot=<-0.0044 0.0485 0.0924 -0.9945>}
 index_02_l {parent=index_01_l, loc=<3.3587 -0.3004 0.0452>, rot=<-0.002 0.0027 0.3225 -0.9466>}
 index_03_l {parent=index_02_l, loc=<2.5578 -0.123 -0.0252>, rot=<-0.002 0.0067 0.2558 -0.9667>}
 middle_metacarpal_l {parent=hand_l, loc=<2.1856 -0.1725 -0.5984>, rot=<0.006 0.0168 -0.0329 -0.9993>}
 middle_01_l {parent=middle_metacarpal_l, loc=<7.3275 -0.2447 0.0129>, rot=<-0.0208 0.0042 0.0858 -0.9961>}
 middle_02_l {parent=middle_01_l, loc=<4.0123 0.1401 0.1175>, rot=<-0.0218 0.0258 0.3512 -0.9357>}
 middle_03_l {parent=middle_02_l, loc=<2.6077 0.0534 0.1373>, rot=<0.0082 0.0188 0.1708 -0.9851>}
 ring_metacarpal_l {parent=hand_l, loc=<2.2074 -0.3625 1.3638>, rot=<-0.0284 0.0722 -0.0458 -0.9959>}
 ring_01_l {parent=ring_metacarpal_l, loc=<7.0295 -0.6197 -0.1446>, rot=<-0.0489 -0.0252 0.1892 -0.9804>}
 ring_02_l {parent=ring_01_l, loc=<3.6174 0.4647 -0.1027>, rot=<0 0 0.2806 -0.9598>}
 ring_03_l {parent=ring_02_l, loc=<2.4152 0.114 0.0199>, rot=<0 0 0.1189 -0.9929>}
 pinky_metacarpal_l {parent=hand_l, loc=<1.7616 -0.9753 3.3814>, rot=<-0.0397 0.0814 -0.069 -0.9935>}
 pinky_01_l {parent=pinky_metacarpal_l, loc=<7.1709 -0.4831 -0.1094>, rot=<-0.0654 -0.0035 0.2242 -0.9723>}
 pinky_02_l {parent=pinky_01_l, loc=<2.8274 0.0893 -0.1043>, rot=<0.0037 0.0011 0.2854 -0.9584>}
 pinky_03_l {parent=pinky_02_l, loc=<2.2046 0.0307 -0.0078>, rot=<-0.0002 0.0098 0.1445 -0.9895>}
 deform_wrist_dn_base_l {parent=hand_l, loc=<0 -1.5326 0>, rot=<0 0 0 -1>}
 deform_wrist_dn_l {parent=deform_wrist_dn_base_l, loc=<0 0 0>, rot=<0 0 0 1>}
 deform_wrist_up_base_l {parent=hand_l, loc=<0 2.3009 0>, rot=<0 0 0 -1>}
 deform_wrist_up_l {parent=deform_wrist_up_base_l, loc=<0 0 0>, rot=<0 0 0 1>}
 lowerarm_twist_02_l {parent=lowerarm_l, loc=<10.6707 0 0>, rot=<0 0 0 1>}
 lowerarm_twist_01_l {parent=lowerarm_l, loc=<18.316 0 0>, rot=<0 0 0 1>}
 deform_elbow_in_base_l {parent=lowerarm_l, loc=<0.1834 3.6062 -0.2572>, rot=<0 0 0 -1>}
 deform_elbow_in_l {parent=deform_elbow_in_base_l, loc=<0 0 0>, rot=<0 0 0 1>}
 deform_elbow_out_base_l {parent=lowerarm_l, loc=<0 -4 0>, rot=<0 0 0 1>}
 deform_elbow_out_l {parent=deform_elbow_out_base_l, loc=<0 0 0>, rot=<0 0 0 1>}
 upperarm_twist_01_l {parent=upperarm_l, loc=<7.2135 -0.1134 0>, rot=<0 0 0 1>}
 deform_upperarm_bk_base_l {parent=upperarm_twist_01_l, loc=<-6.4053 -4.3746 0.314>, rot=<0 0 0 -1>}
 deform_upperarm_bk_l {parent=deform_upperarm_bk_base_l, loc=<0 0 0>, rot=<0 0 0 1>}
 deform_upperarm_fr_base_l {parent=upperarm_twist_01_l, loc=<-5.6312 5.092 0.3354>, rot=<0 0 0 -1>}
 deform_upperarm_fr_l {parent=deform_upperarm_fr_base_l, loc=<0 0 0>, rot=<0 0 0 1>}
 deform_upperarm_up_base_l {parent=upperarm_twist_01_l, loc=<-8.891 0.4825 3.9517>, rot=<0 0 0 -1>}
 deform_upperarm_up_l {parent=deform_upperarm_up_base_l, loc=<0 0 0>, rot=<0 0 0 1>}
 upperarm_twist_02_l {parent=upperarm_l, loc=<19.7998 -0.1134 0>, rot=<0 0 0 1>}
 clavicle_r {parent=spine_05, loc=<5.4831 -0.1053 3.2946>, rot=<0.7289 0.0465 -0.6829 0.009>}
 upperarm_r {parent=clavicle_r, loc=<-9.8658 1.2136 0.4105>, rot=<0.0164 -0.3245 -0.0269 -0.9454>}
 lowerarm_r {parent=upperarm_r, loc=<-26.465 0.8345 -0.259>, rot=<0.0036 0.0161 -0.2298 -0.9731>}
 hand_r {parent=lowerarm_r, loc=<-22.7355 0.3203 0.6106>, rot=<-0.6784 -0.0223 0.0735 -0.7307>}
 thumb_01_r {parent=hand_r, loc=<-1.3118 0.3191 2.459>, rot=<0.6546 -0.4042 0.0664 -0.6354>}
 thumb_02_r {parent=thumb_01_r, loc=<-5.9336 0.0448 -0.0401>, rot=<-0.0037 0.0213 0.298 -0.9543>}
 thumb_03_r {parent=thumb_02_r, loc=<-2.2223 -0.3682 0.0907>, rot=<0.0118 -0.013 0.0066 -0.9998>}
 index_metacarpal_r {parent=hand_r, loc=<-1.6535 -0.0183 2.96>, rot=<0.0158 -0.0124 -0.0173 -0.9996>}
 index_01_r {parent=index_metacarpal_r, loc=<-8.3008 -0.0357 0.0524>, rot=<-0.0044 0.0485 0.0924 -0.9945>}
 index_02_r {parent=index_01_r, loc=<-3.3587 0.3004 -0.0452>, rot=<-0.002 0.0027 0.3225 -0.9466>}
 index_03_r {parent=index_02_r, loc=<-2.5578 0.123 0.0252>, rot=<-0.002 0.0067 0.2558 -0.9667>}
 middle_metacarpal_r {parent=hand_r, loc=<-2.1856 0.1725 0.5984>, rot=<0.006 0.0168 -0.0329 -0.9993>}
 middle_01_r {parent=middle_metacarpal_r, loc=<-7.3275 0.2447 -0.0129>, rot=<-0.0208 0.0042 0.0858 -0.9961>}
 middle_02_r {parent=middle_01_r, loc=<-4.0123 -0.1401 -0.1175>, rot=<-0.0218 0.0258 0.3512 -0.9357>}
 middle_03_r {parent=middle_02_r, loc=<-2.6077 -0.0534 -0.1373>, rot=<0.0082 0.0188 0.1708 -0.9851>}
 ring_metacarpal_r {parent=hand_r, loc=<-2.2074 0.3625 -1.3638>, rot=<-0.0284 0.0722 -0.0458 -0.9959>}
 ring_01_r {parent=ring_metacarpal_r, loc=<-7.0295 0.6197 0.1446>, rot=<-0.0489 -0.0252 0.1892 -0.9804>}
 ring_02_r {parent=ring_01_r, loc=<-3.6174 -0.4647 0.1027>, rot=<0 0 0.2806 -0.9598>}
 ring_03_r {parent=ring_02_r, loc=<-2.4152 -0.114 -0.0199>, rot=<0 0 0.1189 -0.9929>}
 pinky_metacarpal_r {parent=hand_r, loc=<-1.7616 0.9753 -3.3814>, rot=<-0.0397 0.0814 -0.069 -0.9935>}
 pinky_01_r {parent=pinky_metacarpal_r, loc=<-7.1709 0.4831 0.1094>, rot=<-0.0654 -0.0035 0.2242 -0.9723>}
 pinky_02_r {parent=pinky_01_r, loc=<-2.8274 -0.0893 0.1043>, rot=<0.0037 0.0011 0.2854 -0.9584>}
 pinky_03_r {parent=pinky_02_r, loc=<-2.2046 -0.0307 0.0078>, rot=<-0.0002 0.0098 0.1445 -0.9895>}
 deform_wrist_dn_base_r {parent=hand_r, loc=<-0.0001 1.5328 0>, rot=<0 0 0 -1>}
 deform_wrist_dn_r {parent=deform_wrist_dn_base_r, loc=<0 0 0>, rot=<0 0 0 1>}
 deform_wrist_up_base_r {parent=hand_r, loc=<-0.0002 -2.3007 -0.0001>, rot=<0 0 0 -1>}
 deform_wrist_up_r {parent=deform_wrist_up_base_r, loc=<0 0 0>, rot=<0 0 0 1>}
 lowerarm_twist_02_r {parent=lowerarm_r, loc=<-10.6707 0 0>, rot=<0 0 0 -1>}
 lowerarm_twist_01_r {parent=lowerarm_r, loc=<-18.316 0 0>, rot=<0 0 0 -1>}
 deform_elbow_in_base_r {parent=lowerarm_r, loc=<-0.1834 -3.6062 0.2572>, rot=<0 0 0 -1>}
 deform_elbow_in_r {parent=deform_elbow_in_base_r, loc=<0 0 0>, rot=<0 0 0 1>}
 deform_elbow_out_base_r {parent=lowerarm_r, loc=<0 4 -0.0001>, rot=<0 0 0 -1>}
 deform_elbow_out_r {parent=deform_elbow_out_base_r, loc=<0 0 0>, rot=<0 0 0 1>}
 upperarm_twist_01_r {parent=upperarm_r, loc=<-7.2135 0.1134 0>, rot=<0 0 0 -1>}
 deform_upperarm_bk_base_r {parent=upperarm_twist_01_r, loc=<6.4053 4.3746 -0.314>, rot=<0 0 0 -1>}
 deform_upperarm_bk_r {parent=deform_upperarm_bk_base_r, loc=<0 0 0>, rot=<0 0 0 1>}
 deform_upperarm_fr_base_r {parent=upperarm_twist_01_r, loc=<5.6312 -5.092 -0.3354>, rot=<0 0 0 -1>}
 deform_upperarm_fr_r {parent=deform_upperarm_fr_base_r, loc=<0 0 0>, rot=<0 0 0 1>}
 deform_upperarm_up_base_r {parent=upperarm_twist_01_r, loc=<8.891 -0.4825 -3.9517>, rot=<0 0 0 -1>}
 deform_upperarm_up_r {parent=deform_upperarm_up_base_r, loc=<0 0 0>, rot=<0 0 0 1>}
 upperarm_twist_02_r {parent=upperarm_r, loc=<-19.7998 0.1134 0>, rot=<0 0 0 -1>}
 attach_backpack {parent=spine_05, loc=<-0.123 -13.473 0>, rot=<-0.0093 -0.707 0.0093 -0.707>}
 attach_cape {parent=spine_05, loc=<9.017 -11.714 0>, rot=<0.0616 -0.7044 -0.0616 -0.7044>}
 MSTR_sash {parent=spine_05, loc=<1.5541 8.9643 -0.0974>, rot=<-0.2302 -0.6762 -0.6677 -0.2099>}
 dyn_sash_01 {parent=MSTR_sash, loc=<0 0 0>, rot=<0 0 0 1>}
 dyn_sash_02 {parent=dyn_sash_01, loc=<4.0328 0 0>, rot=<0.0034 -0.1704 0.0067 -0.9853>}
 deform_pec_aim_base_l {parent=spine_04, loc=<3.2456 9.3309 -1.2237>, rot=<0.0603 -0.4033 0.3318 -0.8507>}
 deform_pec_aim_l {parent=deform_pec_aim_base_l, loc=<0 0 0>, rot=<0 0 0 1>}
 deform_pec_base_l {parent=deform_pec_aim_l, loc=<9.3665 0 0>, rot=<0 0 0 1>}
 deform_pec_l {parent=deform_pec_base_l, loc=<0 0 0>, rot=<0 0 0 1>}
 deform_pec_aim_base_r {parent=spine_04, loc=<3.2452 9.3309 1.2237>, rot=<-0.4033 -0.0603 0.8507 0.3318>}
 deform_pec_aim_r {parent=deform_pec_aim_base_r, loc=<0 0 0>, rot=<0 0 0 1>}
 deform_pec_base_r {parent=deform_pec_aim_r, loc=<-9.3667 -0.0003 -0.0004>, rot=<0 0 0 1>}
 deform_pec_r {parent=deform_pec_base_r, loc=<0 0 0>, rot=<0 0 0 1>}
 MSTR_hipRibbon {parent=spine_02, loc=<2.7138 5.9584 7.4616>, rot=<0.3955 -0.1925 0.0458 -0.8969>}
 dyn_hipRibbon_01_B_r {parent=MSTR_hipRibbon, loc=<-1.7532 -2.1018 0.0769>, rot=<-0.1829 0.6364 0.7172 -0.2172>}
 dyn_hipRibbon_02_B_r {parent=dyn_hipRibbon_01_B_r, loc=<4.8683 0 0>, rot=<0 0 0 1>}
 dyn_hipRibbon_01_B_l {parent=MSTR_hipRibbon, loc=<-2.943 2.2764 0.0372>, rot=<-0.1141 -0.7065 -0.6905 -0.1048>}
 dyn_hipRibbon_02_B_l {parent=dyn_hipRibbon_01_B_l, loc=<4.5202 0 0>, rot=<0 0 0 1>}
 dyn_hipRibbon_01_A_l {parent=MSTR_hipRibbon, loc=<-0.1976 1.4766 0.1181>, rot=<0.5212 0.4446 0.6378 0.352>}
 dyn_hipRibbon_02_A_l {parent=dyn_hipRibbon_01_A_l, loc=<4.3975 0 0>, rot=<0 0 0 1>}
 dyn_hipRibbon_01_A_r {parent=MSTR_hipRibbon, loc=<0.1147 -1.4847 -0.0101>, rot=<0.6403 -0.3786 0.5489 -0.3813>}
 dyn_hipRibbon_02_A_r {parent=dyn_hipRibbon_01_A_r, loc=<4.2914 0 0>, rot=<0 0 0 1>}
 MSTR_backRibbon {parent=spine_02, loc=<4.9252 -6.5837 0.0105>, rot=<0 0 0 1>}
 MSTR_backRibbon_B_l {parent=MSTR_backRibbon, loc=<0.1928 -0.9759 -0.3899>, rot=<0.4832 -0.8105 0.1463 -0.297>}
 dyn_backRibbon_01_B_l {parent=MSTR_backRibbon_B_l, loc=<0 0 0>, rot=<0 0 0 1>}
 dyn_backRibbon_02_B_l {parent=dyn_backRibbon_01_B_l, loc=<6.7766 0 0>, rot=<0.0847 -0.0529 0.0169 -0.9949>}
 dyn_backRibbon_03_B_l {parent=dyn_backRibbon_02_B_l, loc=<6.9322 0 0>, rot=<0.1029 -0.0838 -0.14 -0.9812>}
 dyn_backRibbon_04_B_l {parent=dyn_backRibbon_03_B_l, loc=<7.0909 0 0>, rot=<0 0 0 1>}
 MSTR_backRibbon_B_r {parent=MSTR_backRibbon, loc=<0.1926 -0.9758 0.3689>, rot=<0.8105 0.4832 -0.297 -0.1463>}
 dyn_backRibbon_01_B_r {parent=MSTR_backRibbon_B_r, loc=<0 0 0>, rot=<0 0 0 1>}
 dyn_backRibbon_02_B_r {parent=dyn_backRibbon_01_B_r, loc=<-6.7765 0.0004 -0.0004>, rot=<0.0847 -0.0529 0.0169 -0.9949>}
 dyn_backRibbon_03_B_r {parent=dyn_backRibbon_02_B_r, loc=<-6.9323 -0.0002 0.0001>, rot=<0.1029 -0.0838 -0.14 -0.9812>}
 dyn_backRibbon_04_B_r {parent=dyn_backRibbon_03_B_r, loc=<-7.091 -0.0001 0>, rot=<0 0 0 1>}
 MSTR_backRibbon_A_l {parent=MSTR_backRibbon, loc=<-0.155 -1.7397 -1.3197>, rot=<0.4298 -0.7323 -0.0529 -0.5256>}
 dyn_backRibbon_A_01_l {parent=MSTR_backRibbon_A_l, loc=<0 0 0>, rot=<0 0 0 1>}
 dyn_backRibbon_A_02_l {parent=dyn_backRibbon_A_01_l, loc=<5.2696 0 0>, rot=<0.0395 -0.0136 -0.2373 -0.9705>}
 dyn_backRibbon_A_03_l {parent=dyn_backRibbon_A_02_l, loc=<7.1105 0 0>, rot=<0 0 0 1>}
 MSTR_backRibbon_A_r {parent=MSTR_backRibbon, loc=<-0.1547 -1.7397 1.3197>, rot=<0.7323 0.4298 -0.5256 0.0529>}
 dyn_backRibbon_A_01_r {parent=MSTR_backRibbon_A_r, loc=<0 0 0>, rot=<0 0 0 1>}
 dyn_backRibbon_A_02_r {parent=dyn_backRibbon_A_01_r, loc=<-5.2696 -0.0005 0.0005>, rot=<0.0395 -0.0136 -0.2373 -0.9705>}
 dyn_backRibbon_A_03_r {parent=dyn_backRibbon_A_02_r, loc=<-7.1103 0.0003 -0.0004>, rot=<0 0 0 1>}
 MSTR_backSash_l {parent=MSTR_backRibbon, loc=<-2.2648 -2.5267 -2.1463>, rot=<0.1275 -0.6572 0.6893 -0.2769>}
 dyn_backSash_01_l {parent=MSTR_backSash_l, loc=<0 0 0>, rot=<0 0 0 1>}
 dyn_backSash_02_l {parent=dyn_backSash_01_l, loc=<20.8777 0 0>, rot=<-0.001 -0.0498 -0.0033 -0.9988>}
 dyn_backSash_03_l {parent=dyn_backSash_02_l, loc=<19.4363 0 0>, rot=<0.0024 -0.0866 0.0249 -0.9959>}
 dyn_backSash_04_l {parent=dyn_backSash_03_l, loc=<20.4104 0 0>, rot=<0 0 0 1>}
 MSTR_backSash_r {parent=MSTR_backRibbon, loc=<-2.2647 -2.5267 2.1463>, rot=<0.6572 0.1275 -0.2769 -0.6893>}
 dyn_backSash_01_r {parent=MSTR_backSash_r, loc=<0 0 0>, rot=<0 0 0 1>}
 dyn_backSash_02_r {parent=dyn_backSash_01_r, loc=<-20.8779 0 0.0001>, rot=<-0.001 -0.0498 -0.0033 -0.9988>}
 dyn_backSash_03_r {parent=dyn_backSash_02_r, loc=<-19.4363 0 0>, rot=<0.0024 -0.0866 0.0249 -0.9959>}
 dyn_backSash_04_r {parent=dyn_backSash_03_r, loc=<-20.4103 -0.0001 0>, rot=<0 0 0 1>}
 MSTR_chains {parent=spine_02, loc=<0.6431 4.6171 -0.07>, rot=<0 0 0 1>}
 MSTR_chain_A_l {parent=MSTR_chains, loc=<2.6565 2.7571 -5.6068>, rot=<-0.2473 -0.9508 0.0503 0.1798>}
 dyn_chain_A_01_l {parent=MSTR_chain_A_l, loc=<0 0 0>, rot=<0 0 0 1>}
 dyn_chain_A_02_l {parent=dyn_chain_A_01_l, loc=<7.8404 0 0>, rot=<0.0442 -0.3752 0.0822 -0.9222>}
 dyn_chain_A_03_l {parent=dyn_chain_A_02_l, loc=<7.808 0 0>, rot=<0 0 0 1>}
 MSTR_star_A_02 {parent=dyn_chain_A_03_l, loc=<0.1862 0.0733 0.3147>, rot=<0.5743 0.2735 0.5068 -0.5818>}
 dyn_star_A_02 {parent=MSTR_star_A_02, loc=<0 0 0>, rot=<0 0 0 1>}
 MSTR_star_A_03 {parent=dyn_chain_A_01_l, loc=<6.3906 0.0155 0.751>, rot=<0.5277 0.345 0.5329 -0.5644>}
 dyn_star_A_03 {parent=MSTR_star_A_03, loc=<0 0 0>, rot=<0 0 0 1>}
 MSTR_chain_B_l {parent=MSTR_chains, loc=<3.1921 -0.176 -7.7281>, rot=<-0.3063 -0.9518 0.0064 0.0156>}
 dyn_chain_B_01_l {parent=MSTR_chain_B_l, loc=<0 0 0>, rot=<0 0 0 1>}
 dyn_chain_B_02_l {parent=dyn_chain_B_01_l, loc=<7.1431 0 0>, rot=<-0.1207 -0.218 0.1376 -0.9586>}
 dyn_chain_B_03_l {parent=dyn_chain_B_02_l, loc=<8.8217 0 0>, rot=<-0.0787 -0.3352 0.1979 -0.9178>}
 dyn_chain_B_04_l {parent=dyn_chain_B_03_l, loc=<9.323 0 0>, rot=<0 0 0 1>}
 MSTR_star_B_03 {parent=dyn_chain_B_04_l, loc=<-0.0689 0.1689 0.4811>, rot=<0.7119 0.3899 0.3154 -0.4917>}
 dyn_star_B_03 {parent=MSTR_star_B_03, loc=<0 0 0>, rot=<0 0 0 1>}
 MSTR_star_B_04 {parent=dyn_chain_B_03_l, loc=<3.7348 0.3656 1.2428>, rot=<-0.4507 -0.5695 -0.4965 0.4754>}
 dyn_star_B_04 {parent=MSTR_star_B_04, loc=<0 0 0>, rot=<0 0 0 1>}
 MSTR_star_B_05 {parent=dyn_chain_B_02_l, loc=<5.3431 0.0543 0.8993>, rot=<0.4955 0.4297 0.4939 -0.5708>}
 dyn_star_B_05 {parent=MSTR_star_B_05, loc=<0 0 0>, rot=<0 0 0 1>}
 MSTR_star_B_06 {parent=dyn_chain_B_01_l, loc=<6.5506 -0.5288 0.3865>, rot=<0.4714 0.3031 0.5539 -0.6157>}
 dyn_star_B_06 {parent=MSTR_star_B_06, loc=<0 0 0>, rot=<0 0 0 1>}
 MSTR_chain_B_r {parent=MSTR_chains, loc=<0.7695 0.9997 7.5076>, rot=<-0.0378 -0.9615 0.0227 0.2712>}
 dyn_chain_B_01_r {parent=MSTR_chain_B_r, loc=<0 0 0>, rot=<0 0 0 1>}
 dyn_chain_B_02_r {parent=dyn_chain_B_01_r, loc=<6.8362 0 0>, rot=<0.0914 0.2444 -0.1539 -0.953>}
 dyn_chain_B_03_r {parent=dyn_chain_B_02_r, loc=<5.3096 0 0>, rot=<0.0927 0.4988 -0.0501 -0.8603>}
 dyn_chain_B_04_r {parent=dyn_chain_B_03_r, loc=<7.0378 0 0>, rot=<0 0 0 1>}
 MSTR_star_B_02 {parent=dyn_chain_B_03_r, loc=<0.2332 -0.1134 -0.4015>, rot=<0.4845 -0.6324 -0.5183 -0.3108>}
 dyn_star_B_02 {parent=MSTR_star_B_02, loc=<0 0 0>, rot=<0 0 0 1>}
 MSTR_star_B_01 {parent=dyn_chain_B_02_r, loc=<0.7338 -0.2581 -0.5305>, rot=<0.6577 -0.7032 -0.2243 -0.1508>}
 dyn_star_B_01 {parent=MSTR_star_B_01, loc=<0 0 0>, rot=<0 0 0 1>}
 MSTR_chain_A_r {parent=MSTR_chains, loc=<0.3045 1.2736 7.4622>, rot=<-0.3355 -0.93 -0.0339 0.1463>}
 dyn_chain_A_01_r {parent=MSTR_chain_A_r, loc=<0 0 0>, rot=<0 0 0 1>}
 dyn_chain_A_02_r {parent=dyn_chain_A_01_r, loc=<4.9758 0 0>, rot=<0.0651 0.5153 -0.0242 -0.8542>}
 dyn_chain_A_03_r {parent=dyn_chain_A_02_r, loc=<7.2036 0 0>, rot=<0 0 0 1>}
 MSTR_star_A_01 {parent=dyn_chain_A_02_r, loc=<-0.2339 -0.0894 -0.3947>, rot=<0.4916 -0.6826 -0.5408 -0.0051>}
 dyn_star_A_01 {parent=MSTR_star_A_01, loc=<0 0 0>, rot=<0 0 0 1>}
 thigh_l {parent=pelvis, loc=<-1.0855 -0.2109 -8.7053>, rot=<0.0009 0.0355 0.0155 -0.9993>}
 calf_l {parent=thigh_l, loc=<-42.5284 -0.0019 -0.7297>, rot=<0.0006 -0.0161 -0.0391 -0.9991>}
 foot_l {parent=calf_l, loc=<-46.7401 -2.5014 -0.592>, rot=<0.0427 -0.0238 0.0353 -0.9982>}
 ball_l {parent=foot_l, loc=<-3.5658 10.8066 -0.1544>, rot=<0.0001 -0.0001 -0.7063 -0.7079>}
 foot_newpivot_l {parent=foot_l, loc=<7.8414 0.9849 0.1343>, rot=<-0.0096 0.0172 -0.0189 -0.9996>}
 calf_twist_02_l {parent=calf_l, loc=<-19.1546 -1.7943 -0.2638>, rot=<0.0237 0.0095 -0.0035 -0.9997>}
 calf_twist_01_l {parent=calf_l, loc=<-35.8483 -1.9197 -0.5794>, rot=<0.0237 0.0095 -0.0035 -0.9997>}
 deform_knee_in_base_l {parent=calf_l, loc=<0.8031 -5.0231 0.2456>, rot=<0 0 0 -1>}
 deform_knee_in_l {parent=deform_knee_in_base_l, loc=<0 0 0>, rot=<0 0 0 1>}
 deform_knee_out_base_l {parent=calf_l, loc=<0 4 0>, rot=<0 0 0 -1>}
 deform_knee_out_l {parent=deform_knee_out_base_l, loc=<0 0 0>, rot=<0 0 0 1>}
 thigh_twist_01_l {parent=thigh_l, loc=<-27.952 0 0>, rot=<0.0013 0 0 -1>}
 deform_groin_base_l {parent=thigh_l, loc=<0.4801 5.6776 -0.2486>, rot=<0 0 0 -1>}
 deform_groin_l {parent=deform_groin_base_l, loc=<0 0 0>, rot=<0 0 0 1>}
 thigh_r {parent=pelvis, loc=<-1.0855 -0.2109 8.7053>, rot=<0.0355 -0.0009 0.9993 0.0155>}
 calf_r {parent=thigh_r, loc=<42.5284 0.0019 0.7297>, rot=<0.0006 -0.0161 -0.0391 -0.9991>}
 foot_r {parent=calf_r, loc=<46.7401 2.5014 0.592>, rot=<0.0427 -0.0238 0.0353 -0.9982>}
 ball_r {parent=foot_r, loc=<3.5658 -10.8066 0.1544>, rot=<0.0001 -0.0001 -0.7063 -0.7079>}
 foot_newpivot_r {parent=foot_r, loc=<-7.8414 -0.9849 -0.1343>, rot=<0 0 0 1>}
 calf_twist_02_r {parent=calf_r, loc=<19.1546 1.7943 0.2638>, rot=<0.0237 0.0095 -0.0035 -0.9997>}
 calf_twist_01_r {parent=calf_r, loc=<35.8483 1.9197 0.5794>, rot=<0.0237 0.0095 -0.0035 -0.9997>}
 deform_knee_in_base_r {parent=calf_r, loc=<-0.8031 5.0231 -0.2457>, rot=<0 0 0 -1>}
 deform_knee_in_r {parent=deform_knee_in_base_r, loc=<0 0 0>, rot=<0 0 0 1>}
 deform_knee_out_base_r {parent=calf_r, loc=<0 -4 0>, rot=<0 0 0 -1>}
 deform_knee_out_r {parent=deform_knee_out_base_r, loc=<0 0 0>, rot=<0 0 0 1>}
 thigh_twist_01_r {parent=thigh_r, loc=<27.952 0 0>, rot=<0.0013 0 0 -1>}
 deform_groin_base_r {parent=thigh_r, loc=<-0.4801 -5.6776 0.2486>, rot=<0 0 0 -1>}
 deform_groin_r {parent=deform_groin_base_r, loc=<0 0 0>, rot=<0 0 0 1>}
 deform_glute_base_l {parent=pelvis, loc=<-7.3466 -9.0213 -9.1245>, rot=<0.0172 0.031 -0.4477 -0.8935>}
 deform_glute_l {parent=deform_glute_base_l, loc=<0 0 0>, rot=<0 0 0 1>}
 deform_glute_base_r {parent=pelvis, loc=<-7.3464 -9.0213 9.1245>, rot=<0.031 -0.0172 0.8935 -0.4477>}
 deform_glute_r {parent=deform_glute_base_r, loc=<0 0 0>, rot=<0 0 0 1>}
 MSTR_backSkirt_l {parent=pelvis, loc=<0.7678 -9.276 -10.4531>, rot=<0.0367 -0.8205 0.5422 -0.1771>}
 dyn_backSkirt_01_l {parent=MSTR_backSkirt_l, loc=<0 0 0>, rot=<0 0 0 1>}
 dyn_backSkirt_02_l {parent=dyn_backSkirt_01_l, loc=<9.1921 0 0>, rot=<0.0011 -0.024 0.0108 -0.9997>}
 MSTR_frontSkirt_l {parent=pelvis, loc=<0.1961 5.8697 -9.863>, rot=<-0.0279 -0.8877 -0.4363 -0.1443>}
 dyn_frontSkirt_01_l {parent=MSTR_frontSkirt_l, loc=<0 0 0>, rot=<0 0 0 1>}
 dyn_frontSkirt_02_l {parent=dyn_frontSkirt_01_l, loc=<9.1324 0 0>, rot=<0.0004 0.0056 0.0071 -1>}
 MSTR_sideSkirt_l {parent=pelvis, loc=<0.2271 -1.0964 -14.2767>, rot=<-0.0013 -0.9781 0.0604 -0.1993>}
 dyn_sideSkirt_01_l {parent=MSTR_sideSkirt_l, loc=<0 0 0>, rot=<0 0 0 1>}
 dyn_sideSkirt_02_l {parent=dyn_sideSkirt_01_l, loc=<9.3419 0 0>, rot=<-0.0002 -0.021 -0.0035 -0.9998>}
 MSTR_frontSkirt_r {parent=pelvis, loc=<0.0161 5.8008 10.0893>, rot=<0.1509 0.4741 0.8666 0.0374>}
 dyn_frontSkirt_01_r {parent=MSTR_frontSkirt_r, loc=<0 0 0>, rot=<0 0 0 1>}
 dyn_frontSkirt_02_r {parent=dyn_frontSkirt_01_r, loc=<9.1921 0 0>, rot=<0.0011 -0.024 0.0108 -0.9997>}
 MSTR_sideSkirt_r {parent=pelvis, loc=<0.227 -1.0964 14.2767>, rot=<0.1993 -0.0604 0.9781 0.0013>}
 dyn_sideSkirt_01_r {parent=MSTR_sideSkirt_r, loc=<0 0 0>, rot=<0 0 0 1>}
 dyn_sideSkirt_02_r {parent=dyn_sideSkirt_01_r, loc=<9.3419 0 0>, rot=<0.0002 -0.021 0.0035 -0.9998>}
 MSTR_backSkirt_r {parent=pelvis, loc=<0.5696 -9.3218 10.3441>, rot=<0.1705 -0.4391 0.8813 -0.0384>}
 dyn_backSkirt_01_r {parent=MSTR_backSkirt_r, loc=<0 0 0>, rot=<0 0 0 1>}
 dyn_backSkirt_02_r {parent=dyn_backSkirt_01_r, loc=<9.1324 0 0>, rot=<0.0004 0.0056 0.0071 -1>}
 MSTR_frontSkirt {parent=pelvis, loc=<-0.2986 9.9285 0.0787>, rot=<-0.0854 -0.7107 -0.695 -0.0682>}
 dyn_frontSkirt_01 {parent=MSTR_frontSkirt, loc=<0 0 0>, rot=<0 0 0 1>}
 dyn_frontSkirt_02 {parent=dyn_frontSkirt_01, loc=<9.7239 0 0>, rot=<-0.0006 -0.0226 -0.0056 -0.9997>}
 MSTR_backSkirt {parent=pelvis, loc=<0.3065 -11.632 -0.1052>, rot=<0.0963 -0.6906 0.7081 -0.1117>}
 dyn_backSkirt_01 {parent=MSTR_backSkirt, loc=<0 0 0>, rot=<0 0 0 1>}
 dyn_backSkirt_02 {parent=dyn_backSkirt_01, loc=<9.7239 0 0>, rot=<-0.001 0.0487 -0.0055 -0.9988>}
 ground {parent=root, loc=<0 0 0>, rot=<0 0 0 1>}
]
variables [
 crouching {type=bool, default=False}
 driving {type=bool, default=False}
 rban alpha {type=double, default=0}
 rban enabled {type=bool, default=False}
 sliding {type=bool, default=False}
]
