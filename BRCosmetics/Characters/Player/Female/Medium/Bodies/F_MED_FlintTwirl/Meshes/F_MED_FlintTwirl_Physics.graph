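# Physics asset for the FlintTwirl medium female player cosmetic body, defining secondary motion for its ribbons, sashes and back skirt.
# Capsules on pelvis, spine and legs, plus an unconstrained ground box, provide body and floor shapes.
# Dynamic bone chains hang from MSTR anchors constrained to the spine, while the skirt chains from the pelvis.
# Ribbon joints use angular limits, but skirt and lower back-sash joints are set Free despite authored limit angles.
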PhysicsAsset /BRCosmetics/Characters/Player/Female/Medium/Bodies/F_MED_FlintTwirl/Meshes/F_MED_FlintTwirl_Physics
bodies {
 MSTR_backRibbon_A_l [{shapeType=Capsule, length=5, radius=1, shapeName=MSTR_backRibbon_A_l_capsule}]
 MSTR_backRibbon_A_r [{shapeType=Capsule, length=5, radius=1, shapeName=MSTR_backRibbon_A_r_capsule}]
 MSTR_backRibbon_B_l [{shapeType=Capsule, length=5, radius=1, shapeName=MSTR_backRibbon_B_l_capsule}]
 MSTR_backRibbon_B_r [{shapeType=Capsule, length=5, radius=1, shapeName=MSTR_backRibbon_B_r_capsule}]
 MSTR_backSash_l [{shapeType=Capsule, length=5, radius=1, rotation=<rot 90 0 0>}]
 MSTR_backSash_r [{shapeType=Capsule, length=5, radius=1, rotation=<rot 90 0 0>}]
 MSTR_hipRibbon [{
   shapeType=Capsule
   length=4.0872
   radius=2.7584
   shapeName=MSTR_hipRibbon_capsule
   center=<-0.6273 0.8787 0.3245>
   rotation=<rot 0 25.7153 -84.2503>
  }]
 MSTR_sash [{shapeType=Capsule, length=0.505, radius=0.505, shapeName=MSTR_sash_capsule, rotation=<rot 90 0 0>}]
 calf_l [{shapeType=Capsule, length=25, radius=5, shapeName=calf_l_capsule, center=<-17 -2 0>, rotation=<rot 90 5 0>}]
 calf_r [{shapeType=Capsule, length=25, radius=5, shapeName=calf_r_capsule, center=<17 2 0>, rotation=<rot 90 5 0>}]
 dyn_backRibbon_01_B_l [{shapeType=Capsule, length=2.8448, radius=2, shapeName=dyn_backRibbon_01_B_l_capsule, center=<3.9692 0 0>, rotation=<rot 90 0 0>}]
 dyn_backRibbon_01_B_r [{shapeType=Capsule, length=2.8448, radius=2, shapeName=dyn_backRibbon_01_B_r_capsule, center=<-3.9692 0 0>, rotation=<rot 90 0 0>}]
 dyn_backRibbon_02_B_l [{shapeType=Capsule, length=3.2786, radius=2.6, shapeName=dyn_backRibbon_02_B_l_capsule, center=<3.9692 0 0>, rotation=<rot 90 0 0>}]
 dyn_backRibbon_02_B_r [{shapeType=Capsule, length=2.8448, radius=2.6, shapeName=dyn_backRibbon_02_B_r_capsule, center=<-3.9692 0 0>, rotation=<rot 90 0 0>}]
 dyn_backRibbon_03_B_l [{shapeType=Capsule, length=3.5213, radius=2.8, shapeName=dyn_backRibbon_03_B_l_capsule, center=<4.2688 -0.3305 0>}]
 dyn_backRibbon_03_B_r [{shapeType=Capsule, length=3.5213, radius=2.8, shapeName=dyn_backRibbon_03_B_r_capsule, center=<-4.2688 0.3305 0>}]
 dyn_backRibbon_A_01_l [{shapeType=Capsule, length=2.7181, radius=1.4, shapeName=dyn_backRibbon_A_01_l_capsule, center=<0.9937 0 0>, rotation=<rot 90 0 0>}]
 dyn_backRibbon_A_01_r [{shapeType=Capsule, length=2.1191, radius=1.4, shapeName=dyn_backRibbon_A_01_r_capsule, center=<-0.9937 0 0>, rotation=<rot 90 0 0>}]
 dyn_backRibbon_A_02_l [{shapeType=Capsule, length=3.8126, radius=3, shapeName=dyn_backRibbon_A_02_l_capsule, center=<3.1049 -0.4491 1.2671>, rotation=<rot 0 0 -180>}]
 dyn_backRibbon_A_02_r [{shapeType=Capsule, length=3.8126, radius=3, shapeName=dyn_backRibbon_A_02_r_capsule, center=<-3.1049 0.4491 -1.1658>}]
 dyn_backSash_01_l [
  {shapeType=Capsule, length=10, radius=3, shapeName=dyn_backSash_01_l_capsule, center=<13.2107 0 0>, rotation=<rot 90 0 0>}
  {
   shapeType=Capsule
   length=14.0219
   radius=1.244
   shapeName=dyn_backSash_01_r_capsule
   center=<12.0368 5.5507 -2.5492>
   rotation=<rot 82.9813 17.4555 -0.0044>
  }
 ]
 dyn_backSash_01_r [
  {shapeType=Capsule, length=10, radius=3, shapeName=dyn_backSash_01_r_capsule, center=<-13.2107 0 0>, rotation=<rot 90 0 0>}
  {
   shapeType=Capsule
   length=14.0219
   radius=1.244
   shapeName=dyn_backSash_01_r_capsule
   center=<-12.0368 -5.5507 2.5492>
   rotation=<rot 82.9813 17.4555 -0.0044>
  }
 ]
 dyn_backSash_02_l [{shapeType=Capsule, length=10, radius=3, shapeName=dyn_backSash_02_l_capsule, center=<13.2107 0 0>, rotation=<rot 90 0 0>}]
 dyn_backSash_02_r [{shapeType=Capsule, length=10, radius=3, shapeName=dyn_backSash_02_r_capsule, center=<-13.2107 0 0>, rotation=<rot 90 0 0>}]
 dyn_backSash_03_l [{shapeType=Capsule, length=10, radius=3, shapeName=dyn_backSash_03_l_capsule, center=<12 0 0>, rotation=<rot 90 0 0>}]
 dyn_backSash_03_r [{shapeType=Capsule, length=10, radius=3, shapeName=dyn_backSash_03_r_capsule, center=<-12 0 0>, rotation=<rot 90 0 0>}]
 dyn_backSash_04_l [
  {shapeType=Capsule, length=11.3098, radius=3.4101, shapeName=dyn_backSash_04_l_capsule, center=<12.9142 1.0817 0>, rotation=<rot 90 180 0>}
  {
   shapeType=Capsule
   length=17.4568
   radius=2.4212
   shapeName=dyn_backSash_01_l_capsule
   center=<25.4664 -5.4559 1.085>
   rotation=<rot 82.9812 17.4551 -160.773>
  }
 ]
 dyn_backSash_04_r [
  {shapeType=Capsule, length=11.3098, radius=3.4101, shapeName=dyn_backSash_04_r_capsule, center=<-12.9142 -1.0817 0>, rotation=<rot 90 0 0>}
  {
   shapeType=Capsule
   length=19.9306
   radius=2.4212
   shapeName=dyn_backSash_01_r_capsule
   center=<-25.4664 5.4559 -1.085>
   rotation=<rot 82.9812 17.4551 19.227>
  }
 ]
 dyn_backSkirt_01 [{shapeType=Capsule, length=20, radius=3, shapeName=dyn_backSkirt_01_capsule, center=<4.2852 0 -0.4431>, rotation=<rot 0 0 90>}]
 dyn_backSkirt_02 [{
   shapeType=Capsule
   length=18.2843
   radius=4.7157
   shapeName=dyn_backSkirt_02_capsule
   center=<5.8821 0 -0.6331>
   rotation=<rot -0.0002 0 90.9537>
  }]
 dyn_hipRibbon_01_A_l [{shapeType=Capsule, length=3, radius=1.5, shapeName=dyn_hipRibbon_01_A_l_capsule, center=<1.7159 0 0>, rotation=<rot 90 0 0>}]
 dyn_hipRibbon_01_A_r [{shapeType=Capsule, length=3, radius=1.5, shapeName=dyn_hipRibbon_01_A_r_capsule, center=<1.7159 -0.7742 0>, rotation=<rot 90 0 0>}]
 dyn_hipRibbon_01_B_l [{shapeType=Capsule, length=3.2689, radius=1.2312, shapeName=dyn_hipRibbon_01_B_l_capsule, center=<2.2355 0 0>, rotation=<rot 90 0 0>}]
 dyn_hipRibbon_01_B_r [{
   shapeType=Capsule
   length=3.0973
   radius=1.4027
   shapeName=dyn_hipRibbon_01_B_r_capsule
   center=<2.2355 1.2799 0.9341>
   rotation=<rot 90 0 0>
  }]
 dyn_sash_01 [{shapeType=Capsule, length=2.3285, radius=1.2719, shapeName=dyn_sash_01_capsule, center=<0.1783 0 0.5537>, rotation=<rot 90 0 0>}]
 dyn_sash_02 [{shapeType=Capsule, length=3.0963, radius=1.7193, shapeName=dyn_sash_02_capsule, center=<1.9103 0 0.479>, rotation=<rot 83.5252 0 0>}]
 ground [{shapeType=Box, extentX=150, extentY=150, extentZ=4.625, shapeName=root_box, center=<0 -36.3376 2.3008>, rotation=<rot 0 -90 0>}]
 pelvis [{shapeType=Capsule, length=13.9597, radius=10.2676, center=<0 -1.551 0>}]
 spine_01 [{shapeType=Capsule, length=9.37, radius=9.2416, shapeName=spine_01_capsule, center=<0.2776 -0.6026 0>}]
 spine_02 [{shapeType=Capsule, length=7.9436, radius=8.8157, center=<0 2.6672 0>}]
 spine_05 [{shapeType=Capsule, length=9.6958, radius=9.0349, shapeName=spine_05_capsule, center=<-5.0741 1.7033 0>}]
 thigh_l [{shapeType=Capsule, length=27, radius=7, center=<-22.5 0 0>, rotation=<rot 90 0 0>}]
 thigh_r [{shapeType=Capsule, length=27, radius=7, center=<22.5 0 0>, rotation=<rot 90 0 0>}]
}
constraints [
 {
  bone1Name=dyn_backSkirt_01
  bone2Name=pelvis
  swing1LimitDegrees=45
  swing1Motion=Free
  swing2LimitDegrees=45
  swing2Motion=Free
  twistLimitDegrees=45
  twistMotion=Free
 }
 {
  bone1Name=dyn_backSkirt_02
  bone2Name=dyn_backSkirt_01
  swing1LimitDegrees=45
  swing1Motion=Free
  swing2LimitDegrees=45
  swing2Motion=Free
  twistLimitDegrees=45
  twistMotion=Free
 }
 {
  bone1Name=spine_01
  bone2Name=pelvis
  swing1LimitDegrees=45
  swing1Motion=Free
  swing2LimitDegrees=45
  swing2Motion=Free
  twistLimitDegrees=45
  twistMotion=Free
 }
 {bone1Name=spine_02, bone2Name=spine_01, swing1LimitDegrees=45, swing2LimitDegrees=45, twistLimitDegrees=45}
 {
  bone1Name=MSTR_backRibbon_A_l
  bone2Name=spine_02
  swing1LimitDegrees=45
  swing1Motion=Free
  swing2LimitDegrees=45
  swing2Motion=Free
  twistLimitDegrees=45
  twistMotion=Free
 }
 {bone1Name=dyn_backRibbon_A_01_l, bone2Name=MSTR_backRibbon_A_l, swing1LimitDegrees=30, swing2LimitDegrees=30, twistLimitDegrees=0}
 {bone1Name=dyn_backRibbon_A_02_l, bone2Name=dyn_backRibbon_A_01_l, swing1LimitDegrees=30, swing2LimitDegrees=10, twistLimitDegrees=0}
 {
  bone1Name=MSTR_backRibbon_A_r
  bone2Name=spine_02
  swing1LimitDegrees=45
  swing1Motion=Free
  swing2LimitDegrees=45
  swing2Motion=Free
  twistLimitDegrees=45
  twistMotion=Free
 }
 {bone1Name=dyn_backRibbon_A_01_r, bone2Name=MSTR_backRibbon_A_r, swing1LimitDegrees=30, swing2LimitDegrees=30, twistLimitDegrees=0}
 {bone1Name=dyn_backRibbon_A_02_r, bone2Name=dyn_backRibbon_A_01_r, swing1LimitDegrees=30, swing2LimitDegrees=10, twistLimitDegrees=0}
 {
  bone1Name=MSTR_backRibbon_B_l
  bone2Name=spine_02
  swing1LimitDegrees=45
  swing1Motion=Free
  swing2LimitDegrees=45
  swing2Motion=Free
  twistLimitDegrees=45
  twistMotion=Free
 }
 {bone1Name=dyn_backRibbon_01_B_l, bone2Name=MSTR_backRibbon_B_l, swing1LimitDegrees=10, swing2LimitDegrees=10, twistLimitDegrees=0}
 {bone1Name=dyn_backRibbon_02_B_l, bone2Name=dyn_backRibbon_01_B_l, swing1LimitDegrees=25, swing2LimitDegrees=10, twistLimitDegrees=0}
 {bone1Name=dyn_backRibbon_03_B_l, bone2Name=dyn_backRibbon_02_B_l, swing1LimitDegrees=35, swing2LimitDegrees=20, twistLimitDegrees=5}
 {
  bone1Name=MSTR_backRibbon_B_r
  bone2Name=spine_02
  swing1LimitDegrees=45
  swing1Motion=Free
  swing2LimitDegrees=45
  swing2Motion=Free
  twistLimitDegrees=45
  twistMotion=Free
 }
 {bone1Name=dyn_backRibbon_01_B_r, bone2Name=MSTR_backRibbon_B_r, swing1LimitDegrees=10, swing2LimitDegrees=10, twistLimitDegrees=0}
 {bone1Name=dyn_backRibbon_02_B_r, bone2Name=dyn_backRibbon_01_B_r, swing1LimitDegrees=25, swing2LimitDegrees=10, twistLimitDegrees=0}
 {bone1Name=dyn_backRibbon_03_B_r, bone2Name=dyn_backRibbon_02_B_r, swing1LimitDegrees=35, swing2LimitDegrees=20, twistLimitDegrees=5}
 {bone1Name=MSTR_backSash_l, bone2Name=spine_02, swing1LimitDegrees=45, swing2LimitDegrees=45, twistLimitDegrees=45}
 {bone1Name=dyn_backSash_01_l, bone2Name=MSTR_backSash_l, swing1LimitDegrees=45, swing2LimitDegrees=15, twistLimitDegrees=5}
 {
  bone1Name=dyn_backSash_02_l
  bone2Name=dyn_backSash_01_l
  swing1LimitDegrees=50
  swing1Motion=Free
  swing2LimitDegrees=30
  swing2Motion=Free
  twistLimitDegrees=5
  twistMotion=Free
 }
 {
  bone1Name=dyn_backSash_03_l
  bone2Name=dyn_backSash_02_l
  swing1LimitDegrees=60
  swing1Motion=Free
  swing2LimitDegrees=45
  swing2Motion=Free
  twistLimitDegrees=10
  twistMotion=Free
 }
 {
  bone1Name=dyn_backSash_04_l
  bone2Name=dyn_backSash_03_l
  swing1LimitDegrees=90
  swing1Motion=Free
  swing2LimitDegrees=60
  swing2Motion=Free
  twistLimitDegrees=25
  twistMotion=Free
 }
 {bone1Name=MSTR_backSash_r, bone2Name=spine_02, swing1LimitDegrees=45, swing2LimitDegrees=45, twistLimitDegrees=45}
 {bone1Name=dyn_backSash_01_r, bone2Name=MSTR_backSash_r, swing1LimitDegrees=45, swing2LimitDegrees=15, twistLimitDegrees=5}
 {
  bone1Name=dyn_backSash_02_r
  bone2Name=dyn_backSash_01_r
  swing1LimitDegrees=50
  swing1Motion=Free
  swing2LimitDegrees=30
  swing2Motion=Free
  twistLimitDegrees=5
  twistMotion=Free
 }
 {
  bone1Name=dyn_backSash_03_r
  bone2Name=dyn_backSash_02_r
  swing1LimitDegrees=60
  swing1Motion=Free
  swing2LimitDegrees=45
  swing2Motion=Free
  twistLimitDegrees=10
  twistMotion=Free
 }
 {
  bone1Name=dyn_backSash_04_r
  bone2Name=dyn_backSash_03_r
  swing1LimitDegrees=90
  swing1Motion=Free
  swing2LimitDegrees=60
  swing2Motion=Free
  twistLimitDegrees=25
  twistMotion=Free
 }
 {
  bone1Name=MSTR_hipRibbon
  bone2Name=spine_02
  swing1LimitDegrees=45
  swing1Motion=Free
  swing2LimitDegrees=45
  swing2Motion=Free
  twistLimitDegrees=45
  twistMotion=Free
 }
 {bone1Name=dyn_hipRibbon_01_A_l, bone2Name=MSTR_hipRibbon, swing1LimitDegrees=45, swing2LimitDegrees=45, twistLimitDegrees=0}
 {bone1Name=dyn_hipRibbon_01_A_r, bone2Name=MSTR_hipRibbon, swing1LimitDegrees=45, swing2LimitDegrees=45, twistLimitDegrees=0}
 {bone1Name=dyn_hipRibbon_01_B_l, bone2Name=MSTR_hipRibbon, swing1LimitDegrees=5, swing2LimitDegrees=45, twistLimitDegrees=0}
 {bone1Name=dyn_hipRibbon_01_B_r, bone2Name=MSTR_hipRibbon, swing1LimitDegrees=15, swing2LimitDegrees=45, twistLimitDegrees=0}
 {
  bone1Name=spine_05
  bone2Name=spine_02
  swing1LimitDegrees=45
  swing1Motion=Free
  swing2LimitDegrees=45
  swing2Motion=Free
  twistLimitDegrees=45
  twistMotion=Free
 }
 {
  bone1Name=MSTR_sash
  bone2Name=spine_05
  swing1LimitDegrees=45
  swing1Motion=Free
  swing2LimitDegrees=45
  swing2Motion=Free
  twistLimitDegrees=45
  twistMotion=Free
 }
 {bone1Name=dyn_sash_01, bone2Name=MSTR_sash, swing1LimitDegrees=45, swing2LimitDegrees=45, twistLimitDegrees=0}
 {bone1Name=dyn_sash_02, bone2Name=dyn_sash_01, swing1LimitDegrees=45, swing2LimitDegrees=45, twistLimitDegrees=0}
 {bone1Name=thigh_l, bone2Name=pelvis, swing1LimitDegrees=45, swing2LimitDegrees=45, twistLimitDegrees=45}
 {
  bone1Name=calf_l
  bone2Name=thigh_l
  swing1LimitDegrees=45
  swing1Motion=Free
  swing2LimitDegrees=45
  swing2Motion=Free
  twistLimitDegrees=45
  twistMotion=Free
 }
 {bone1Name=thigh_r, bone2Name=pelvis, swing1LimitDegrees=45, swing2LimitDegrees=45, twistLimitDegrees=45}
 {
  bone1Name=calf_r
  bone2Name=thigh_r
  swing1LimitDegrees=45
  swing1Motion=Free
  swing2LimitDegrees=45
  swing2Motion=Free
  twistLimitDegrees=45
  twistMotion=Free
 }
]
