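# Post-process style AnimBlueprint for the FlintTwirl female body, animating skirts, pendant, stars, sashes and ribbons.
# Copied parent pose runs through Control Rig and AnimDynamics into a cached pose.
# bRigidBodyEnabled selects RigidBody simulation with chain constraints and backpack ribbon fixes, or AnimDynamics sashes with DBNO ribbon offsets.
# Update event sets gravity, simulation alpha and ribbon weights from backbling, skydive and vehicle state, plus interpolated slope-warp values.
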AnimBlueprint /BRCosmetics/Characters/Player/Female/Medium/Bodies/F_MED_FlintTwirl/Meshes/F_MED_FlintTwirl_AnimBP
target_skeleton=&/F_MED_FlintTwirl_Skeleton
graphs [
 AnimGraph {nodes [
   n1 = Anim Dynamics - Bone: dyn_backSash_01_l!(
    ComponentPose=Local To Component(LocalPose=Use cached pose 'CRCache_preDyn'())
    Alpha=1
    GravityOverride:Vector="0,0,-980"
    ExternalForce="0,0,0"
   ) -> (Pose) {
    angularDampingOverride=0.5
    angularSpringConstant=250
    bAngularSpring=true
    bOverrideAngularDamping=true
    bUseGravityOverride=true
    bUsePlanarLimit=false
    gravityScale=0.35
    simulationSpace=BoneRelative
    boundBone {boneName=dyn_backSash_01_l}
    componentAppliedLinearAccClamp=<0 0 0>
    gravityOverride {z=-980}
    physicsBodyDefinitions [{
      boundBone {boneName=dyn_backSash_01_l}
      boxExtents=<20 30 30>
      constraintSetup {angularLimitsMax=<0 0 5>, angularLimitsMin=<0 -15 0>, angularTarget=<1 0 0>}
      localJointOffset=<0.5 0 0>
     }]
    relativeSpaceBone {boneName=root}
    simSpaceSettings {maxAngularAcceleration=20, maxAngularVelocity=10, simSpaceAngularAlpha=1}
   }
   n2 = Anim Dynamics - Bone: dyn_backSash_02_l!(ComponentPose=@n1, Alpha=1, GravityOverride:Vector="0,0,-980", ExternalForce="0,0,0") -> (Pose) {
    angularDampingOverride=0.5
    angularSpringConstant=250
    bAngularSpring=true
    bOverrideAngularDamping=true
    bUseGravityOverride=true
    bUsePlanarLimit=false
    gravityScale=0.35
    simulationSpace=BoneRelative
    boundBone {boneName=dyn_backSash_02_l}
    componentAppliedLinearAccClamp=<0 0 0>
    gravityOverride {z=-980}
    physicsBodyDefinitions [{
      boundBone {boneName=dyn_backSash_02_l}
      boxExtents=<20 30 30>
      constraintSetup {angularLimitsMax=<0 0 10>, angularLimitsMin=<0 -15 0>, angularTarget=<1 0 0>}
      localJointOffset=<0.5 0 0>
     }]
    relativeSpaceBone {boneName=root}
    simSpaceSettings {maxAngularAcceleration=20, maxAngularVelocity=10, simSpaceAngularAlpha=1}
   }
   n3 = Anim Dynamics - Bone: dyn_backSash_03_l!(ComponentPose=@n2, Alpha=1, GravityOverride:Vector="0,0,-980", ExternalForce="0,0,0") -> (Pose) {
    angularDampingOverride=0.5
    angularSpringConstant=250
    bAngularSpring=true
    bOverrideAngularDamping=true
    bUseGravityOverride=true
    bUsePlanarLimit=false
    gravityScale=0.35
    simulationSpace=BoneRelative
    boundBone {boneName=dyn_backSash_03_l}
    componentAppliedLinearAccClamp=<0 0 0>
    gravityOverride {z=-980}
    physicsBodyDefinitions [{
      boundBone {boneName=dyn_backSash_03_l}
      boxExtents=<20 30 30>
      constraintSetup {angularLimitsMax=<0 0 15>, angularLimitsMin=<0 -15 -5>, angularTarget=<1 0 0>}
      localJointOffset=<0.5 0 0>
     }]
    relativeSpaceBone {boneName=root}
    simSpaceSettings {maxAngularAcceleration=20, maxAngularVelocity=10, simSpaceAngularAlpha=1}
   }
   n4 = Anim Dynamics - Bone: dyn_backSash_04_l!(ComponentPose=@n3, Alpha=1, GravityOverride:Vector="0,0,-980", ExternalForce="0,0,0") -> (Pose) {
    angularDampingOverride=0.5
    angularSpringConstant=250
    bAngularSpring=true
    bOverrideAngularDamping=true
    bUseGravityOverride=true
    bUsePlanarLimit=false
    gravityScale=0.35
    simulationSpace=BoneRelative
    boundBone {boneName=dyn_backSash_04_l}
    componentAppliedLinearAccClamp=<0 0 0>
    gravityOverride {z=-980}
    physicsBodyDefinitions [{
      boundBone {boneName=dyn_backSash_04_l}
      boxExtents=<20 30 30>
      constraintSetup {angularLimitsMax=<0 0 20>, angularLimitsMin=<0 -15 -10>, angularTarget=<1 0 0>}
      localJointOffset=<0.5 0 0>
     }]
    relativeSpaceBone {boneName=root}
    simSpaceSettings {maxAngularAcceleration=20, maxAngularVelocity=10, simSpaceAngularAlpha=1}
   }
   n5 = Anim Dynamics - Bone: dyn_backSash_01_r!(ComponentPose=@n4, Alpha=1, GravityOverride:Vector="0,0,-980", ExternalForce="0,0,0") -> (Pose) {
    angularDampingOverride=0.5
    angularSpringConstant=250
    bAngularSpring=true
    bOverrideAngularDamping=true
    bUseGravityOverride=true
    bUsePlanarLimit=false
    gravityScale=0.35
    simulationSpace=BoneRelative
    boundBone {boneName=dyn_backSash_01_r}
    componentAppliedLinearAccClamp=<0 0 0>
    gravityOverride {z=-980}
    physicsBodyDefinitions [{
      boundBone {boneName=dyn_backSash_01_r}
      boxExtents=<20 30 30>
      constraintSetup {angularLimitsMax=<0 0 5>, angularLimitsMin=<0 -15 0>, angularTarget=<1 0 0>}
      localJointOffset=<-0.5 0 0>
     }]
    relativeSpaceBone {boneName=root}
    simSpaceSettings {maxAngularAcceleration=20, maxAngularVelocity=10, simSpaceAngularAlpha=1}
   }
   n6 = Anim Dynamics - Bone: dyn_backSash_02_r!(ComponentPose=@n5, Alpha=1, GravityOverride:Vector="0,0,-980", ExternalForce="0,0,0") -> (Pose) {
    angularDampingOverride=0.5
    angularSpringConstant=250
    bAngularSpring=true
    bOverrideAngularDamping=true
    bUseGravityOverride=true
    bUsePlanarLimit=false
    gravityScale=0.35
    simulationSpace=BoneRelative
    boundBone {boneName=dyn_backSash_02_r}
    componentAppliedLinearAccClamp=<0 0 0>
    gravityOverride {z=-980}
    physicsBodyDefinitions [{
      boundBone {boneName=dyn_backSash_02_r}
      boxExtents=<20 30 30>
      constraintSetup {angularLimitsMax=<0 0 10>, angularLimitsMin=<0 -15 0>, angularTarget=<1 0 0>}
      localJointOffset=<-0.5 0 0>
     }]
    relativeSpaceBone {boneName=root}
    simSpaceSettings {maxAngularAcceleration=20, maxAngularVelocity=10, simSpaceAngularAlpha=1}
   }
   n7 = Anim Dynamics - Bone: dyn_backSash_03_r!(ComponentPose=@n6, Alpha=1, GravityOverride:Vector="0,0,-980", ExternalForce="0,0,0") -> (Pose) {
    angularDampingOverride=0.5
    angularSpringConstant=250
    bAngularSpring=true
    bOverrideAngularDamping=true
    bUseGravityOverride=true
    bUsePlanarLimit=false
    gravityScale=0.35
    simulationSpace=BoneRelative
    boundBone {boneName=dyn_backSash_03_r}
    componentAppliedLinearAccClamp=<0 0 0>
    gravityOverride {z=-980}
    physicsBodyDefinitions [{
      boundBone {boneName=dyn_backSash_03_r}
      boxExtents=<20 30 30>
      constraintSetup {angularLimitsMax=<0 0 15>, angularLimitsMin=<0 -15 -5>, angularTarget=<1 0 0>}
      localJointOffset=<-0.5 0 0>
     }]
    relativeSpaceBone {boneName=root}
    simSpaceSettings {maxAngularAcceleration=20, maxAngularVelocity=10, simSpaceAngularAlpha=1}
   }
   n8 = Anim Dynamics - Bone: dyn_backSash_04_r!(ComponentPose=@n7, Alpha=1, GravityOverride:Vector="0,0,-980", ExternalForce="0,0,0") -> (Pose) {
    angularDampingOverride=0.5
    angularSpringConstant=250
    bAngularSpring=true
    bOverrideAngularDamping=true
    bUseGravityOverride=true
    bUsePlanarLimit=false
    gravityScale=0.35
    simulationSpace=BoneRelative
    boundBone {boneName=dyn_backSash_04_r}
    componentAppliedLinearAccClamp=<0 0 0>
    gravityOverride {z=-980}
    physicsBodyDefinitions [{
      boundBone {boneName=dyn_backSash_04_r}
      boxExtents=<20 30 30>
      constraintSetup {angularLimitsMax=<0 0 20>, angularLimitsMin=<0 -15 -10>, angularTarget=<1 0 0>}
      localJointOffset=<-0.5 0 0>
     }]
    relativeSpaceBone {boneName=root}
    simSpaceSettings {maxAngularAcceleration=20, maxAngularVelocity=10, simSpaceAngularAlpha=1}
   }
   n9 = "To Float (Boolean)"(InBool=Get bIsDBNO())
   n10 = "Transform (Modify) Bone - Bone: MSTR_backRibbon"(
    ComponentPose="Transform (Modify) Bone - Bone: MSTR_backRibbon_B_r"(
     ComponentPose="Transform (Modify) Bone - Bone: MSTR_backRibbon_B_l"(ComponentPose=@n8, Alpha=@n9, Translation="0,0,0", Rotation="0,0,-10", Scale="1,1,1") {rotationMode=BMM_Additive, boneToModify {boneName=MSTR_backRibbon_B_l}, rotation {roll=-10}}
     Alpha=@n9
     Translation="0,0,0"
     Rotation="0,0,-10"
     Scale="1,1,1"
    ) {rotationMode=BMM_Additive, boneToModify {boneName=MSTR_backRibbon_B_r}, rotation {roll=-10}}
    Alpha=@n9
    Translation="0,0,0"
    Rotation="0,-20,0"
    Scale="1,1,1"
   ) {
    rotationMode=BMM_Additive
    rotationSpace=BCS_BoneSpace
    scaleMode=BMM_Replace
    scaleSpace=BCS_BoneSpace
    translationSpace=BCS_BoneSpace
    boneToModify {boneName=MSTR_backRibbon}
    rotation {yaw=-20}
   }
   n11 = "Transform (Modify) Bone - Bone: MSTR_backRibbon"(
    ComponentPose=Local To Component(LocalPose=Use cached pose 'CRCache_preDyn'())
    Alpha=Get is_wearing_backpack() -> (is_wearing_backpack:double)
    Translation="-4,-1.5,0"
    Rotation="0,0,0"
    Scale="1,1,1"
   ) {
    rotationMode=BMM_Additive
    scaleMode=BMM_Replace
    scaleSpace=BCS_BoneSpace
    translationMode=BMM_Additive
    translationSpace=BCS_BoneSpace
    boneToModify {boneName=MSTR_backRibbon}
    translation=<-4 -1.5>
   }
   n12 = RigidBody!(
    ComponentPose=@n11
    Alpha=Get RBAN alpha() -> (RBAN alpha:double)
    OverrideWorldGravity:Vector="0,0,0"
    ExternalForce
    SimSpaceSettings:Sim Space Settings Structure=(WorldAlpha=1,VelocityScaleZ=0.4,DampingAlpha=1,MaxLinearVelocity=300,MaxAngularVelocity=10,MaxLinearAcceleration=1000,MaxAngularAcceleration=40,ExternalLinearDragV=(X=0.3,Y=0.3,Z=0.3),ExternalLinearVelocity=(X=0,Y=0,Z=0),ExternalAngularVelocity=(X=0,Y=0,Z=0))
   ) -> (Pose) {
    bClampLinearTranslationLimitToRefPose=true
    bOverrideWorldGravity=true
    cachedBoundsScale=1
    overridePhysicsAsset=&/F_MED_FlintTwirl_Physics
    simulationSpace=BaseBoneSpace
    alphaBoolBlend {blendOption=Linear}
    baseBoneRef {boneName=spine_01}
    componentAppliedLinearAccClamp=<0 0 0>
    overrideWorldGravity {z=-980}
    simSpaceSettings {
     maxAngularAcceleration=40
     maxAngularVelocity=10
     maxLinearAcceleration=1000
     maxLinearVelocity=300
     velocityScaleZ=0.4
     worldAlpha=1
     externalLinearDragV=<0.3 0.3 0.3>
    }
   }
   n13 = "Constraint - dyn_chain_B_01_r (R)"(ComponentPose=@n12, Alpha=0.25, ConstraintWeights_0=1) {
    boneToModify {boneName=dyn_chain_B_01_r}
    constraintSetup [{transformType=Rotation, targetBone {boneName=dyn_hipRibbon_01_B_r}}]
    constraintWeights [1]
   }
   n14 = "Constraint - dyn_chain_A_01_r (R)"(ComponentPose=@n13, Alpha=0.25, ConstraintWeights_0=1) {
    boneToModify {boneName=dyn_chain_A_01_r}
    constraintSetup [{transformType=Rotation, targetBone {boneName=dyn_hipRibbon_01_B_l}}]
    constraintWeights [1]
   }
   n15 = Get backpack_skydive() -> (backpack_skydive:double)
   n16 = Layered blend per bone(BasePose=Component To Local(ComponentPose=@n14), BlendPoses_0=Slot 'DefaultSlot'(), BlendWeights_0=@n15) {
    curveBlendOption=BlendByWeight
    layerSetup [{branchFilters [
       {blendDepth=1, boneName=dyn_backRibbon_01_B_l}
       {blendDepth=1, boneName=dyn_backRibbon_01_B_r}
       {blendDepth=1, boneName=dyn_backRibbon_01_A_l}
       {blendDepth=1, boneName=dyn_backRibbon_01_A_r}
      ]}]
   }
   n17 = "Transform (Modify) Bone - Bone: MSTR_backRibbon_B_l"(
    ComponentPose="Transform (Modify) Bone - Bone: MSTR_backRibbon"(ComponentPose=Local To Component(LocalPose=@n16), Alpha=@n15, Translation="0,0,0", Rotation="0,0,-5", Scale="1,1,1") {rotationMode=BMM_Additive, boneToModify {boneName=MSTR_backRibbon}, rotation {roll=-5}}
    Alpha=@n15
    Translation="0,0,0"
    Rotation="0,0,-25"
    Scale="1,1,1"
   ) {rotationMode=BMM_Additive, boneToModify {boneName=MSTR_backRibbon_B_l}, rotation {roll=-25}}
   n18 = Layered blend per bone(
    BasePose=Component To Local(ComponentPose="Transform (Modify) Bone - Bone: MSTR_backRibbon_B_r"(ComponentPose=@n17, Alpha=@n15, Translation="0,0,0", Rotation="0,0,-25", Scale="1,1,1") {rotationMode=BMM_Additive, boneToModify {boneName=MSTR_backRibbon_B_r}, rotation {roll=-25}})
    BlendPoses_0=Slot 'DefaultSlot'()
    BlendWeights_0=Get backpack_not_skydive() -> (backpack_not_skydive:double)
   ) {curveBlendOption=BlendByWeight, layerSetup [{branchFilters [{boneName=MSTR_backRibbon_A_l}, {boneName=MSTR_backRibbon_A_r}]}]}
   Output Pose(Result=Blend Poses by bool(BlendPose_0=@n18, BlendPose_1=Component To Local(ComponentPose=@n10), bActiveValue=Get bRigidBodyEnabled()) {blendType=Linear}) {layerGroup=None}
   n19 = Control Rig!(
    Source=Copy Pose From Mesh() {bUseAttachedParent=true}
    Alpha
    crouching:bool=Get bIsCrouching()
    sliding:bool=Get bIsSliding()
    rban alpha:double=Get RBAN alpha()
    rban enabled:bool=Get bRigidBodyEnabled()
    driving:bool=Get is_in_car_seat() -> (is_in_car_seat:bool)
   ) -> (Pose) {controlRigAssetReference {blueprintRigClass=&/F_MED_FlintTwirl_CtrlRig.F_MED_FlintTwirl_CtrlRig_C}}
   n20 = Anim Dynamics - Bone: dyn_frontSkirt_01_l!(ComponentPose=Local To Component(LocalPose=@n19), GravityOverride:Vector="0,0,-980", ExternalForce="0,0,0") -> (Pose) {
    angularDampingOverride=0.999
    angularSpringConstant=5000
    bAngularSpring=true
    bOverrideAngularDamping=true
    bUseGravityOverride=true
    gravityScale=0.35
    simulationSpace=BoneRelative
    boundBone {boneName=dyn_frontSkirt_01_l}
    componentAppliedLinearAccClamp=<0 0 0>
    gravityOverride {z=-980}
    physicsBodyDefinitions [{
      boundBone {boneName=dyn_frontSkirt_01_l}
      boxExtents=<20 30 30>
      constraintSetup {angularLimitsMax=<0 0 5>, angularLimitsMin=<0 -10 -5>, angularTarget=<1 0 0>}
      localJointOffset=<4 0 0>
     }]
    relativeSpaceBone {boneName=root}
    simSpaceSettings {maxAngularAcceleration=20, maxAngularVelocity=5, simSpaceAngularAlpha=1}
   }
   n21 = Anim Dynamics - Bone: dyn_frontSkirt_02_l!(ComponentPose=@n20, GravityOverride:Vector="0,0,-980", ExternalForce="0,0,0") -> (Pose) {
    angularDampingOverride=0.999
    angularSpringConstant=5000
    bAngularSpring=true
    bOverrideAngularDamping=true
    bUseGravityOverride=true
    gravityScale=0.35
    simulationSpace=BoneRelative
    boundBone {boneName=dyn_frontSkirt_02_l}
    componentAppliedLinearAccClamp=<0 0 0>
    gravityOverride {z=-980}
    physicsBodyDefinitions [{
      boundBone {boneName=dyn_frontSkirt_02_l}
      boxExtents=<20 30 30>
      constraintSetup {angularLimitsMax=<0 0 10>, angularLimitsMin=<0 -20 -10>, angularTarget=<1 0 0>}
      localJointOffset=<4 0 0>
     }]
    relativeSpaceBone {boneName=root}
    simSpaceSettings {maxAngularAcceleration=20, maxAngularVelocity=5, simSpaceAngularAlpha=1}
   }
   n22 = Anim Dynamics - Bone: dyn_sideSkirt_01_l!(ComponentPose=@n21, GravityOverride:Vector="0,0,-980", ExternalForce="0,0,0") -> (Pose) {
    angularDampingOverride=0.999
    angularSpringConstant=5000
    bAngularSpring=true
    bOverrideAngularDamping=true
    bUseGravityOverride=true
    gravityScale=0.35
    simulationSpace=BoneRelative
    boundBone {boneName=dyn_sideSkirt_01_l}
    componentAppliedLinearAccClamp=<0 0 0>
    gravityOverride {z=-980}
    physicsBodyDefinitions [{
      boundBone {boneName=dyn_sideSkirt_01_l}
      boxExtents=<20 30 30>
      constraintSetup {angularLimitsMax=<0 0 5>, angularLimitsMin=<0 -10 -5>, angularTarget=<1 0 0>}
      localJointOffset=<4 0 0>
     }]
    relativeSpaceBone {boneName=root}
    simSpaceSettings {maxAngularAcceleration=20, maxAngularVelocity=5, simSpaceAngularAlpha=1}
   }
   n23 = Anim Dynamics - Bone: dyn_sideSkirt_02_l!(ComponentPose=@n22, GravityOverride:Vector="0,0,-980", ExternalForce="0,0,0") -> (Pose) {
    angularDampingOverride=0.999
    angularSpringConstant=5000
    bAngularSpring=true
    bOverrideAngularDamping=true
    bUseGravityOverride=true
    gravityScale=0.35
    simulationSpace=BoneRelative
    boundBone {boneName=dyn_sideSkirt_02_l}
    componentAppliedLinearAccClamp=<0 0 0>
    gravityOverride {z=-980}
    physicsBodyDefinitions [{
      boundBone {boneName=dyn_sideSkirt_02_l}
      boxExtents=<20 30 30>
      constraintSetup {angularLimitsMax=<0 0 10>, angularLimitsMin=<0 -20 -10>, angularTarget=<1 0 0>}
      localJointOffset=<4 0 0>
     }]
    relativeSpaceBone {boneName=root}
    simSpaceSettings {maxAngularAcceleration=20, maxAngularVelocity=5, simSpaceAngularAlpha=1}
   }
   n24 = Anim Dynamics - Bone: dyn_backSkirt_01_l!(ComponentPose=@n23, GravityOverride:Vector="0,0,-980", ExternalForce="0,0,0") -> (Pose) {
    angularDampingOverride=0.999
    angularSpringConstant=5000
    bAngularSpring=true
    bOverrideAngularDamping=true
    bUseGravityOverride=true
    gravityScale=0.35
    simulationSpace=BoneRelative
    boundBone {boneName=dyn_backSkirt_01_l}
    componentAppliedLinearAccClamp=<0 0 0>
    gravityOverride {z=-980}
    physicsBodyDefinitions [{
      boundBone {boneName=dyn_backSkirt_01_l}
      boxExtents=<20 30 30>
      constraintSetup {angularLimitsMax=<0 0 5>, angularLimitsMin=<0 -10 -5>, angularTarget=<1 0 0>}
      localJointOffset=<4 0 0>
     }]
    relativeSpaceBone {boneName=root}
    simSpaceSettings {maxAngularAcceleration=20, maxAngularVelocity=5, simSpaceAngularAlpha=1}
   }
   n25 = Anim Dynamics - Bone: dyn_backSkirt_02_l!(ComponentPose=@n24, GravityOverride:Vector="0,0,-980", ExternalForce="0,0,0") -> (Pose) {
    angularDampingOverride=0.999
    angularSpringConstant=5000
    bAngularSpring=true
    bOverrideAngularDamping=true
    bUseGravityOverride=true
    gravityScale=0.35
    simulationSpace=BoneRelative
    boundBone {boneName=dyn_backSkirt_02_l}
    componentAppliedLinearAccClamp=<0 0 0>
    gravityOverride {z=-980}
    physicsBodyDefinitions [{
      boundBone {boneName=dyn_backSkirt_02_l}
      boxExtents=<20 30 30>
      constraintSetup {angularLimitsMax=<0 0 10>, angularLimitsMin=<0 -20 -10>, angularTarget=<1 0 0>}
      localJointOffset=<4 0 0>
     }]
    relativeSpaceBone {boneName=root}
    simSpaceSettings {maxAngularAcceleration=20, maxAngularVelocity=5, simSpaceAngularAlpha=1}
   }
   n26 = Anim Dynamics - Bone: dyn_frontSkirt_01_r!(ComponentPose=@n25, GravityOverride:Vector="0,0,-980", ExternalForce="0,0,0") -> (Pose) {
    angularDampingOverride=0.999
    angularSpringConstant=5000
    bAngularSpring=true
    bOverrideAngularDamping=true
    bUseGravityOverride=true
    gravityScale=0.35
    simulationSpace=BoneRelative
    boundBone {boneName=dyn_frontSkirt_01_r}
    componentAppliedLinearAccClamp=<0 0 0>
    gravityOverride {z=-980}
    physicsBodyDefinitions [{
      boundBone {boneName=dyn_frontSkirt_01_r}
      boxExtents=<20 30 30>
      constraintSetup {angularLimitsMax=<0 0 5>, angularLimitsMin=<0 -10 -5>, angularTarget=<1 0 0>}
      localJointOffset=<4 0 0>
     }]
    relativeSpaceBone {boneName=root}
    simSpaceSettings {maxAngularAcceleration=20, maxAngularVelocity=5, simSpaceAngularAlpha=1}
   }
   n27 = Anim Dynamics - Bone: dyn_frontSkirt_02_r!(ComponentPose=@n26, GravityOverride:Vector="0,0,-980", ExternalForce="0,0,0") -> (Pose) {
    angularDampingOverride=0.999
    angularSpringConstant=5000
    bAngularSpring=true
    bOverrideAngularDamping=true
    bUseGravityOverride=true
    gravityScale=0.35
    simulationSpace=BoneRelative
    boundBone {boneName=dyn_frontSkirt_02_r}
    componentAppliedLinearAccClamp=<0 0 0>
    gravityOverride {z=-980}
    physicsBodyDefinitions [{
      boundBone {boneName=dyn_frontSkirt_02_r}
      boxExtents=<20 30 30>
      constraintSetup {angularLimitsMax=<0 0 10>, angularLimitsMin=<0 -20 -10>, angularTarget=<1 0 0>}
      localJointOffset=<4 0 0>
     }]
    relativeSpaceBone {boneName=root}
    simSpaceSettings {maxAngularAcceleration=20, maxAngularVelocity=5, simSpaceAngularAlpha=1}
   }
   n28 = Anim Dynamics - Bone: dyn_sideSkirt_01_r!(ComponentPose=@n27, GravityOverride:Vector="0,0,-980", ExternalForce="0,0,0") -> (Pose) {
    angularDampingOverride=0.999
    angularSpringConstant=5000
    bAngularSpring=true
    bOverrideAngularDamping=true
    bUseGravityOverride=true
    gravityScale=0.35
    simulationSpace=BoneRelative
    boundBone {boneName=dyn_sideSkirt_01_r}
    componentAppliedLinearAccClamp=<0 0 0>
    gravityOverride {z=-980}
    physicsBodyDefinitions [{
      boundBone {boneName=dyn_sideSkirt_01_r}
      boxExtents=<20 30 30>
      constraintSetup {angularLimitsMax=<0 0 5>, angularLimitsMin=<0 -10 -5>, angularTarget=<1 0 0>}
      localJointOffset=<4 0 0>
     }]
    relativeSpaceBone {boneName=root}
    simSpaceSettings {maxAngularAcceleration=20, maxAngularVelocity=5, simSpaceAngularAlpha=1}
   }
   n29 = Anim Dynamics - Bone: dyn_sideSkirt_02_r!(ComponentPose=@n28, GravityOverride:Vector="0,0,-980", ExternalForce="0,0,0") -> (Pose) {
    angularDampingOverride=0.999
    angularSpringConstant=5000
    bAngularSpring=true
    bOverrideAngularDamping=true
    bUseGravityOverride=true
    gravityScale=0.35
    simulationSpace=BoneRelative
    boundBone {boneName=dyn_sideSkirt_02_r}
    componentAppliedLinearAccClamp=<0 0 0>
    gravityOverride {z=-980}
    physicsBodyDefinitions [{
      boundBone {boneName=dyn_sideSkirt_02_r}
      boxExtents=<20 30 30>
      constraintSetup {angularLimitsMax=<0 0 10>, angularLimitsMin=<0 -20 -10>, angularTarget=<1 0 0>}
      localJointOffset=<4 0 0>
     }]
    relativeSpaceBone {boneName=root}
    simSpaceSettings {maxAngularAcceleration=20, maxAngularVelocity=5, simSpaceAngularAlpha=1}
   }
   n30 = Anim Dynamics - Bone: dyn_backSkirt_01_r!(ComponentPose=@n29, GravityOverride:Vector="0,0,-980", ExternalForce="0,0,0") -> (Pose) {
    angularDampingOverride=0.999
    angularSpringConstant=5000
    bAngularSpring=true
    bOverrideAngularDamping=true
    bUseGravityOverride=true
    gravityScale=0.35
    simulationSpace=BoneRelative
    boundBone {boneName=dyn_backSkirt_01_r}
    componentAppliedLinearAccClamp=<0 0 0>
    gravityOverride {z=-980}
    physicsBodyDefinitions [{
      boundBone {boneName=dyn_backSkirt_01_r}
      boxExtents=<20 30 30>
      constraintSetup {angularLimitsMax=<0 0 5>, angularLimitsMin=<0 -10 -5>, angularTarget=<1 0 0>}
      localJointOffset=<4 0 0>
     }]
    relativeSpaceBone {boneName=root}
    simSpaceSettings {maxAngularAcceleration=20, maxAngularVelocity=5, simSpaceAngularAlpha=1}
   }
   n31 = Anim Dynamics - Bone: dyn_backSkirt_02_r!(ComponentPose=@n30, GravityOverride:Vector="0,0,-980", ExternalForce="0,0,0") -> (Pose) {
    angularDampingOverride=0.999
    angularSpringConstant=5000
    bAngularSpring=true
    bOverrideAngularDamping=true
    bUseGravityOverride=true
    gravityScale=0.35
    simulationSpace=BoneRelative
    boundBone {boneName=dyn_backSkirt_02_r}
    componentAppliedLinearAccClamp=<0 0 0>
    gravityOverride {z=-980}
    physicsBodyDefinitions [{
      boundBone {boneName=dyn_backSkirt_02_r}
      boxExtents=<20 30 30>
      constraintSetup {angularLimitsMax=<0 0 10>, angularLimitsMin=<0 -20 -10>, angularTarget=<1 0 0>}
      localJointOffset=<4 0 0>
     }]
    relativeSpaceBone {boneName=root}
    simSpaceSettings {maxAngularAcceleration=20, maxAngularVelocity=5, simSpaceAngularAlpha=1}
   }
   n32 = Anim Dynamics - Bone: dyn_frontSkirt_01!(ComponentPose=@n31, GravityOverride:Vector="0,0,-980", ExternalForce="0,0,0") -> (Pose) {
    angularDampingOverride=0.999
    angularSpringConstant=5000
    bAngularSpring=true
    bOverrideAngularDamping=true
    bUseGravityOverride=true
    gravityScale=0.35
    simulationSpace=BoneRelative
    boundBone {boneName=dyn_frontSkirt_01}
    componentAppliedLinearAccClamp=<0 0 0>
    gravityOverride {z=-980}
    physicsBodyDefinitions [{
      boundBone {boneName=dyn_frontSkirt_01}
      boxExtents=<20 30 30>
      constraintSetup {angularLimitsMax=<0 0 5>, angularLimitsMin=<0 -10 -5>, angularTarget=<1 0 0>}
      localJointOffset=<4 0 0>
     }]
    relativeSpaceBone {boneName=root}
    simSpaceSettings {maxAngularAcceleration=20, maxAngularVelocity=5, simSpaceAngularAlpha=1}
   }
   n33 = Anim Dynamics - Bone: dyn_frontSkirt_02!(ComponentPose=@n32, GravityOverride:Vector="0,0,-980", ExternalForce="0,0,0") -> (Pose) {
    angularDampingOverride=0.999
    angularSpringConstant=5000
    bAngularSpring=true
    bOverrideAngularDamping=true
    bUseGravityOverride=true
    gravityScale=0.35
    simulationSpace=BoneRelative
    boundBone {boneName=dyn_frontSkirt_02}
    componentAppliedLinearAccClamp=<0 0 0>
    gravityOverride {z=-980}
    physicsBodyDefinitions [{
      boundBone {boneName=dyn_frontSkirt_02}
      boxExtents=<20 30 30>
      constraintSetup {angularLimitsMax=<0 0 10>, angularLimitsMin=<0 -20 -10>, angularTarget=<1 0 0>}
      localJointOffset=<4 0 0>
     }]
    relativeSpaceBone {boneName=root}
    simSpaceSettings {maxAngularAcceleration=20, maxAngularVelocity=5, simSpaceAngularAlpha=1}
   }
   n34 = Anim Dynamics - Bone: dyn_backSkirt_01!(ComponentPose=@n33, GravityOverride:Vector="0,0,-980", ExternalForce="0,0,0") -> (Pose) {
    angularDampingOverride=0.999
    angularSpringConstant=5000
    bAngularSpring=true
    bOverrideAngularDamping=true
    bUseGravityOverride=true
    gravityScale=0.35
    simulationSpace=BoneRelative
    boundBone {boneName=dyn_backSkirt_01}
    componentAppliedLinearAccClamp=<0 0 0>
    gravityOverride {z=-980}
    physicsBodyDefinitions [{
      boundBone {boneName=dyn_backSkirt_01}
      boxExtents=<20 30 30>
      constraintSetup {angularLimitsMax=<0 0 5>, angularLimitsMin=<0 -10 -5>, angularTarget=<1 0 0>}
      localJointOffset=<4 0 0>
     }]
    relativeSpaceBone {boneName=root}
    simSpaceSettings {maxAngularAcceleration=20, maxAngularVelocity=5, simSpaceAngularAlpha=1}
   }
   n35 = Anim Dynamics - Bone: dyn_backSkirt_02!(ComponentPose=@n34, GravityOverride:Vector="0,0,-980", ExternalForce="0,0,0") -> (Pose) {
    angularDampingOverride=0.999
    angularSpringConstant=5000
    bAngularSpring=true
    bOverrideAngularDamping=true
    bUseGravityOverride=true
    gravityScale=0.35
    simulationSpace=BoneRelative
    boundBone {boneName=dyn_backSkirt_02}
    componentAppliedLinearAccClamp=<0 0 0>
    gravityOverride {z=-980}
    physicsBodyDefinitions [{
      boundBone {boneName=dyn_backSkirt_02}
      boxExtents=<20 30 30>
      constraintSetup {angularLimitsMax=<0 0 10>, angularLimitsMin=<0 -20 -10>, angularTarget=<1 0 0>}
      localJointOffset=<4 0 0>
     }]
    relativeSpaceBone {boneName=root}
    simSpaceSettings {maxAngularAcceleration=20, maxAngularVelocity=5, simSpaceAngularAlpha=1}
   }
   n36 = Anim Dynamics - Bone: dyn_pendant!(ComponentPose=@n35, GravityOverride:Vector="0,0,-980", ExternalForce="0,0,0") -> (Pose) {
    angularDampingOverride=0.9
    angularSpringConstant=5000
    bAngularSpring=true
    bOverrideAngularDamping=true
    bUseGravityOverride=true
    gravityScale=0.35
    simulationSpace=BoneRelative
    boundBone {boneName=dyn_pendant}
    componentAppliedLinearAccClamp=<0 0 0>
    gravityOverride {z=-980}
    physicsBodyDefinitions [{
      boundBone {boneName=dyn_pendant}
      boxExtents=<20 30 30>
      constraintSetup {angularLimitsMax=<5 20 0>, angularLimitsMin=<-5 -20 -30>, angularTarget=<1 0 0>}
      localJointOffset=<4 0 0>
     }]
    relativeSpaceBone {boneName=root}
    simSpaceSettings {maxAngularAcceleration=20, maxAngularVelocity=10, simSpaceAngularAlpha=1}
   }
   n37 = Anim Dynamics - Bone: dyn_star_A_01!(ComponentPose=@n36, GravityOverride:Vector="0,0,-980", ExternalForce="0,0,0") -> (Pose) {
    angularDampingOverride=0.9
    angularSpringConstant=5000
    bAngularSpring=true
    bOverrideAngularDamping=true
    bUseGravityOverride=true
    gravityScale=0.35
    simulationSpace=BoneRelative
    boundBone {boneName=dyn_star_A_01}
    componentAppliedLinearAccClamp=<0 0 0>
    gravityOverride {z=-980}
    physicsBodyDefinitions [{
      boundBone {boneName=dyn_star_A_01}
      boxExtents=<20 30 30>
      constraintSetup {angularLimitsMax=<5 30 0>, angularLimitsMin=<-5 -30 -30>, angularTarget=<1 0 0>}
      localJointOffset=<4 0 0>
     }]
    relativeSpaceBone {boneName=root}
    simSpaceSettings {maxAngularAcceleration=20, maxAngularVelocity=10, simSpaceAngularAlpha=1}
   }
   n38 = Anim Dynamics - Bone: dyn_star_A_02!(ComponentPose=@n37, GravityOverride:Vector="0,0,-980", ExternalForce="0,0,0") -> (Pose) {
    angularDampingOverride=0.9
    angularSpringConstant=5000
    bAngularSpring=true
    bOverrideAngularDamping=true
    bUseGravityOverride=true
    gravityScale=0.35
    simulationSpace=BoneRelative
    boundBone {boneName=dyn_star_A_02}
    componentAppliedLinearAccClamp=<0 0 0>
    gravityOverride {z=-980}
    physicsBodyDefinitions [{
      boundBone {boneName=dyn_star_A_02}
      boxExtents=<20 30 30>
      constraintSetup {angularLimitsMax=<5 60 30>, angularLimitsMin=<-5 -60 -30>, angularTarget=<1 0 0>}
      localJointOffset=<4 0 0>
     }]
    relativeSpaceBone {boneName=root}
    simSpaceSettings {maxAngularAcceleration=20, maxAngularVelocity=10, simSpaceAngularAlpha=1}
   }
   n39 = Anim Dynamics - Bone: dyn_star_A_03!(ComponentPose=@n38, GravityOverride:Vector="0,0,-980", ExternalForce="0,0,0") -> (Pose) {
    angularDampingOverride=0.9
    angularSpringConstant=5000
    bAngularSpring=true
    bOverrideAngularDamping=true
    bUseGravityOverride=true
    gravityScale=0.35
    simulationSpace=BoneRelative
    boundBone {boneName=dyn_star_A_03}
    componentAppliedLinearAccClamp=<0 0 0>
    gravityOverride {z=-980}
    physicsBodyDefinitions [{
      boundBone {boneName=dyn_star_A_03}
      boxExtents=<20 30 30>
      constraintSetup {angularLimitsMax=<5 60 30>, angularLimitsMin=<-5 -60 -30>, angularTarget=<1 0 0>}
      localJointOffset=<4 0 0>
     }]
    relativeSpaceBone {boneName=root}
    simSpaceSettings {maxAngularAcceleration=20, maxAngularVelocity=10, simSpaceAngularAlpha=1}
   }
   n40 = Anim Dynamics - Bone: dyn_star_B_01!(ComponentPose=@n39, GravityOverride:Vector="0,0,-980", ExternalForce="0,0,0") -> (Pose) {
    angularDampingOverride=0.9
    angularSpringConstant=5000
    bAngularSpring=true
    bOverrideAngularDamping=true
    bUseGravityOverride=true
    gravityScale=0.35
    simulationSpace=BoneRelative
    boundBone {boneName=dyn_star_B_01}
    componentAppliedLinearAccClamp=<0 0 0>
    gravityOverride {z=-980}
    physicsBodyDefinitions [{
      boundBone {boneName=dyn_star_B_01}
      boxExtents=<20 30 30>
      constraintSetup {angularLimitsMax=<5 30 0>, angularLimitsMin=<-5 -30 -30>, angularTarget=<1 0 0>}
      localJointOffset=<4 0 0>
     }]
    relativeSpaceBone {boneName=root}
    simSpaceSettings {maxAngularAcceleration=20, maxAngularVelocity=10, simSpaceAngularAlpha=1}
   }
   n41 = Anim Dynamics - Bone: dyn_star_B_02!(ComponentPose=@n40, GravityOverride:Vector="0,0,-980", ExternalForce="0,0,0") -> (Pose) {
    angularDampingOverride=0.9
    angularSpringConstant=5000
    bAngularSpring=true
    bOverrideAngularDamping=true
    bUseGravityOverride=true
    gravityScale=0.35
    simulationSpace=BoneRelative
    boundBone {boneName=dyn_star_B_02}
    componentAppliedLinearAccClamp=<0 0 0>
    gravityOverride {z=-980}
    physicsBodyDefinitions [{
      boundBone {boneName=dyn_star_B_02}
      boxExtents=<20 30 30>
      constraintSetup {angularLimitsMax=<5 60 30>, angularLimitsMin=<-5 -60 -30>, angularTarget=<1 0 0>}
      localJointOffset=<4 0 0>
     }]
    relativeSpaceBone {boneName=root}
    simSpaceSettings {maxAngularAcceleration=20, maxAngularVelocity=10, simSpaceAngularAlpha=1}
   }
   n42 = Anim Dynamics - Bone: dyn_star_B_03!(ComponentPose=@n41, GravityOverride:Vector="0,0,-980", ExternalForce="0,0,0") -> (Pose) {
    angularDampingOverride=0.9
    angularSpringConstant=5000
    bAngularSpring=true
    bOverrideAngularDamping=true
    bUseGravityOverride=true
    gravityScale=0.35
    simulationSpace=BoneRelative
    boundBone {boneName=dyn_star_B_03}
    componentAppliedLinearAccClamp=<0 0 0>
    gravityOverride {z=-980}
    physicsBodyDefinitions [{
      boundBone {boneName=dyn_star_B_03}
      boxExtents=<20 30 30>
      constraintSetup {angularLimitsMax=<5 60 30>, angularLimitsMin=<-5 -60 -30>, angularTarget=<1 0 0>}
      localJointOffset=<4 0 0>
     }]
    relativeSpaceBone {boneName=root}
    simSpaceSettings {maxAngularAcceleration=20, maxAngularVelocity=10, simSpaceAngularAlpha=1}
   }
   n43 = Anim Dynamics - Bone: dyn_star_B_04!(ComponentPose=@n42, GravityOverride:Vector="0,0,-980", ExternalForce="0,0,0") -> (Pose) {
    angularDampingOverride=0.9
    angularSpringConstant=5000
    bAngularSpring=true
    bOverrideAngularDamping=true
    bUseGravityOverride=true
    gravityScale=0.35
    simulationSpace=BoneRelative
    boundBone {boneName=dyn_star_B_04}
    componentAppliedLinearAccClamp=<0 0 0>
    gravityOverride {z=-980}
    physicsBodyDefinitions [{
      boundBone {boneName=dyn_star_B_04}
      boxExtents=<20 30 30>
      constraintSetup {angularLimitsMax=<5 60 30>, angularLimitsMin=<-5 -60 -30>, angularTarget=<1 0 0>}
      localJointOffset=<4 0 0>
     }]
    relativeSpaceBone {boneName=root}
    simSpaceSettings {maxAngularAcceleration=20, maxAngularVelocity=10, simSpaceAngularAlpha=1}
   }
   n44 = Anim Dynamics - Bone: dyn_star_B_05!(ComponentPose=@n43, GravityOverride:Vector="0,0,-980", ExternalForce="0,0,0") -> (Pose) {
    angularDampingOverride=0.9
    angularSpringConstant=5000
    bAngularSpring=true
    bOverrideAngularDamping=true
    bUseGravityOverride=true
    gravityScale=0.35
    simulationSpace=BoneRelative
    boundBone {boneName=dyn_star_B_05}
    componentAppliedLinearAccClamp=<0 0 0>
    gravityOverride {z=-980}
    physicsBodyDefinitions [{
      boundBone {boneName=dyn_star_B_05}
      boxExtents=<20 30 30>
      constraintSetup {angularLimitsMax=<5 60 30>, angularLimitsMin=<-5 -60 -30>, angularTarget=<1 0 0>}
      localJointOffset=<4 0 0>
     }]
    relativeSpaceBone {boneName=root}
    simSpaceSettings {maxAngularAcceleration=20, maxAngularVelocity=10, simSpaceAngularAlpha=1}
   }
   n45 = Anim Dynamics - Bone: dyn_star_B_06!(ComponentPose=@n44, GravityOverride:Vector="0,0,-980", ExternalForce="0,0,0") -> (Pose) {
    angularDampingOverride=0.9
    angularSpringConstant=5000
    bAngularSpring=true
    bOverrideAngularDamping=true
    bUseGravityOverride=true
    gravityScale=0.35
    simulationSpace=BoneRelative
    boundBone {boneName=dyn_star_B_06}
    componentAppliedLinearAccClamp=<0 0 0>
    gravityOverride {z=-980}
    physicsBodyDefinitions [{
      boundBone {boneName=dyn_star_B_06}
      boxExtents=<20 30 30>
      constraintSetup {angularLimitsMax=<5 60 30>, angularLimitsMin=<-5 -60 -30>, angularTarget=<1 0 0>}
      localJointOffset=<4 0 0>
     }]
    relativeSpaceBone {boneName=root}
    simSpaceSettings {maxAngularAcceleration=20, maxAngularVelocity=10, simSpaceAngularAlpha=1}
   }
   n46 = "Transform (Modify) Bone - Bone: ground"!(ComponentPose=@n45, bAlphaBoolEnabled:bool=True, Translation="0,0,0", Rotation="0,0,0", Scale="1,1,1") -> (Pose) {
    alphaInputType=Bool
    rotationMode=BMM_Replace
    rotationSpace=BCS_WorldSpace
    translationMode=BMM_Replace
    translationSpace=BCS_BoneSpace
    alphaBoolBlend {blendOption=Linear}
    boneToModify {boneName=ground}
   }
   Save cached pose 'CRCache_preDyn'(Pose=Component To Local(ComponentPose=@n46)) {cacheName=CRCache_preDyn}
  ]}
 EventGraph {nodes [
   n1 = Event BlueprintUpdateAnimation()
   > n2 = "Gravity Override State Function (Parameter Asset) (Nativized)"(
    CharacterPartAnimInstance=Self-Reference()
    GameTimeInSeconds=GetGameTimeInSeconds()
    Parameters=&/F_MED_FlintTwirl_Parameters
   )
   > Set Gravity Override(Gravity Override=@n2)
   > Set RBAN Settings Out(RBAN Settings Out=@n2.OutRigidBodyAnimNodeParameters)
   > Handle Slope Warp!(execute, self, DeltaTime:double=@n1.DeltaTimeX) -> (then)
   n3 = CheckForBackbling()
   n4 = AND Boolean(A=NOT Boolean(A=@n3.bIsBackblingACape), B=@n3)
   > Set is_wearing_backpack(is_wearing_backpack:double="To Float (Boolean)"(InBool=@n4)) -> (Output_Get:double)
   > n5 = Branch(Condition=AND Boolean(A=@n4, B=Get bIsSkyDiving()))
   > n6 = Set backpack_skydive(backpack_skydive:double=0.5) -> (Output_Get:double)
   n7 = Branch(execute=[@n15.then, @n6.then], Condition=AND Boolean!(A=@n3, B=@n3.bIsBackblingACape, C:bool=@n3.bIsBackblingLong) -> (ReturnValue))
   > n8 = Set RBAN alpha(RBAN alpha:double=0.25) -> (Output_Get:double)
   n9 = AND Boolean!(
    A=Get bIsInVehicle() -> (bIsInVehicle:bool)
    B=NOT Boolean(A=Get bIsGolfCartVehicleBackPassenger() -> (bIsGolfCartVehicleBackPassenger:bool))
    C=NOT Boolean(A=Get bVehicleSeatForcesCrouch() -> (bVehicleSeatForcesCrouch:bool))
    D:bool=NOT Boolean(A=Get bIsMotorcycleDriver() -> (bIsMotorcycleDriver:bool))
   ) -> (ReturnValue)
   Set is_in_car_seat(execute=[@n8.then, @n14.then], is_in_car_seat:bool=@n9) -> (Output_Get:bool)
   n10 = Get bIsMotorcyclePassenger() -> (bIsMotorcyclePassenger:bool)
   > n11 = Branch(Condition=AND Boolean!(A=@n9, B=NOT Boolean(A=@n10)) -> (ReturnValue))
   > n12 = Set RBAN alpha(RBAN alpha=0)
   n13 = Get is_wearing_backpack()
   Set backpack_not_skydive(execute=[@n11.else, @n12.then], backpack_not_skydive:double=float - float(A=float / float(A=@n13, B=2), B=Get backpack_skydive())) -> (Output_Get:double)
   n14 = Set RBAN alpha(execute=@n7.else, RBAN alpha=1)
   n15 = Set backpack_skydive(execute=@n5.else, backpack_skydive=0)
  ]}
 Handle Slope Warp {nodes [
   n1 = Handle Slope Warp!() -> (then, DeltaTime:double=~) {is=title K2Node_FunctionEntry}
   > n2 = Sequence()
   n3 = Get bIsOnGround() -> (bIsOnGround:bool)
   > n4 = Branch(Condition=OR Boolean!(A=NOT Boolean(A=@n3), B=Get bIsPlayingEmote(), C:bool=Get bIsInVehicle()) -> (ReturnValue))
   > Set Target Info Input Root Slope Translation(Target Info Input Root Slope Translation:Vector="0,0,-100") -> (Output_Get:Vector)
   Set Slope Info Input Slope Rotation(
    execute=@n4.else
    Slope Info Input Slope Rotation:Rotator=RInterpTo!(
     Current:Rotator=Get Slope Info Input Slope Rotation() -> (Slope Info Input Slope Rotation:Rotator)
     Target:Rotator=Get!() -> (
      SlopeInfoInput_RootSlopeTranslation:Vector=~
      SlopeInfoInput_LeftFootSlopeTranslation:Vector=~
      SlopeInfoInput_RightFootSlopeTranslation:Vector=~
      SlopeInfoInput_SlopeRotation:Rotator=~
     ).SlopeInfoInput_SlopeRotation
     DeltaTime:float=@n1
     InterpSpeed:float=7
    ) -> (ReturnValue:Rotator)
   ) -> (Output_Get:Rotator)
   > Set Target Info Input Root Slope Translation(Target Info Input Root Slope Translation="0, 0, 0")
   Reroute Node(InputPin=@n2.then_1)
   Reroute Node()
   > Set Slope Info Input Root Slope Translation(Slope Info Input Root Slope Translation:Vector=VInterpTo!(
     Current:Vector=Get Slope Info Input Root Slope Translation() -> (Slope Info Input Root Slope Translation:Vector)
     Target:Vector=Get Target Info Input Root Slope Translation() -> (Target Info Input Root Slope Translation:Vector)
     DeltaTime:float=@n1
     InterpSpeed:float=10
    ) -> (ReturnValue:Vector)) -> (Output_Get:Vector)
  ]}
]
variables [
 backpack_not_skydive {type=number}
 backpack_skydive {type=number}
 Gravity Override {type=Vector}
 is_in_car_seat {type=boolean}
 is_wearing_backpack {type=number}
 RBAN alpha {type=number}
 RBAN Settings Out {
  type=FortRigidBodyAnimNodeParameters
  default {simSpaceSettings {dampingAlpha=0, maxAngularAcceleration=16, maxAngularVelocity=4, maxLinearAcceleration=800, maxLinearVelocity=200, velocityScaleZ=0.5}}
 }
 Slope Info Input Root Slope Translation {type=Vector}
 Slope Info Input Slope Rotation {type=Rotator}
 Target Info Input Root Slope Translation {type=Vector}
]
