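# Physics asset of colliders for the cape on the medium female FlintTwirl player body, likely used by its cloth or secondary-motion simulation.
# Capsules cover the pelvis, parts of the spine, arms, hands, legs and feet, plus the back ribbon and sash dynamic bones.
# Constraints chain these bodies along the skeleton, with the ribbon and sash branching from MSTR_backRibbon under spine_02.
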PhysicsAsset /BRCosmetics/Characters/Player/Female/Medium/Bodies/F_MED_FlintTwirl/Meshes/F_MED_FlintTwirl_CapeColliders
bodies {
 MSTR_backRibbon [{shapeType=Capsule, length=25, radius=10, shapeName=MSTR_backRibbon_capsule, center=<-2 -3 0>}]
 calf_l [{shapeType=Capsule, length=28.6339, radius=11.064, center=<-27.6471 -0.0714 -0.9885>, rotation=<rot 0 79.183 -88.6776>}]
 calf_r [{shapeType=Capsule, length=28.6339, radius=11.064, center=<27.6471 0.0714 0.9885>, rotation=<rot 0 79.183 -88.6776>}]
 dyn_backRibbon_03_B_l [{shapeType=Capsule, length=10, radius=7, shapeName=dyn_backRibbon_03_B_l_capsule, center=<1 0 0>}]
 dyn_backRibbon_03_B_r [{shapeType=Capsule, length=10, radius=7, shapeName=dyn_backRibbon_03_B_r_capsule, center=<-1 0 0>}]
 dyn_backRibbon_04_B_l [{shapeType=Capsule, length=8, radius=7, shapeName=dyn_backRibbon_04_B_l_capsule}]
 dyn_backRibbon_04_B_r [{shapeType=Capsule, length=8, radius=7, shapeName=dyn_backRibbon_04_B_r_capsule}]
 dyn_backSash_03_l [{
   shapeType=Capsule
   length=22.2429
   radius=10.3698
   shapeName=dyn_backSash_03_l_capsule
   center=<12.0637 0.7388 -1.3077>
   rotation=<rot 0 94.146 -87.5547>
  }]
 dyn_backSash_03_r [{
   shapeType=Capsule
   length=21.9319
   radius=10.3645
   shapeName=dyn_backSash_03_r_capsule
   center=<-11.7578 -0.6987 1.3173>
   rotation=<rot 0 94.1006 -87.5327>
  }]
 dyn_backSash_04_l [{shapeType=Capsule, length=15, radius=10, shapeName=dyn_backSash_04_l_capsule, center=<22 0 0>, rotation=<rot 90 0 0>}]
 dyn_backSash_04_r [{shapeType=Capsule, length=15, radius=10, shapeName=dyn_backSash_04_r_capsule, center=<-22 0 0>, rotation=<rot 90 0 0>}]
 foot_l [{shapeType=Capsule, length=10.5, radius=7, shapeName=foot_l_capsule, center=<-3.92 5.601 0>, rotation=<rot 0 6.7594 90>}]
 foot_r [{shapeType=Capsule, length=10.5, radius=7, shapeName=foot_l_capsule, center=<3.92 -5.601 0>, rotation=<rot 0 6.7594 -90>}]
 hand_l [{shapeType=Capsule, length=8.7464, radius=5.6478, center=<8.2519 -2.6605 -0.069>, rotation=<rot 0 81.071 -71.8704>}]
 hand_r [{shapeType=Capsule, length=8.7464, radius=5.6478, center=<-8.2519 2.6605 0.069>, rotation=<rot 0 81.071 -71.8704>}]
 lowerarm_l [{shapeType=Capsule, length=14.2095, radius=7.8957, center=<12.2213 0.1704 -1.4491>, rotation=<rot 0 -91.2928 -87.4337>}]
 lowerarm_r [{shapeType=Capsule, length=14.2095, radius=7.8957, center=<-12.2213 -0.1704 1.4491>, rotation=<rot 0 -91.2928 -87.4337>}]
 pelvis [
  {shapeType=Capsule, length=13.7459, radius=25, center=<-17.6144 -10.2595 2.1935>, rotation=<rot 10.9044 -90 -1.326>}
  {shapeType=Capsule, length=13.7459, radius=25, shapeName=pelvis_capsule, center=<-39.8178 -16.0412 2.5482>, rotation=<rot 10.9044 -90 -1.326>}
 ]
 spine_01 [{shapeType=Capsule, length=10.8777, radius=20, shapeName=spine_01_capsule, center=<-2.2546 -2.8236 0>}]
 spine_02 [{shapeType=Capsule, length=13.0093, radius=9.4845, shapeName=spine_02_capsule, center=<6.5823 -2.2502 0>}]
 spine_05 [{shapeType=Capsule, length=15.0464, radius=9.6148, shapeName=spine_05_capsule, center=<2.1137 0.8943 0>}]
 thigh_l [{shapeType=Capsule, length=20.3977, radius=14.7105, center=<-27.582 -5.5485 1.5122>, rotation=<rot 0 -90.1171 93.1475>}]
 thigh_r [{shapeType=Capsule, length=20.3977, radius=14.7105, center=<27.582 5.5485 -1.5122>, rotation=<rot 0 -90.1171 -86.8525>}]
 upperarm_l [{shapeType=Capsule, length=13.6971, radius=5.8414, center=<13.0072 0.2824 -0.9397>, rotation=<rot 0 88.3156 -84.273>}]
 upperarm_r [{shapeType=Capsule, length=13.6967, radius=5.8411, center=<-13.0035 -0.2709 0.94>, rotation=<rot 0 88.5178 -84.275>}]
}
constraints [
 {
  bone1Name=spine_01
  bone2Name=pelvis
  swing1LimitDegrees=45
  swing1Motion=Free
  swing2LimitDegrees=45
  swing2Motion=Free
  twistLimitDegrees=45
  twistMotion=Free
 }
 {
  bone1Name=spine_02
  bone2Name=spine_01
  swing1LimitDegrees=45
  swing1Motion=Free
  swing2LimitDegrees=45
  swing2Motion=Free
  twistLimitDegrees=45
  twistMotion=Free
 }
 {
  bone1Name=MSTR_backRibbon
  bone2Name=spine_02
  swing1LimitDegrees=45
  swing1Motion=Free
  swing2LimitDegrees=45
  swing2Motion=Free
  twistLimitDegrees=45
  twistMotion=Free
 }
 {
  bone1Name=dyn_backRibbon_03_B_l
  bone2Name=MSTR_backRibbon
  swing1LimitDegrees=45
  swing1Motion=Free
  swing2LimitDegrees=45
  swing2Motion=Free
  twistLimitDegrees=45
  twistMotion=Free
 }
 {
  bone1Name=dyn_backRibbon_04_B_l
  bone2Name=dyn_backRibbon_03_B_l
  swing1LimitDegrees=45
  swing1Motion=Free
  swing2LimitDegrees=45
  swing2Motion=Free
  twistLimitDegrees=45
  twistMotion=Free
 }
 {
  bone1Name=dyn_backRibbon_03_B_r
  bone2Name=MSTR_backRibbon
  swing1LimitDegrees=45
  swing1Motion=Free
  swing2LimitDegrees=45
  swing2Motion=Free
  twistLimitDegrees=45
  twistMotion=Free
 }
 {
  bone1Name=dyn_backRibbon_04_B_r
  bone2Name=dyn_backRibbon_03_B_r
  swing1LimitDegrees=45
  swing1Motion=Free
  swing2LimitDegrees=45
  swing2Motion=Free
  twistLimitDegrees=45
  twistMotion=Free
 }
 {
  bone1Name=dyn_backSash_03_l
  bone2Name=MSTR_backRibbon
  swing1LimitDegrees=45
  swing1Motion=Free
  swing2LimitDegrees=45
  swing2Motion=Free
  twistLimitDegrees=45
  twistMotion=Free
 }
 {
  bone1Name=dyn_backSash_04_l
  bone2Name=dyn_backSash_03_l
  swing1LimitDegrees=45
  swing1Motion=Free
  swing2LimitDegrees=45
  swing2Motion=Free
  twistLimitDegrees=45
  twistMotion=Free
 }
 {
  bone1Name=dyn_backSash_03_r
  bone2Name=MSTR_backRibbon
  swing1LimitDegrees=45
  swing1Motion=Free
  swing2LimitDegrees=45
  swing2Motion=Free
  twistLimitDegrees=45
  twistMotion=Free
 }
 {
  bone1Name=dyn_backSash_04_r
  bone2Name=dyn_backSash_03_r
  swing1LimitDegrees=45
  swing1Motion=Free
  swing2LimitDegrees=45
  swing2Motion=Free
  twistLimitDegrees=45
  twistMotion=Free
 }
 {
  bone1Name=spine_05
  bone2Name=spine_02
  swing1LimitDegrees=45
  swing1Motion=Free
  swing2LimitDegrees=45
  swing2Motion=Free
  twistLimitDegrees=45
  twistMotion=Free
 }
 {bone1Name=upperarm_l, bone2Name=spine_05, swing1LimitDegrees=45, swing2LimitDegrees=45, twistLimitDegrees=45}
 {bone1Name=lowerarm_l, bone2Name=upperarm_l, swing1LimitDegrees=45, swing2LimitDegrees=45, twistLimitDegrees=45}
 {bone1Name=hand_l, bone2Name=lowerarm_l, swing1LimitDegrees=45, swing2LimitDegrees=45, twistLimitDegrees=45}
 {bone1Name=upperarm_r, bone2Name=spine_05, swing1LimitDegrees=45, swing2LimitDegrees=45, twistLimitDegrees=45}
 {bone1Name=lowerarm_r, bone2Name=upperarm_r, swing1LimitDegrees=45, swing2LimitDegrees=45, twistLimitDegrees=45}
 {bone1Name=hand_r, bone2Name=lowerarm_r, swing1LimitDegrees=45, swing2LimitDegrees=45, twistLimitDegrees=45}
 {
  bone1Name=thigh_l
  bone2Name=pelvis
  swing1LimitDegrees=45
  swing1Motion=Free
  swing2LimitDegrees=45
  swing2Motion=Free
  twistLimitDegrees=45
  twistMotion=Free
 }
 {bone1Name=calf_l, bone2Name=thigh_l, swing1LimitDegrees=45, swing2LimitDegrees=45, twistLimitDegrees=45}
 {
  bone1Name=foot_l
  bone2Name=calf_l
  swing1LimitDegrees=45
  swing1Motion=Free
  swing2LimitDegrees=45
  swing2Motion=Free
  twistLimitDegrees=45
  twistMotion=Free
 }
 {
  bone1Name=thigh_r
  bone2Name=pelvis
  swing1LimitDegrees=45
  swing1Motion=Free
  swing2LimitDegrees=45
  swing2Motion=Free
  twistLimitDegrees=45
  twistMotion=Free
 }
 {bone1Name=calf_r, bone2Name=thigh_r, swing1LimitDegrees=45, swing2LimitDegrees=45, twistLimitDegrees=45}
 {
  bone1Name=foot_r
  bone2Name=calf_r
  swing1LimitDegrees=45
  swing1Motion=Free
  swing2LimitDegrees=45
  swing2Motion=Free
  twistLimitDegrees=45
  twistMotion=Free
 }
]
