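# Physics asset for the FlintTwirl female medium character's face accessory part, covering hair bangs, strands, ponytail, hair ribbons and hat sash.
# Capsules on the head, neck, upper spine, clavicles and upper arms form the body volume, joined by free constraints.
# MSTR root bones anchor chains of dyn_ capsules with mostly limited swing and locked or narrow twist.
# Most roots attach freely to the head, but the front strand root locks onto the center bang.
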PhysicsAsset /BRCosmetics/Characters/Player/Female/Medium/Bodies/F_MED_FlintTwirl/Meshes/Parts/F_MED_FlintTwirl_FaceAcc_Physics
bodies {
 MSTR_bangs_c [{shapeType=Capsule, length=0.505, radius=0.505, shapeName=MSTR_bangs_c_capsule, rotation=<rot 0 -90 0>}]
 MSTR_bangs_l [{shapeType=Capsule, length=0.505, radius=0.505, shapeName=MSTR_bangs_l_capsule, rotation=<rot 0 -90 0>}]
 MSTR_bangs_r [{shapeType=Capsule, length=0.1, radius=0.1, shapeName=MSTR_bangs_r_capsule, rotation=<rot 0 -90 0>}]
 MSTR_hairRibbon_C_l [{
   shapeType=Capsule
   length=1.4523
   radius=1.1427
   shapeName=MSTR_hairRibbon_C_l_capsule
   center=<-0.4858 0.0199 -0.7503>
   rotation=<rot 0 30.4602 -64.8225>
  }]
 MSTR_hairRibbon_C_r [{
   shapeType=Capsule
   length=1.8715
   radius=1.1819
   shapeName=MSTR_hairRibbon_C_r_capsule
   center=<-0.5894 -0.1686 -0.8085>
   rotation=<rot 0 144.1617 -65.2796>
  }]
 MSTR_hatSash [{shapeType=Capsule, length=1.1555, radius=1, shapeName=MSTR_hatSash_capsule, rotation=<rot 0 0 90>}]
 MSTR_ponytail [{shapeType=Capsule, length=5, radius=1, shapeName=MSTR_ponytail_capsule}]
 MSTR_sidebangs_l [{shapeType=Capsule, length=2, radius=1, shapeName=MSTR_sidebangs_l_capsule, rotation=<rot 0 0 90>}]
 MSTR_strandFront_c [{shapeType=Capsule, length=0.505, radius=0.505, shapeName=MSTR_strandFront_c_capsule, rotation=<rot 0 -90 0>}]
 MSTR_strand_r [{shapeType=Capsule, length=0.505, radius=0.505, shapeName=MSTR_strand_r_capsule, rotation=<rot 0 -90 0>}]
 clavicle_l [{shapeType=Capsule, length=8, radius=5.5, shapeName=clavicle_l_capsule, center=<5.9389 -0.3809 -0.9669>, rotation=<rot 79.0229 -0.001 -0.001>}]
 clavicle_r [{
   shapeType=Capsule
   length=7.7333
   radius=6.2666
   shapeName=clavicle_r_capsule
   center=<-4.936 1.252 0.675>
   rotation=<rot 81.8254 -0.0003 -0.0003>
  }]
 dyn_bangs_01_c [{shapeType=Capsule, length=4.1691, radius=2, shapeName=dyn_bangs_01_c_capsule, center=<6.4181 -0.4396 0.529>, rotation=<rot 90 -98.6008 0>}]
 dyn_bangs_01_l [{shapeType=Capsule, length=2, radius=2, shapeName=dyn_bangs_01_l_capsule, center=<3.2459 0 0>, rotation=<rot 70.9622 90.0648 -179.9999>}]
 dyn_bangs_01_r [{shapeType=Capsule, length=3.7063, radius=2, shapeName=dyn_bangs_01_r_capsule, center=<4.0511 0 0>, rotation=<rot 79.2681 90.0003 -179.9997>}]
 dyn_bangs_02_c [{shapeType=Capsule, length=4.1691, radius=2, shapeName=dyn_bangs_02_c_capsule, center=<7.6592 -1.5106 -0.029>, rotation=<rot 90 -89.545 0>}]
 dyn_bangs_02_l [{
   shapeType=Capsule
   length=3.4034
   radius=2
   shapeName=dyn_bangs_02_l_capsule
   center=<3.1925 -0.1421 -0.3157>
   rotation=<rot 71.9827 -4.6153 9.6388>
  }]
 dyn_bangs_02_r [{
   shapeType=Capsule
   length=4.3878
   radius=2
   shapeName=dyn_bangs_02_r_capsule
   center=<4.7158 0.4529 0.9313>
   rotation=<rot 84.7691 90.0443 -179.9506>
  }]
 dyn_bangs_03_r [{
   shapeType=Capsule
   length=2.5863
   radius=2
   shapeName=dyn_bangs_03_r_capsule
   center=<6.2514 -0.9306 -1.4825>
   rotation=<rot 76.4362 90.0533 -179.9991>
  }]
 dyn_hairRibbon_01_C_l [{
   shapeType=Capsule
   length=5.3366
   radius=2.6918
   shapeName=dyn_hairRibbon_01_C_l_capsule
   center=<3.1408 -0.6679 1.7745>
   rotation=<rot -0.0744 92.6315 -84.3148>
  }]
 dyn_hairRibbon_01_C_r [{
   shapeType=Capsule
   length=3.589
   radius=2.6519
   shapeName=dyn_hairRibbon_01_C_r_capsule
   center=<3.4793 0.5947 1.43>
   rotation=<rot 1.2123 88.5077 -86.4604>
  }]
 dyn_hairRibbon_02_C_l [{
   shapeType=Capsule
   length=5.2404
   radius=2.4106
   shapeName=dyn_hairRibbon_02_C_l_capsule
   center=<6.1448 0.3033 0.8423>
   rotation=<rot 0 -91.9543 -89.5649>
  }]
 dyn_hairRibbon_02_C_r [{
   shapeType=Capsule
   length=5.1602
   radius=2.5595
   shapeName=dyn_hairRibbon_02_C_r_capsule
   center=<4.9882 -0.0597 0.7015>
   rotation=<rot 0 -91.1007 -87.5724>
  }]
 dyn_hairRibbon_03_C_l [{shapeType=Capsule, length=2.733, radius=2.378, shapeName=dyn_hairRibbon_03_C_l_capsule, center=<1.9404 0.2194 0>, rotation=<rot 90 0 0>}]
 dyn_hairRibbon_03_C_r [{
   shapeType=Capsule
   length=3.0555
   radius=2.6543
   shapeName=dyn_hairRibbon_03_C_r_capsule
   center=<0.9988 0 0.35>
   rotation=<rot 86.361 0.0001 0.0002>
  }]
 dyn_hatSash_01 [{shapeType=Capsule, length=2.0634, radius=2, shapeName=dyn_hatSash_01_capsule, center=<4.5514 0 1.1764>, rotation=<rot 0 0 90>}]
 dyn_hatSash_02 [{
   shapeType=Capsule
   length=4.3689
   radius=3.1022
   shapeName=dyn_hatSash_02_capsule
   center=<6.3294 -0.0483 0.7644>
   rotation=<rot 0 -43.7207 90>
  }]
 dyn_ponytail_01 [{shapeType=Capsule, length=5, radius=2, shapeName=dyn_ponytail_01_capsule, rotation=<rot 52.0086 179.9997 179.9999>}]
 dyn_ponytail_02 [{shapeType=Capsule, length=6.0677, radius=2.3202, shapeName=dyn_ponytail_02_capsule, center=<1.5137 0 0>, rotation=<rot 90 0 0>}]
 dyn_ponytail_03 [{
   shapeType=Capsule
   length=4.4396
   radius=2.5604
   shapeName=dyn_ponytail_03_capsule
   center=<1.5884 0.7564 0.2122>
   rotation=<rot 82.5237 -0.0002 -0.0002>
  }]
 dyn_ponytail_04 [{
   shapeType=Capsule
   length=3.4155
   radius=2.6245
   shapeName=dyn_ponytail_04_capsule
   center=<1.7954 -1.2322 -0.2482>
   rotation=<rot 90 -5.2809 0>
  }]
 dyn_ponytail_05 [{shapeType=Capsule, length=5.434, radius=3.2169, shapeName=dyn_ponytail_05_capsule, center=<3.6125 0.0422 -0.2687>, rotation=<rot 90 3.79 0>}]
 dyn_sidebangs_01_l [{shapeType=Capsule, length=3.6013, radius=2, shapeName=dyn_sidebangs_01_l_capsule, center=<1.4477 0 0>, rotation=<rot 90 0 0>}]
 dyn_sidebangs_02_l [{shapeType=Capsule, length=3.4197, radius=2.346, shapeName=dyn_sidebangs_02_l_capsule, center=<2.5568 0 0.7283>, rotation=<rot 90 10.3003 0>}]
 dyn_sidebangs_03_l [{shapeType=Capsule, length=5.5717, radius=2.738, shapeName=dyn_sidebangs_03_l_capsule, center=<2.9914 0.9848 0.836>, rotation=<rot 90 0 0>}]
 dyn_sidebangs_04_l [{
   shapeType=Capsule
   length=5.6389
   radius=2.742
   shapeName=dyn_sidebangs_04_l_capsule
   center=<2.0249 -1.4747 0.5504>
   rotation=<rot 90 -5.9031 0>
  }]
 dyn_strandFront_01_c [{
   shapeType=Capsule
   length=1.7758
   radius=0.6959
   shapeName=dyn_strandFront_01_c_capsule
   center=<2.391 0.2884 0.3231>
   rotation=<rot 0 -100.2678 -81.7451>
  }]
 dyn_strandFront_02_c [{
   shapeType=Capsule
   length=2.3071
   radius=0.505
   shapeName=dyn_strandFront_02_c_capsule
   center=<2.5792 -1.5041 -0.9222>
   rotation=<rot 0 -136.6158 -69.4575>
  }]
 dyn_strand_01_r [{shapeType=Capsule, length=4.9666, radius=1.5149, shapeName=dyn_strand_01_r_capsule, center=<3.7764 0 0>, rotation=<rot 90 0 0>}]
 dyn_strand_02_r [{shapeType=Capsule, length=3.2543, radius=1.4353, shapeName=dyn_strand_02_r_capsule, center=<2.9885 -0.7208 -1.2954>, rotation=<rot 90 0 0>}]
 dyn_strand_03_r [{
   shapeType=Capsule
   length=2.5487
   radius=1.3933
   shapeName=dyn_strand_03_r_capsule
   center=<1.4989 -0.2958 0.5208>
   rotation=<rot 69.1262 -180 -180>
  }]
 head [
  {shapeType=Capsule, length=3.4213, radius=8.1577, center=<5.5615 1.9908 0>, rotation=<rot 90 0 0>}
  {shapeType=Capsule, length=4.3696, radius=5, shapeName=head_capsule, center=<7.3033 -3.01 4.5808>, rotation=<rot 90 90 0>}
  {shapeType=Capsule, length=4.3696, radius=5, shapeName=head_capsule, center=<4.2335 -5.9325 -2.9312>, rotation=<rot 90 0 0>}
 ]
 neck_01 [{shapeType=Capsule, length=0.356, radius=4, shapeName=neck_01_capsule, rotation=<rot 90 0 0>}]
 spine_05 [
  {shapeType=Capsule, length=10.4937, radius=6.9, shapeName=spine_05_capsule, center=<2.7725 -1.0365 0>}
  {shapeType=Capsule, length=4.0702, radius=4.8051, shapeName=spine_05_capsule, center=<8.5644 -0.1062 0>}
  {shapeType=Capsule, length=16.8634, radius=10, shapeName=spine_05_capsule, center=<11.8771 -15.0449 5.2007>, rotation=<rot 0 -0.001 -48.1087>}
  {shapeType=Capsule, length=19.5753, radius=10.3946, shapeName=spine_05_capsule, center=<30.5002 -15.4905 0.908>, rotation=<rot 90 0 0>}
 ]
 upperarm_FA_extra_r [{shapeType=Capsule, length=10, radius=4, shapeName=upperarm_FA_extra_r_capsule, center=<-10.4568 -7.3577 2.2487>, rotation=<rot 90 0 0>}]
 upperarm_l [{shapeType=Capsule, length=13, radius=4.8, shapeName=upperarm_l_capsule, center=<7.3976 0 -0.5911>, rotation=<rot 90 0 0>}]
 upperarm_r [{shapeType=Capsule, length=25.3832, radius=5.5, shapeName=upperarm_r_capsule, center=<-15.8116 -0.6638 1.4082>, rotation=<rot 90 0 0>}]
}
constraints [
 {
  bone1Name=neck_01
  bone2Name=spine_05
  swing1LimitDegrees=45
  swing1Motion=Free
  swing2LimitDegrees=45
  swing2Motion=Free
  twistLimitDegrees=45
  twistMotion=Free
 }
 {
  bone1Name=clavicle_l
  bone2Name=spine_05
  swing1LimitDegrees=45
  swing1Motion=Free
  swing2LimitDegrees=45
  swing2Motion=Free
  twistLimitDegrees=45
  twistMotion=Free
 }
 {
  bone1Name=clavicle_r
  bone2Name=spine_05
  swing1LimitDegrees=45
  swing1Motion=Free
  swing2LimitDegrees=45
  swing2Motion=Free
  twistLimitDegrees=45
  twistMotion=Free
 }
 {
  bone1Name=upperarm_l
  bone2Name=clavicle_l
  swing1LimitDegrees=45
  swing1Motion=Free
  swing2LimitDegrees=45
  swing2Motion=Free
  twistLimitDegrees=45
  twistMotion=Free
 }
 {
  bone1Name=upperarm_r
  bone2Name=clavicle_r
  swing1LimitDegrees=45
  swing1Motion=Free
  swing2LimitDegrees=45
  swing2Motion=Free
  twistLimitDegrees=45
  twistMotion=Free
 }
 {
  bone1Name=head
  bone2Name=neck_01
  swing1LimitDegrees=45
  swing1Motion=Free
  swing2LimitDegrees=45
  swing2Motion=Free
  twistLimitDegrees=45
  twistMotion=Free
 }
 {
  bone1Name=MSTR_hatSash
  bone2Name=head
  swing1LimitDegrees=45
  swing1Motion=Free
  swing2LimitDegrees=45
  swing2Motion=Free
  twistLimitDegrees=45
  twistMotion=Free
 }
 {
  bone1Name=MSTR_ponytail
  bone2Name=head
  swing1LimitDegrees=45
  swing1Motion=Free
  swing2LimitDegrees=45
  swing2Motion=Free
  twistLimitDegrees=45
  twistMotion=Free
 }
 {
  bone1Name=MSTR_sidebangs_l
  bone2Name=head
  swing1LimitDegrees=45
  swing1Motion=Free
  swing2LimitDegrees=45
  swing2Motion=Free
  twistLimitDegrees=45
  twistMotion=Free
 }
 {
  bone1Name=MSTR_hairRibbon_C_l
  bone2Name=head
  swing1LimitDegrees=45
  swing1Motion=Free
  swing2LimitDegrees=45
  swing2Motion=Free
  twistLimitDegrees=45
  twistMotion=Free
 }
 {
  bone1Name=MSTR_hairRibbon_C_r
  bone2Name=head
  swing1LimitDegrees=45
  swing1Motion=Free
  swing2LimitDegrees=45
  swing2Motion=Free
  twistLimitDegrees=45
  twistMotion=Free
 }
 {
  bone1Name=MSTR_strand_r
  bone2Name=head
  swing1LimitDegrees=45
  swing1Motion=Free
  swing2LimitDegrees=45
  swing2Motion=Free
  twistLimitDegrees=45
  twistMotion=Free
 }
 {
  bone1Name=MSTR_bangs_r
  bone2Name=head
  swing1LimitDegrees=45
  swing1Motion=Free
  swing2LimitDegrees=45
  swing2Motion=Free
  twistLimitDegrees=45
  twistMotion=Free
 }
 {
  bone1Name=MSTR_bangs_c
  bone2Name=head
  swing1LimitDegrees=45
  swing1Motion=Free
  swing2LimitDegrees=45
  swing2Motion=Free
  twistLimitDegrees=45
  twistMotion=Free
 }
 {
  bone1Name=MSTR_bangs_l
  bone2Name=head
  swing1LimitDegrees=45
  swing1Motion=Free
  swing2LimitDegrees=45
  swing2Motion=Free
  twistLimitDegrees=45
  twistMotion=Free
 }
 {bone1Name=dyn_bangs_01_c, bone2Name=MSTR_bangs_c, swing1LimitDegrees=5, swing1Motion=Locked, swing2LimitDegrees=10, twistLimitDegrees=5}
 {
  bone1Name=MSTR_strandFront_c
  bone2Name=dyn_bangs_01_c
  swing1LimitDegrees=45
  swing1Motion=Locked
  swing2LimitDegrees=45
  swing2Motion=Locked
  twistLimitDegrees=45
  twistMotion=Locked
 }
 {bone1Name=dyn_bangs_02_c, bone2Name=dyn_bangs_01_c, swing1LimitDegrees=7, swing2LimitDegrees=10, twistLimitDegrees=5}
 {
  bone1Name=dyn_strandFront_01_c
  bone2Name=MSTR_strandFront_c
  swing1LimitDegrees=5
  swing2LimitDegrees=5
  twistLimitDegrees=45
  twistMotion=Locked
 }
 {
  bone1Name=dyn_strandFront_02_c
  bone2Name=dyn_strandFront_01_c
  swing1LimitDegrees=5
  swing2LimitDegrees=5
  twistLimitDegrees=45
  twistMotion=Locked
 }
 {
  bone1Name=dyn_bangs_01_l
  bone2Name=MSTR_bangs_l
  swing1LimitDegrees=3
  swing1Motion=Locked
  swing2LimitDegrees=3
  twistLimitDegrees=45
  twistMotion=Locked
 }
 {bone1Name=dyn_bangs_02_l, bone2Name=dyn_bangs_01_l, swing1LimitDegrees=4, swing2LimitDegrees=7, twistLimitDegrees=45, twistMotion=Locked}
 {
  bone1Name=dyn_bangs_01_r
  bone2Name=MSTR_bangs_r
  swing1LimitDegrees=1
  swing1Motion=Locked
  swing2LimitDegrees=3
  swing2Motion=Locked
  twistLimitDegrees=45
  twistMotion=Locked
 }
 {bone1Name=dyn_bangs_02_r, bone2Name=dyn_bangs_01_r, swing1LimitDegrees=3, swing2LimitDegrees=3, twistLimitDegrees=45, twistMotion=Locked}
 {bone1Name=dyn_bangs_03_r, bone2Name=dyn_bangs_02_r, swing1LimitDegrees=3, swing2LimitDegrees=4, twistLimitDegrees=45, twistMotion=Locked}
 {
  bone1Name=dyn_hairRibbon_01_C_l
  bone2Name=MSTR_hairRibbon_C_l
  swing1LimitDegrees=15
  swing2LimitDegrees=10
  twistLimitDegrees=45
  twistMotion=Locked
 }
 {bone1Name=dyn_hairRibbon_02_C_l, bone2Name=dyn_hairRibbon_01_C_l, swing1LimitDegrees=15, swing2LimitDegrees=20, twistLimitDegrees=5}
 {bone1Name=dyn_hairRibbon_03_C_l, bone2Name=dyn_hairRibbon_02_C_l, swing1LimitDegrees=10, swing2LimitDegrees=10, twistLimitDegrees=5}
 {
  bone1Name=dyn_hairRibbon_01_C_r
  bone2Name=MSTR_hairRibbon_C_r
  swing1LimitDegrees=17
  swing2LimitDegrees=22
  twistLimitDegrees=45
  twistMotion=Locked
 }
 {bone1Name=dyn_hairRibbon_02_C_r, bone2Name=dyn_hairRibbon_01_C_r, swing1LimitDegrees=25, swing2LimitDegrees=35, twistLimitDegrees=2.5}
 {bone1Name=dyn_hairRibbon_03_C_r, bone2Name=dyn_hairRibbon_02_C_r, swing1LimitDegrees=20, swing2LimitDegrees=20, twistLimitDegrees=5}
 {bone1Name=dyn_hatSash_01, bone2Name=MSTR_hatSash, swing1LimitDegrees=55, swing2LimitDegrees=25, twistLimitDegrees=5}
 {bone1Name=dyn_hatSash_02, bone2Name=dyn_hatSash_01, swing1LimitDegrees=55, swing2LimitDegrees=25, twistLimitDegrees=10}
 {bone1Name=dyn_ponytail_01, bone2Name=MSTR_ponytail, swing1LimitDegrees=15, swing2LimitDegrees=10, twistLimitDegrees=3, twistMotion=Locked}
 {bone1Name=dyn_ponytail_02, bone2Name=dyn_ponytail_01, swing1LimitDegrees=30, swing2LimitDegrees=10, twistLimitDegrees=3}
 {bone1Name=dyn_ponytail_03, bone2Name=dyn_ponytail_02, swing1LimitDegrees=30, swing2LimitDegrees=40, twistLimitDegrees=3}
 {bone1Name=dyn_ponytail_04, bone2Name=dyn_ponytail_03, swing1LimitDegrees=35, swing2LimitDegrees=25, twistLimitDegrees=3}
 {bone1Name=dyn_ponytail_05, bone2Name=dyn_ponytail_04, swing1LimitDegrees=35, swing2LimitDegrees=25, twistLimitDegrees=3}
 {bone1Name=dyn_sidebangs_01_l, bone2Name=MSTR_sidebangs_l, swing1LimitDegrees=12, swing2LimitDegrees=20, twistLimitDegrees=0}
 {bone1Name=dyn_sidebangs_02_l, bone2Name=dyn_sidebangs_01_l, swing1LimitDegrees=16, swing2LimitDegrees=25, twistLimitDegrees=0}
 {bone1Name=dyn_sidebangs_03_l, bone2Name=dyn_sidebangs_02_l, swing1LimitDegrees=20, swing2LimitDegrees=25, twistLimitDegrees=0}
 {bone1Name=dyn_sidebangs_04_l, bone2Name=dyn_sidebangs_03_l, swing1LimitDegrees=20, swing2LimitDegrees=25, twistLimitDegrees=0}
 {bone1Name=dyn_strand_01_r, bone2Name=MSTR_strand_r, swing1LimitDegrees=15, swing2LimitDegrees=20, twistLimitDegrees=5, twistMotion=Locked}
 {bone1Name=dyn_strand_02_r, bone2Name=dyn_strand_01_r, swing1LimitDegrees=15, swing2LimitDegrees=20, twistLimitDegrees=5}
 {bone1Name=dyn_strand_03_r, bone2Name=dyn_strand_02_r, swing1LimitDegrees=20, swing2LimitDegrees=20, twistLimitDegrees=5}
 {
  bone1Name=upperarm_FA_extra_r
  bone2Name=upperarm_r
  swing1LimitDegrees=45
  swing1Motion=Free
  swing2LimitDegrees=45
  swing2Motion=Free
  twistLimitDegrees=45
  twistMotion=Free
 }
]
