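# Physics asset for the emote variant of the FlintTwirl female character's face accessory part, covering hair ribbons, hat sash, ponytail, left sidebangs and right strand.
# Capsule bodies on the upper spine, neck, head, clavicles and upper arms form the body volume around these pieces.
# Each piece has an MSTR root linked to the head by a free-rotation constraint, feeding a dyn_ bone chain with tight swing and twist limits, often twist-locked at the first link.
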PhysicsAsset /BRCosmetics/Characters/Player/Female/Medium/Bodies/F_MED_FlintTwirl/Meshes/Parts/F_MED_FlintTwirl_FaceAcc_Physics_Emote
bodies {
 MSTR_hairRibbon_C_l [{
   shapeType=Capsule
   length=1.4523
   radius=1.1427
   shapeName=MSTR_hairRibbon_C_l_capsule
   center=<-0.4858 0.0199 -0.7503>
   rotation=<rot 0 30.4602 -64.8225>
  }]
 MSTR_hairRibbon_C_r [{
   shapeType=Capsule
   length=1.8715
   radius=1.1819
   shapeName=MSTR_hairRibbon_C_r_capsule
   center=<-0.5894 -0.1686 -0.8085>
   rotation=<rot 0 144.1617 -65.2796>
  }]
 MSTR_hatSash [{shapeType=Capsule, length=1.1555, radius=1, shapeName=MSTR_hatSash_capsule, rotation=<rot 0 0 90>}]
 MSTR_ponytail [{shapeType=Capsule, length=5, radius=1, shapeName=MSTR_ponytail_capsule}]
 MSTR_sidebangs_l [{shapeType=Capsule, length=2, radius=1, shapeName=MSTR_sidebangs_l_capsule, rotation=<rot 0 0 90>}]
 MSTR_strand_r [{shapeType=Capsule, length=0.505, radius=0.505, shapeName=MSTR_strand_r_capsule, rotation=<rot 0 -90 0>}]
 clavicle_l [{shapeType=Capsule, length=8, radius=5.5, shapeName=clavicle_l_capsule, center=<5.9389 -0.3809 -0.9669>, rotation=<rot 79.0229 -0.001 -0.001>}]
 clavicle_r [{shapeType=Capsule, length=8, radius=6, shapeName=clavicle_r_capsule, center=<-4.8983 0.8939 0.9436>, rotation=<rot 81.8254 -0.0003 -0.0003>}]
 dyn_hairRibbon_01_C_l [{
   shapeType=Capsule
   length=5.3366
   radius=2.6918
   shapeName=dyn_hairRibbon_01_C_l_capsule
   center=<3.1408 -0.6679 1.7745>
   rotation=<rot -0.0744 92.6315 -84.3148>
  }]
 dyn_hairRibbon_01_C_r [{
   shapeType=Capsule
   length=3.589
   radius=2.6519
   shapeName=dyn_hairRibbon_01_C_r_capsule
   center=<3.4793 0.5947 1.43>
   rotation=<rot 1.2123 88.5077 -86.4604>
  }]
 dyn_hairRibbon_02_C_l [{
   shapeType=Capsule
   length=5.2699
   radius=2.2696
   shapeName=dyn_hairRibbon_02_C_l_capsule
   center=<4.9816 -0.0858 0.851>
   rotation=<rot 0 -91.9543 -89.5649>
  }]
 dyn_hairRibbon_02_C_r [{
   shapeType=Capsule
   length=5.1613
   radius=2.3682
   shapeName=dyn_hairRibbon_02_C_r_capsule
   center=<5.1587 -0.063 0.55>
   rotation=<rot 0 -91.1007 -87.5724>
  }]
 dyn_hatSash_01 [{shapeType=Capsule, length=2.0634, radius=2, shapeName=dyn_hatSash_01_capsule, center=<4.5514 0 1.1764>, rotation=<rot 0 0 90>}]
 dyn_hatSash_02 [{
   shapeType=Capsule
   length=3.7547
   radius=3.1022
   shapeName=dyn_hatSash_02_capsule
   center=<6.3469 -0.3638 0.7644>
   rotation=<rot 0.0001 2.0399 90>
  }]
 dyn_ponytail_01 [{shapeType=Capsule, length=5, radius=2, shapeName=dyn_ponytail_01_capsule, rotation=<rot 52.0086 179.9997 179.9999>}]
 dyn_ponytail_02 [{shapeType=Capsule, length=6.0677, radius=2.3202, shapeName=dyn_ponytail_02_capsule, center=<1.5137 0 0>, rotation=<rot 90 0 0>}]
 dyn_ponytail_03 [{
   shapeType=Capsule
   length=4.5306
   radius=2.4694
   shapeName=dyn_ponytail_03_capsule
   center=<1.5884 0 0.2122>
   rotation=<rot 82.5237 -0.0002 -0.0002>
  }]
 dyn_ponytail_04 [{
   shapeType=Capsule
   length=3.4155
   radius=2.6245
   shapeName=dyn_ponytail_04_capsule
   center=<1.7954 -1.2322 -0.2482>
   rotation=<rot 90 -5.2809 0>
  }]
 dyn_ponytail_05 [{shapeType=Capsule, length=5.4923, radius=3.1586, shapeName=dyn_ponytail_05_capsule, center=<3.6183 0 0>, rotation=<rot 90 0 0>}]
 dyn_sidebangs_01_l [{shapeType=Capsule, length=3.6013, radius=2, shapeName=dyn_sidebangs_01_l_capsule, center=<1.4477 0 0>, rotation=<rot 90 0 0>}]
 dyn_sidebangs_02_l [{shapeType=Capsule, length=3.4197, radius=2.346, shapeName=dyn_sidebangs_02_l_capsule, center=<2.5568 0 0.7283>, rotation=<rot 90 10.3003 0>}]
 dyn_sidebangs_03_l [{shapeType=Capsule, length=5.5717, radius=2.738, shapeName=dyn_sidebangs_03_l_capsule, center=<2.9914 0.9848 0.836>, rotation=<rot 90 0 0>}]
 dyn_sidebangs_04_l [{
   shapeType=Capsule
   length=5.6389
   radius=2.742
   shapeName=dyn_sidebangs_04_l_capsule
   center=<2.0249 -1.4747 0.5504>
   rotation=<rot 90 -5.9031 0>
  }]
 dyn_strand_01_r [{shapeType=Capsule, length=4.9666, radius=1.5149, shapeName=dyn_strand_01_r_capsule, center=<3.7764 0 0>, rotation=<rot 90 0 0>}]
 dyn_strand_02_r [{shapeType=Capsule, length=2.8397, radius=1.8499, shapeName=dyn_strand_02_r_capsule, center=<2.9885 0 -0.5825>, rotation=<rot 90 0 0>}]
 dyn_strand_03_r [{
   shapeType=Capsule
   length=2.5487
   radius=1.3933
   shapeName=dyn_strand_03_r_capsule
   center=<1.4989 -0.2958 0.5208>
   rotation=<rot 69.1262 -180 -180>
  }]
 head [
  {shapeType=Capsule, length=7.7305, radius=7.638, center=<5.5615 0 0>, rotation=<rot 90 0 0>}
  {shapeType=Capsule, length=4.3696, radius=5, shapeName=head_capsule, center=<0 4.2625 -0.9702>, rotation=<rot 90 0 0>}
  {shapeType=Capsule, length=4.3696, radius=5, shapeName=head_capsule, center=<7.3033 -3.01 4.5808>, rotation=<rot 90 90 0>}
  {shapeType=Capsule, length=4.3696, radius=5, shapeName=head_capsule, center=<4.2335 -5.9325 -2.9312>, rotation=<rot 90 0 0>}
 ]
 neck_01 [{shapeType=Capsule, length=0.356, radius=4, shapeName=neck_01_capsule, rotation=<rot 90 0 0>}]
 neck_02 [{shapeType=Capsule, length=0.356, radius=4, shapeName=neck_02_capsule, rotation=<rot 87.6208 179.9999 180>}]
 spine_05 [
  {shapeType=Capsule, length=10.4937, radius=6.9474, shapeName=spine_05_capsule, center=<2.7725 -1.0365 0>}
  {shapeType=Capsule, length=4.0702, radius=4.8051, shapeName=spine_05_capsule, center=<8.5644 -0.1062 0>}
 ]
 upperarm_l [{shapeType=Capsule, length=13, radius=4.8, shapeName=upperarm_l_capsule, center=<7.3976 0 -0.5911>, rotation=<rot 90 0 0>}]
 upperarm_r [{shapeType=Capsule, length=13, radius=5.5, shapeName=upperarm_r_capsule, center=<-7.3976 2 0>, rotation=<rot 90 0 0>}]
}
constraints [
 {bone1Name=dyn_strand_03_r, bone2Name=dyn_strand_02_r, swing1LimitDegrees=20, swing2LimitDegrees=20, twistLimitDegrees=5}
 {
  bone1Name=neck_01
  bone2Name=spine_05
  swing1LimitDegrees=45
  swing1Motion=Free
  swing2LimitDegrees=45
  swing2Motion=Free
  twistLimitDegrees=45
  twistMotion=Free
 }
 {
  bone1Name=clavicle_l
  bone2Name=spine_05
  swing1LimitDegrees=45
  swing1Motion=Free
  swing2LimitDegrees=45
  swing2Motion=Free
  twistLimitDegrees=45
  twistMotion=Free
 }
 {
  bone1Name=clavicle_r
  bone2Name=spine_05
  swing1LimitDegrees=45
  swing1Motion=Free
  swing2LimitDegrees=45
  swing2Motion=Free
  twistLimitDegrees=45
  twistMotion=Free
 }
 {
  bone1Name=upperarm_l
  bone2Name=clavicle_l
  swing1LimitDegrees=45
  swing1Motion=Free
  swing2LimitDegrees=45
  swing2Motion=Free
  twistLimitDegrees=45
  twistMotion=Free
 }
 {
  bone1Name=upperarm_r
  bone2Name=clavicle_r
  swing1LimitDegrees=45
  swing1Motion=Free
  swing2LimitDegrees=45
  swing2Motion=Free
  twistLimitDegrees=45
  twistMotion=Free
 }
 {
  bone1Name=neck_02
  bone2Name=neck_01
  swing1LimitDegrees=45
  swing1Motion=Free
  swing2LimitDegrees=45
  swing2Motion=Free
  twistLimitDegrees=45
  twistMotion=Free
 }
 {
  bone1Name=head
  bone2Name=neck_02
  swing1LimitDegrees=45
  swing1Motion=Free
  swing2LimitDegrees=45
  swing2Motion=Free
  twistLimitDegrees=45
  twistMotion=Free
 }
 {
  bone1Name=MSTR_hatSash
  bone2Name=head
  swing1LimitDegrees=45
  swing1Motion=Free
  swing2LimitDegrees=45
  swing2Motion=Free
  twistLimitDegrees=45
  twistMotion=Free
 }
 {
  bone1Name=MSTR_ponytail
  bone2Name=head
  swing1LimitDegrees=45
  swing1Motion=Free
  swing2LimitDegrees=45
  swing2Motion=Free
  twistLimitDegrees=45
  twistMotion=Free
 }
 {
  bone1Name=MSTR_sidebangs_l
  bone2Name=head
  swing1LimitDegrees=45
  swing1Motion=Free
  swing2LimitDegrees=45
  swing2Motion=Free
  twistLimitDegrees=45
  twistMotion=Free
 }
 {
  bone1Name=MSTR_hairRibbon_C_l
  bone2Name=head
  swing1LimitDegrees=45
  swing1Motion=Free
  swing2LimitDegrees=45
  swing2Motion=Free
  twistLimitDegrees=45
  twistMotion=Free
 }
 {
  bone1Name=MSTR_hairRibbon_C_r
  bone2Name=head
  swing1LimitDegrees=45
  swing1Motion=Free
  swing2LimitDegrees=45
  swing2Motion=Free
  twistLimitDegrees=45
  twistMotion=Free
 }
 {
  bone1Name=MSTR_strand_r
  bone2Name=head
  swing1LimitDegrees=45
  swing1Motion=Free
  swing2LimitDegrees=45
  swing2Motion=Free
  twistLimitDegrees=45
  twistMotion=Free
 }
 {
  bone1Name=dyn_hairRibbon_01_C_l
  bone2Name=MSTR_hairRibbon_C_l
  swing1LimitDegrees=15
  swing2LimitDegrees=3
  twistLimitDegrees=45
  twistMotion=Locked
 }
 {bone1Name=dyn_hairRibbon_02_C_l, bone2Name=dyn_hairRibbon_01_C_l, swing1LimitDegrees=15, swing2LimitDegrees=20, twistLimitDegrees=5}
 {
  bone1Name=dyn_hairRibbon_01_C_r
  bone2Name=MSTR_hairRibbon_C_r
  swing1LimitDegrees=12
  swing2LimitDegrees=10
  twistLimitDegrees=45
  twistMotion=Locked
 }
 {bone1Name=dyn_hairRibbon_02_C_r, bone2Name=dyn_hairRibbon_01_C_r, swing1LimitDegrees=15, swing2LimitDegrees=20, twistLimitDegrees=2.5}
 {bone1Name=dyn_hatSash_01, bone2Name=MSTR_hatSash, swing1LimitDegrees=45, swing2LimitDegrees=25, twistLimitDegrees=5}
 {bone1Name=dyn_hatSash_02, bone2Name=dyn_hatSash_01, swing1LimitDegrees=45, swing2LimitDegrees=25, twistLimitDegrees=10}
 {bone1Name=dyn_ponytail_01, bone2Name=MSTR_ponytail, swing1LimitDegrees=15, swing2LimitDegrees=5, twistLimitDegrees=3, twistMotion=Locked}
 {bone1Name=dyn_ponytail_02, bone2Name=dyn_ponytail_01, swing1LimitDegrees=30, swing2LimitDegrees=20, twistLimitDegrees=3}
 {bone1Name=dyn_ponytail_03, bone2Name=dyn_ponytail_02, swing1LimitDegrees=30, swing2LimitDegrees=40, twistLimitDegrees=3}
 {bone1Name=dyn_ponytail_04, bone2Name=dyn_ponytail_03, swing1LimitDegrees=35, swing2LimitDegrees=25, twistLimitDegrees=3}
 {bone1Name=dyn_ponytail_05, bone2Name=dyn_ponytail_04, swing1LimitDegrees=35, swing2LimitDegrees=25, twistLimitDegrees=3}
 {bone1Name=dyn_sidebangs_01_l, bone2Name=MSTR_sidebangs_l, swing1LimitDegrees=12, swing2LimitDegrees=20, twistLimitDegrees=0}
 {bone1Name=dyn_sidebangs_02_l, bone2Name=dyn_sidebangs_01_l, swing1LimitDegrees=16, swing2LimitDegrees=25, twistLimitDegrees=0}
 {bone1Name=dyn_sidebangs_03_l, bone2Name=dyn_sidebangs_02_l, swing1LimitDegrees=20, swing2LimitDegrees=25, twistLimitDegrees=0}
 {bone1Name=dyn_sidebangs_04_l, bone2Name=dyn_sidebangs_03_l, swing1LimitDegrees=20, swing2LimitDegrees=25, twistLimitDegrees=0}
 {bone1Name=dyn_strand_01_r, bone2Name=MSTR_strand_r, swing1LimitDegrees=20, swing2LimitDegrees=20, twistLimitDegrees=5, twistMotion=Locked}
 {bone1Name=dyn_strand_02_r, bone2Name=dyn_strand_01_r, swing1LimitDegrees=20, swing2LimitDegrees=20, twistLimitDegrees=5}
]
