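# Post-process style AnimBlueprint for the FlintTwirl female FaceAcc cosmetic part, animating ponytail, sidebangs and hair strands.
# It copies parent and Body poses, then runs head-space RigidBody simulation, using pre-posed bones normally and a separate physics asset during emotes.
# With rigid bodies disabled, spine-targeted constraints and offsets replace simulation, and the skin weight profile switches to Alternative.
# Event logic computes gravity, movement and glider values, plus a front-end idle ponytail tweak.
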AnimBlueprint /BRCosmetics/Characters/Player/Female/Medium/Bodies/F_MED_FlintTwirl/Meshes/Parts/F_MED_FlintTwirl_FaceAcc_AnimBP
target_skeleton=&/F_MED_FlintTwirl_FaceAcc_Skeleton
graphs [
 AnimGraph {nodes [
   n1 = Look At - Bone: MSTR_ponytail(ComponentPose=Local To Component(LocalPose=Use cached pose 'Post strands of hair pose'()), Alpha=1) {
    boneToModify {boneName=MSTR_ponytail}
    lookAtLocation=<-2.7993 10.8586 5.9562>
    lookAtTarget {boneReference {boneName=spine_05}}
    lookAt_Axis {axis=<1 0>}
   }
   n2 = Look At - Bone: MSTR_ponytail(ComponentPose=@n1, Alpha=1) {
    boneToModify {boneName=MSTR_ponytail}
    lookAtLocation=<-26.9927 77.1487 58.6931>
    lookAtTarget {boneReference {boneName=spine_05}}
    lookAt_Axis {axis=<1 0>}
   }
   n3 = "Transform (Modify) Bone - Bone: MSTR_ponytail"!(ComponentPose=@n2, bAlphaBoolEnabled:bool=True, Translation="0,0,0", Rotation="30,0,0", Scale="1,1,1") -> (Pose) {
    alphaInputType=Bool
    rotationMode=BMM_Additive
    rotationSpace=BCS_BoneSpace
    translationMode=BMM_Additive
    translationSpace=BCS_BoneSpace
    alphaBoolBlend {blendInTime=0.4, blendOutTime=0.4}
    boneToModify {boneName=MSTR_ponytail}
    rotation {pitch=30}
   }
   n4 = "Transform (Modify) Bone - Bone: upperarm_FA_extra_r"(ComponentPose=@n3, Alpha=1, Translation="0,2,15", Rotation="0,0,0", Scale="1,1,1") {
    rotationMode=BMM_Additive
    rotationSpace=BCS_BoneSpace
    scaleMode=BMM_Additive
    scaleSpace=BCS_BoneSpace
    translationMode=BMM_Additive
    translationSpace=BCS_BoneSpace
    alphaBoolBlend {blendInTime=0.4, blendOutTime=0.4}
    boneToModify {boneName=upperarm_FA_extra_r}
    translation {y=2, z=15}
   }
   n5 = "Transform (Modify) Bone - Bone: MSTR_ponytail"!(ComponentPose=@n4, Alpha:float=1, Translation="0,0,0", Rotation="20,0,0", Scale="1,1,1") -> (Pose) {
    rotationMode=BMM_Additive
    rotationSpace=BCS_BoneSpace
    scaleMode=BMM_Additive
    scaleSpace=BCS_BoneSpace
    translationMode=BMM_Additive
    translationSpace=BCS_BoneSpace
    alphaBoolBlend {blendInTime=0.4, blendOutTime=0.4}
    alphaScaleBiasClamp {bInterpResult=true}
    boneToModify {boneName=MSTR_ponytail}
    rotation {pitch=20}
   }
   n6 = RigidBody!(
    ComponentPose=@n5
    Alpha
    OverrideWorldGravity:Vector="0,0,0"
    ExternalForce
    SimSpaceSettings:Sim Space Settings Structure=(WorldAlpha=1,VelocityScaleZ=0.4,DampingAlpha=1,MaxLinearVelocity=300,MaxAngularVelocity=10,MaxLinearAcceleration=1000,MaxAngularAcceleration=40,ExternalLinearDragV=(X=0.3,Y=0.3,Z=0.3),ExternalLinearVelocity=(X=0,Y=0,Z=0),ExternalAngularVelocity=(X=0,Y=0,Z=0))
   ) -> (Pose) {
    bClampLinearTranslationLimitToRefPose=true
    bOverrideWorldGravity=true
    cachedBoundsScale=1
    overridePhysicsAsset=&/F_MED_FlintTwirl_FaceAcc_Physics
    simulationSpace=BaseBoneSpace
    alphaBoolBlend {blendOption=Linear}
    baseBoneRef {boneName=head}
    componentAppliedLinearAccClamp=<3000 3000 3000>
    overrideWorldGravity {z=-980}
    simSpaceSettings {
     maxAngularAcceleration=40
     maxAngularVelocity=10
     maxLinearAcceleration=1000
     maxLinearVelocity=300
     velocityScaleZ=0.4
     worldAlpha=1
     externalLinearDragV=<0.3 0.3 0.3>
    }
   }
   n7 = "Transform (Modify) Bone - Bone: dyn_ponytail_05"(ComponentPose=@n6, Alpha=Get FE Idle() -> (FE Idle:double), Translation="0,-1,0", Rotation="10,0,0", Scale="1,0.85,1") {
    rotationMode=BMM_Additive
    rotationSpace=BCS_BoneSpace
    scaleMode=BMM_Additive
    scaleSpace=BCS_BoneSpace
    translationMode=BMM_Additive
    translationSpace=BCS_BoneSpace
    alphaBoolBlend {blendInTime=0.4, blendOutTime=0.4}
    boneToModify {boneName=dyn_ponytail_05}
    rotation {pitch=10}
    scale {y=0.85}
    translation {y=-1}
   }
   n8 = Component To Local(ComponentPose=@n7)
   n9 = RigidBody(
    ComponentPose=Local To Component(LocalPose=Use cached pose 'Post strands of hair pose'())
    OverrideWorldGravity="0,0,0"
    SimSpaceSettings=(WorldAlpha=1,VelocityScaleZ=0.4,DampingAlpha=1,MaxLinearVelocity=300,MaxAngularVelocity=10,MaxLinearAcceleration=1000,MaxAngularAcceleration=40,ExternalLinearDragV=(X=0.3,Y=0.3,Z=0.3),ExternalLinearVelocity=(X=0,Y=0,Z=0),ExternalAngularVelocity=(X=0,Y=0,Z=0))
   ) {
    bClampLinearTranslationLimitToRefPose=true
    bOverrideWorldGravity=true
    cachedBoundsScale=1
    overridePhysicsAsset=&/F_MED_FlintTwirl_FaceAcc_Physics_Emote
    simulationSpace=BaseBoneSpace
    alphaBoolBlend {blendOption=Linear}
    baseBoneRef {boneName=head}
    componentAppliedLinearAccClamp=<3000 3000 3000>
    overrideWorldGravity {z=-980}
    simSpaceSettings {
     maxAngularAcceleration=40
     maxAngularVelocity=10
     maxLinearAcceleration=1000
     maxLinearVelocity=300
     velocityScaleZ=0.4
     worldAlpha=1
     externalLinearDragV=<0.3 0.3 0.3>
    }
   }
   n10 = Component To Local(ComponentPose=@n9)
   n11 = Get bIsPlayingEmote()
   n12 = "Constraint - dyn_ponytail_03 (P)"(ComponentPose=Local To Component(LocalPose=Use cached pose 'Post strands of hair pose'()), Alpha=1, ConstraintWeights_0=1) {boneToModify {boneName=dyn_ponytail_03}, constraintSetup [{targetBone {boneName=spine_05}}], constraintWeights [1]}
   n13 = "Constraint - dyn_ponytail_04 (P)"(ComponentPose=@n12, Alpha=1, ConstraintWeights_0=1) {boneToModify {boneName=dyn_ponytail_04}, constraintSetup [{targetBone {boneName=spine_05}}], constraintWeights [1]}
   n14 = "Constraint - dyn_ponytail_05 (P)"(ComponentPose=@n13, Alpha=1, ConstraintWeights_0=1) {boneToModify {boneName=dyn_ponytail_05}, constraintSetup [{targetBone {boneName=spine_05}}], constraintWeights [1]}
   n15 = "Constraint - dyn_sidebangs_03_l (P)"(ComponentPose=@n14, Alpha=0.5, ConstraintWeights_0=1) {boneToModify {boneName=dyn_sidebangs_03_l}, constraintSetup [{targetBone {boneName=spine_05}}], constraintWeights [1]}
   n16 = "Constraint - dyn_sidebangs_04_l (P)"(ComponentPose=@n15, Alpha=1, ConstraintWeights_0=1) {boneToModify {boneName=dyn_sidebangs_04_l}, constraintSetup [{targetBone {boneName=spine_05}}], constraintWeights [1]}
   n17 = "Transform (Modify) Bone - Bone: dyn_sidebangs_01_l"!(ComponentPose=@n16, bAlphaBoolEnabled:bool=True, Translation="0,0,0.5", Rotation="0,0,0", Scale="1,1,1") -> (Pose) {
    alphaInputType=Bool
    rotationMode=BMM_Additive
    rotationSpace=BCS_BoneSpace
    translationMode=BMM_Additive
    translationSpace=BCS_BoneSpace
    alphaBoolBlend {blendInTime=0.2, blendOutTime=0.2}
    boneToModify {boneName=dyn_sidebangs_01_l}
    translation {z=0.5}
   }
   n18 = "Transform (Modify) Bone - Bone: dyn_strand_01_r"!(ComponentPose=@n17, bAlphaBoolEnabled:bool=True, Translation="0,0.5,0", Rotation="0,0,0", Scale="1,1,1") -> (Pose) {
    alphaInputType=Bool
    rotationMode=BMM_Additive
    rotationSpace=BCS_BoneSpace
    translationMode=BMM_Additive
    translationSpace=BCS_BoneSpace
    alphaBoolBlend {blendInTime=0.2, blendOutTime=0.2}
    boneToModify {boneName=dyn_strand_01_r}
    translation {y=0.5}
   }
   Output Pose(Result=Blend Poses by bool(
     BlendPose_0=Blend Poses by bool(BlendPose_0=@n10, BlendPose_1=@n8, bActiveValue=@n11)
     BlendPose_1=Component To Local(ComponentPose=@n18)
     bActiveValue=Get bRigidBodyEnabled()
    )) {layerGroup=None}
   n19 = Layered blend per bone(
    BasePose=Copy Pose From Mesh() {bUseAttachedParent=true}
    BlendPoses_0=Copy Pose From Mesh(SourceMeshComponent=Get Body() -> (Body:Skeletal Mesh Component Budgeted Object Reference)) {bUseAttachedParent=true}
   ) {layerSetup [{branchFilters [{boneName=clavicle_l}, {boneName=clavicle_r}, {boneName=upperarm_l}, {boneName=upperarm_r}]}]}
   Save cached pose 'Post strands of hair pose'(Pose=@n19) {cacheName=Post strands of hair pose}
  ]}
 EventGraph {nodes [
   Event BlueprintUpdateAnimation()
   > n1 = Sequence()
   > n2 = "Gravity Override State Function (Parameter Asset) (Nativized)"(
    CharacterPartAnimInstance=Self-Reference()
    GameTimeInSeconds=GetGameTimeInSeconds()
    Parameters=&/F_MED_FlintTwirl_FaceAcc_Parameters
   )
   > Set Gravity Override(Gravity Override=@n2)
   > Set RBAN Settings Out(RBAN Settings Out=@n2.OutRigidBodyAnimNodeParameters)
   n3 = Get bRigidBodyEnabled()
   > Set RBDisable(RBDisable:double="To Float (Boolean)"(InBool=NOT Boolean(A=@n3))) -> (Output_Get:double)
   n4 = NOT Boolean(A=Get bIsSkyDiving())
   n5 = NOT Boolean(A=Get bIsPlayingEmote())
   n6 = NOT Boolean(A=Get bIsMoving2D())
   n7 = Get bIsInFrontEnd()
   n8 = Get bRigidBodyEnabled()
   > Set FE Idle(FE Idle:double="To Float (Boolean)"(InBool=AND Boolean!(A=@n7, B=@n6, C:bool=@n5, D:bool=@n4, E:bool=@n8) -> (ReturnValue))) -> (Output_Get:double)
   > Branch(Condition=NOT Boolean(A=Get bRigidBodyEnabled()))
   > SetSkinWeightProfile!(
    execute
    self:Skinned Mesh Component Object Reference=Get Body Skel Part() -> (Body Skel Part:Skeletal Mesh Component Object Reference)
    InProfileName:Name=Alternative
    InLayer:ESkinWeightProfileLayer Enum=Primary
   ) -> (then, ReturnValue:bool)
   n9 = Player Movement Function DEPRECATED!(
    execute=@n1.then_1
    CharacterPartAnimInstance:Custom Character Part Anim Instance Object Reference=Self-Reference()
    X Dot Product:double=100
    Y Dot Product:double=100
    Z Dot Product:double=100
    DrawDebug:bool=false
    Debug Scalar:double=50
   ) -> (then, X Movement:double, Y Movement:double, Z Movement:double)
   n10 = NOT Boolean(A=Get bIsSkyDiving())
   n11 = "float > float"(A=BreakVector(InVec=MakeVector(X=@n9, Y=@n9.Y Movement)), B=50)
   > Set Right(Right:double="To Float (Boolean)"(InBool=AND Boolean!(A=@n11, B=@n10) -> (ReturnValue))) -> (Output_Get:double)
   Event OnParachuteOpened()
   > Switch on EGliderType(Selection=Get PlayerGliderType() -> (PlayerGliderType=HangGlider))
   > Set HangGlidingAlpha(HangGlidingAlpha:double=1) -> (Output_Get:double)
   Event OnParachuteClosed()
   > Set HangGlidingAlpha(HangGlidingAlpha=0)
   n12 = Event OnCharacterCustomizationLoaded()
   > Set Body(Body:Skeletal Mesh Component Budgeted Object Reference=Get Array Item with EFort Custom Part Type(Array=@n12.CustomMeshParts)) -> (Output_Get:Skeletal Mesh Component Budgeted Object Reference)
   > Set Body Skel Part(Body Skel Part:Skeletal Mesh Component Object Reference=GetPartSkeletalMeshComponent()) -> (Output_Get:Skeletal Mesh Component Object Reference)
  ]}
]
variables [
 Body {type=SkeletalMeshComponentBudgeted}
 Body Skel Part {type=SkeletalMeshComponent}
 FE Idle {type=number}
 Gravity Override {type=Vector}
 HangGlidingAlpha {type=number}
 RBAN Settings Out {type=FortRigidBodyAnimNodeParameters}
 RBDisable {type=number}
 Right {type=number}
]
